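# Level Sequence that animates a player mannequin for the NickelScan_M custom skin's Battle Royale menu animation.
# A spawned FortPlayerMannequin is posed through a Fortnite_Biped_CtrlRig Control Rig track, with its transform at the origin.
# Legs use IK with constant foot placements, while arms, spine and neck are set to FK.
# Visible animated channels return to their first-frame values at frame 351, consistent with a seamless idle loop.
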LevelSequence /BRCosmetics/Animation/Game/MainPlayer/Menu/BR/CustomSkins/_s40/NickelScan_M/NickelScan_Anim/NickelScan_Anim
bindings [
 mannequin {
  id=#1
  possessed_class=FortPlayerMannequin
  tracks [
   {type=MovieSceneSpawnTrack, sections [{type=MovieSceneSpawnSection, channels [None {type=MovieSceneScriptingBoolChannel, default=true}]}]}
   {
    type=MovieScene3DTransformTrack
    sections [{
      type=MovieScene3DTransformSection
      channels [
       Location.X {type=MovieSceneScriptingDoubleChannel, default=0}
       Location.Y {type=MovieSceneScriptingDoubleChannel, default=0}
       Location.Z {type=MovieSceneScriptingDoubleChannel, default=0}
       Rotation.X {type=MovieSceneScriptingDoubleChannel, default=0}
       Rotation.Y {type=MovieSceneScriptingDoubleChannel, default=0}
       Rotation.Z {type=MovieSceneScriptingDoubleChannel, default=0}
       Scale.X {type=MovieSceneScriptingDoubleChannel, default=1}
       Scale.Y {type=MovieSceneScriptingDoubleChannel, default=1}
       Scale.Z {type=MovieSceneScriptingDoubleChannel, default=1}
      ]
     }]
   }
   {
    type=MovieSceneControlRigParameterTrack
    display_name=Fortnite_Biped_CtrlRig
    sections [{
      type=MovieSceneControlRigParameterSection
      channels [
       leg_l_stretch_switch {type=MovieSceneScriptingBoolChannel, default=true, const [true,59]}
       leg_r_stretch_switch {type=MovieSceneScriptingBoolChannel, default=true, const [true,58]}
       arm_l_fk_local {type=MovieSceneScriptingBoolChannel, default=false, const [false,55]}
       Auto Interpolate_3 {type=MovieSceneScriptingBoolChannel, default=false, const [false,56]}
       Auto Interpolate {type=MovieSceneScriptingBoolChannel, default=false, const [false,56]}
       arm_l_stretch_switch {type=MovieSceneScriptingBoolChannel, default=false, const [false,58]}
       Keep Aligned {type=MovieSceneScriptingBoolChannel, default=false, const [false,58]}
       arm_r_fk_local {type=MovieSceneScriptingBoolChannel, default=false, const [false,53]}
       Auto Interpolate_4 {type=MovieSceneScriptingBoolChannel, default=false, const [false,55]}
       Auto Interpolate_2 {type=MovieSceneScriptingBoolChannel, default=false, const [false,55]}
       arm_r_stretch_switch {type=MovieSceneScriptingBoolChannel, default=false, const [false,58]}
       Keep Aligned_2 {type=MovieSceneScriptingBoolChannel, default=false, const [false,58]}
       head_local {type=MovieSceneScriptingBoolChannel, default=true, const [true,58]}
       Stretch {type=MovieSceneScriptingBoolChannel, default=false, const [false,58]}
       head_stretch {type=MovieSceneScriptingBoolChannel, default=false, const [false,58]}
       spine_stretch_switch {type=MovieSceneScriptingBoolChannel, default=false, const [false,58]}
       arm_l_fk_ik_switch {type=MovieSceneScriptingBoolChannel, default=false, const [false,58]}
       leg_l_fk_ik_switch {type=MovieSceneScriptingBoolChannel, default=true, const [true,58]}
       arm_r_fk_ik_switch {type=MovieSceneScriptingBoolChannel, default=false, const [false,58]}
       leg_r_fk_ik_switch {type=MovieSceneScriptingBoolChannel, default=true, const [true,58]}
       ShowBodyControls {type=MovieSceneScriptingBoolChannel, default=true, const [true,58]}
       spine_fk_ik_switch {type=MovieSceneScriptingBoolChannel, default=false, const [false,58]}
       neck_fk_ik_switch {type=MovieSceneScriptingBoolChannel, default=false, const [false,58]}
       space {type=MovieSceneScriptingByteChannel, default=2, const [2,58]}
       global_ctrl.Location.X {type=MovieSceneScriptingFloatChannel, default=0, const [0,58]}
       global_ctrl.Location.Y {type=MovieSceneScriptingFloatChannel, default=0, const [0,58]}
       global_ctrl.Location.Z {type=MovieSceneScriptingFloatChannel, default=0, const [0,58]}
       global_ctrl.Rotation.X {type=MovieSceneScriptingFloatChannel, default=0, const [0,58]}
       global_ctrl.Rotation.Y {type=MovieSceneScriptingFloatChannel, default=0, const [0,58]}
       global_ctrl.Rotation.Z {type=MovieSceneScriptingFloatChannel, default=0, const [0,58]}
       root_ctrl.Location.X {type=MovieSceneScriptingFloatChannel, default=0, const [0,58]}
       root_ctrl.Location.Y {type=MovieSceneScriptingFloatChannel, default=0, const [0,58]}
       root_ctrl.Location.Z {type=MovieSceneScriptingFloatChannel, default=0, const [0,58]}
       root_ctrl.Rotation.X {type=MovieSceneScriptingFloatChannel, default=0, const [0,58]}
       root_ctrl.Rotation.Y {type=MovieSceneScriptingFloatChannel, default=0, const [0,58]}
       root_ctrl.Rotation.Z {type=MovieSceneScriptingFloatChannel, default=0, const [0,58]}
       root_ctrl.Scale.X {type=MovieSceneScriptingFloatChannel, default=1, const [1,58]}
       root_ctrl.Scale.Y {type=MovieSceneScriptingFloatChannel, default=1, const [1,58]}
       root_ctrl.Scale.Z {type=MovieSceneScriptingFloatChannel, default=1, const [1,58]}
       body_offset_ctrl.Location.X {type=MovieSceneScriptingFloatChannel, default=0, const [-5.235,58]}
       body_offset_ctrl.Location.Y {type=MovieSceneScriptingFloatChannel, default=0, const [0,58]}
       body_offset_ctrl.Location.Z {type=MovieSceneScriptingFloatChannel, default=0, const [0,58]}
       body_offset_ctrl.Rotation.X {type=MovieSceneScriptingFloatChannel, default=0, const [0,58]}
       body_offset_ctrl.Rotation.Y {type=MovieSceneScriptingFloatChannel, default=0, const [0,58]}
       body_offset_ctrl.Rotation.Z {type=MovieSceneScriptingFloatChannel, default=0, const [0,58]}
       body_offset_ctrl.Scale.X {type=MovieSceneScriptingFloatChannel, default=1, const [1,58]}
       body_offset_ctrl.Scale.Y {type=MovieSceneScriptingFloatChannel, default=1, const [1,58]}
       body_offset_ctrl.Scale.Z {type=MovieSceneScriptingFloatChannel, default=1, const [1,58]}
       thigh_l_fk_ctrl.Location.X {type=MovieSceneScriptingFloatChannel, default=0, thinned_from=58, keys [[0,0],[351,0]]}
       thigh_l_fk_ctrl.Location.Y {type=MovieSceneScriptingFloatChannel, default=0, thinned_from=58, keys [[0,0],[351,0]]}
       thigh_l_fk_ctrl.Location.Z {type=MovieSceneScriptingFloatChannel, default=0, thinned_from=58, keys [[0,0],[351,0]]}
       thigh_l_fk_ctrl.Rotation.X {
        type=MovieSceneScriptingFloatChannel
        default=1.216
        thinned_from=58
        keys [[0,19.2811],[31,19.1067],[41,19.0901],[48,19.1439],[54,19.2334],[64,19.2811],[76,19.2095],[82,19.1349],[90,20.41],[101,20.835],[109,19.6894],[115,20.4384],[119,20.6256],[126,20.688],[145,20.6876],[152,20.1932],[159,20.2575],[166,20.4502],[178,20.3712],[183,20.249],[193,20.0878],[205,20.0585],[212,19.5798],[216,18.782],[220,18.3033],[225,18.1238],[239,17.9264],[244,17.7769],[249,17.5278],[255,17.4343],[263,17.3783],[271,17.735],[277,18.8054],[283,19.1027],[292,19.2365],[351,19.2811]]
       }
       thigh_l_fk_ctrl.Rotation.Y {
        type=MovieSceneScriptingFloatChannel
        default=-1.2128
        thinned_from=58
        keys [[0,2.581],[9,2.5997],[13,2.6021],[19,2.5941],[25,2.5807],[31,2.5727],[41,2.5679],[64,2.581],[82,2.5709],[90,2.5538],[101,2.5481],[109,2.6102],[115,2.6631],[119,2.6763],[145,2.6835],[152,2.7191],[159,2.7135],[166,2.6964],[178,2.7006],[193,2.7155],[205,2.717],[212,2.8039],[216,2.9486],[220,3.0354],[225,3.0679],[239,3.1062],[244,3.1374],[249,3.1894],[255,3.2089],[263,3.2206],[271,3.1007],[277,2.7409],[283,2.641],[292,2.596],[301,2.581],[351,2.581]]
       }
       thigh_l_fk_ctrl.Rotation.Z {
        type=MovieSceneScriptingFloatChannel
        default=-0.1664
        thinned_from=58
        keys [[0,20.0947],[13,20.0878],[19,19.892],[25,19.5659],[31,19.3702],[41,19.2527],[48,19.4895],[54,19.8842],[59,20.0158],[64,20.0947],[76,19.7789],[82,19.4501],[90,17.8491],[101,17.3155],[109,18.7825],[115,18.1946],[119,18.0476],[126,17.9987],[145,18.0034],[152,18.0519],[159,18.0332],[166,17.9771],[205,17.9323],[212,17.6388],[216,17.1496],[220,16.8561],[225,16.746],[232,16.68],[239,16.9216],[244,17.3242],[249,17.9953],[255,18.2469],[263,18.3979],[271,18.7161],[277,19.6705],[283,19.9356],[292,20.0549],[351,20.0947]]
       }
       thigh_l_fk_ctrl.Scale.X {type=MovieSceneScriptingFloatChannel, default=1, const [1,58]}
       thigh_l_fk_ctrl.Scale.Y {type=MovieSceneScriptingFloatChannel, default=1, const [1,58]}
       thigh_l_fk_ctrl.Scale.Z {type=MovieSceneScriptingFloatChannel, default=1, const [1,58]}
       calf_l_fk_ctrl.Location.X {
        type=MovieSceneScriptingFloatChannel
        default=0
        thinned_from=58
        keys [[0,-1.1433],[3,-1.1377],[6,-1.1284],[9,-1.1221],[13,-1.1194],[19,-1.1441],[25,-1.1852],[31,-1.2099],[41,-1.2247],[48,-1.2018],[54,-1.1637],[59,-1.151],[64,-1.1433],[76,-1.1739],[82,-1.2057],[90,-1.1742],[101,-1.1637],[109,-1.2203],[115,-1.2147],[126,-1.2128],[133,-1.2088],[139,-1.202],[145,-1.1979],[152,-1.1881],[166,-1.1896],[178,-1.1876],[193,-1.1806],[205,-1.1799],[212,-1.1623],[216,-1.1331],[220,-1.1156],[225,-1.109],[232,-1.105],[244,-1.1014],[249,-1.0976],[263,-1.0953],[271,-1.1043],[277,-1.1313],[283,-1.1388],[351,-1.1433]]
       }
       calf_l_fk_ctrl.Location.Y {
        type=MovieSceneScriptingFloatChannel
        default=0
        thinned_from=58
        keys [[0,0.0855],[3,0.0851],[6,0.0844],[9,0.0839],[13,0.0837],[19,0.0856],[25,0.0886],[31,0.0905],[41,0.0916],[48,0.0899],[54,0.087],[59,0.0861],[64,0.0855],[76,0.0878],[82,0.0902],[90,0.0878],[101,0.087],[109,0.0913],[115,0.0908],[126,0.0907],[133,0.0904],[139,0.0899],[145,0.0896],[152,0.0889],[178,0.0888],[193,0.0883],[205,0.0882],[212,0.0869],[216,0.0847],[220,0.0834],[225,0.0829],[244,0.0824],[249,0.0821],[263,0.0819],[271,0.0826],[277,0.0846],[283,0.0852],[351,0.0855]]
       }
       calf_l_fk_ctrl.Location.Z {
        type=MovieSceneScriptingFloatChannel
        default=0
        thinned_from=58
        keys [[0,-0.0245],[13,-0.024],[19,-0.0245],[25,-0.0254],[31,-0.0259],[41,-0.0262],[48,-0.0258],[54,-0.0249],[64,-0.0245],[76,-0.0252],[82,-0.0258],[90,-0.0252],[101,-0.0249],[109,-0.0261],[152,-0.0255],[205,-0.0253],[212,-0.0249],[216,-0.0243],[220,-0.0239],[263,-0.0235],[271,-0.0237],[277,-0.0242],[351,-0.0245]]
       }
       calf_l_fk_ctrl.Rotation.X {type=MovieSceneScriptingFloatChannel, default=0.0543, thinned_from=58, keys [[0,-0.0451],[351,-0.0451]]}
       calf_l_fk_ctrl.Rotation.Y {type=MovieSceneScriptingFloatChannel, default=1.2238, thinned_from=58, keys [[0,1.2242],[351,1.2242]]}
       calf_l_fk_ctrl.Rotation.Z {type=MovieSceneScriptingFloatChannel, default=0.3732, thinned_from=58, keys [[0,-4.277],[351,-4.277]]}
       calf_l_fk_ctrl.Scale.X {type=MovieSceneScriptingFloatChannel, default=1, const [1,58]}
       calf_l_fk_ctrl.Scale.Y {type=MovieSceneScriptingFloatChannel, default=1, const [1,58]}
       calf_l_fk_ctrl.Scale.Z {type=MovieSceneScriptingFloatChannel, default=1, const [1,58]}
       foot_l_fk_ctrl.Location.X {
        type=MovieSceneScriptingFloatChannel
        default=0
        thinned_from=58
        keys [[0,-1.2497],[3,-1.2436],[6,-1.2334],[9,-1.2266],[13,-1.2236],[19,-1.2506],[25,-1.2956],[31,-1.3225],[41,-1.3387],[48,-1.3137],[54,-1.272],[59,-1.2581],[64,-1.2497],[76,-1.2831],[82,-1.3179],[90,-1.2835],[101,-1.272],[109,-1.3338],[115,-1.3277],[126,-1.3257],[133,-1.3213],[139,-1.3139],[145,-1.3094],[152,-1.2987],[166,-1.3003],[178,-1.2982],[193,-1.2905],[205,-1.2897],[212,-1.2705],[216,-1.2386],[220,-1.2194],[225,-1.2122],[232,-1.2079],[244,-1.2039],[249,-1.1997],[263,-1.1972],[271,-1.2071],[277,-1.2366],[283,-1.2448],[351,-1.2497]]
       }
       foot_l_fk_ctrl.Location.Y {type=MovieSceneScriptingFloatChannel, default=0.0398, thinned_from=58, keys [[0,0.0398],[351,0.0398]]}
       foot_l_fk_ctrl.Location.Z {type=MovieSceneScriptingFloatChannel, default=-0.8885, thinned_from=58, keys [[0,-0.8885],[351,-0.8885]]}
       foot_l_fk_ctrl.Rotation.X {
        type=MovieSceneScriptingFloatChannel
        default=-1.1647
        thinned_from=58
        keys [[0,-0.0005],[9,0],[13,0],[19,-0.0039],[25,-0.0105],[31,-0.0145],[41,-0.0168],[48,-0.0123],[54,-0.0046],[59,-0.0021],[64,-0.0005],[76,-0.0066],[82,-0.013],[90,-0.0747],[101,-0.0952],[109,-0.0393],[115,-0.0674],[119,-0.0744],[126,-0.0768],[133,-0.0765],[145,-0.0757],[152,-0.0732],[159,-0.0738],[166,-0.0756],[205,-0.0757],[212,-0.0813],[216,-0.0907],[220,-0.0963],[225,-0.0984],[232,-0.0996],[239,-0.0942],[244,-0.0851],[249,-0.0699],[255,-0.0642],[263,-0.0607],[271,-0.0494],[277,-0.0156],[283,-0.0062],[292,-0.0019],[301,-0.0005],[351,-0.0005]]
       }
       foot_l_fk_ctrl.Rotation.Y {
        type=MovieSceneScriptingFloatChannel
        default=0.0792
        thinned_from=58
        keys [[0,-5.3483],[13,-5.3534],[31,-5.3236],[41,-5.3188],[64,-5.3483],[82,-5.3257],[90,-5.6302],[101,-5.7317],[109,-5.5065],[115,-5.741],[119,-5.7997],[126,-5.8192],[145,-5.8234],[152,-5.9302],[159,-5.9166],[166,-5.8756],[178,-5.8928],[193,-5.9543],[205,-5.9607],[212,-5.8887],[216,-5.7686],[220,-5.6966],[225,-5.6696],[255,-5.6191],[263,-5.6158],[271,-5.5656],[277,-5.4152],[283,-5.3734],[292,-5.3546],[351,-5.3483]]
       }
       foot_l_fk_ctrl.Rotation.Z {
        type=MovieSceneScriptingFloatChannel
        default=-0.2074
        thinned_from=58
        keys [[0,-5.1841],[13,-5.1846],[19,-5.1673],[25,-5.1385],[31,-5.1212],[41,-5.1108],[48,-5.1315],[54,-5.1658],[64,-5.1841],[76,-5.1567],[82,-5.128],[90,-4.6578],[101,-4.501],[109,-4.904],[115,-4.6454],[119,-4.5807],[126,-4.5592],[145,-4.5627],[152,-4.5303],[166,-4.541],[178,-4.5336],[193,-4.5069],[232,-4.5118],[239,-4.543],[244,-4.5949],[249,-4.6816],[255,-4.714],[263,-4.7335],[271,-4.818],[277,-5.0715],[283,-5.1419],[292,-5.1736],[351,-5.1841]]
       }
       foot_l_fk_ctrl.Scale.X {type=MovieSceneScriptingFloatChannel, default=1, const [1,58]}
       foot_l_fk_ctrl.Scale.Y {type=MovieSceneScriptingFloatChannel, default=1, const [1,58]}
       foot_l_fk_ctrl.Scale.Z {type=MovieSceneScriptingFloatChannel, default=1, const [1,58]}
       ball_l_fk_ctrl.Location.X {type=MovieSceneScriptingFloatChannel, default=0, const [0,58]}
       ball_l_fk_ctrl.Location.Y {type=MovieSceneScriptingFloatChannel, default=0, const [0,58]}
       ball_l_fk_ctrl.Location.Z {type=MovieSceneScriptingFloatChannel, default=0, const [0,58]}
       ball_l_fk_ctrl.Rotation.X {type=MovieSceneScriptingFloatChannel, default=0, const [0,58]}
       ball_l_fk_ctrl.Rotation.Y {type=MovieSceneScriptingFloatChannel, default=0, const [0,58]}
       ball_l_fk_ctrl.Rotation.Z {type=MovieSceneScriptingFloatChannel, default=0, const [0,58]}
       ball_l_fk_ctrl.Scale.X {type=MovieSceneScriptingFloatChannel, default=1, const [1,58]}
       ball_l_fk_ctrl.Scale.Y {type=MovieSceneScriptingFloatChannel, default=1, const [1,58]}
       ball_l_fk_ctrl.Scale.Z {type=MovieSceneScriptingFloatChannel, default=1, const [1,58]}
       foot_l_ik_ctrl.Location.X {type=MovieSceneScriptingFloatChannel, default=0, const [9.492,59]}
       foot_l_ik_ctrl.Location.Y {type=MovieSceneScriptingFloatChannel, default=0, const [14.5163,60]}
       foot_l_ik_ctrl.Location.Z {type=MovieSceneScriptingFloatChannel, default=0, const [0,59]}
       foot_l_ik_ctrl.Rotation.X {type=MovieSceneScriptingFloatChannel, default=0, const [0,59]}
       foot_l_ik_ctrl.Rotation.Y {type=MovieSceneScriptingFloatChannel, default=0, const [0,59]}
       foot_l_ik_ctrl.Rotation.Z {type=MovieSceneScriptingFloatChannel, default=0, const [-10.5,59]}
       foot_l_ik_ctrl.Scale.X {type=MovieSceneScriptingFloatChannel, default=1, const [1,59]}
       foot_l_ik_ctrl.Scale.Y {type=MovieSceneScriptingFloatChannel, default=1, const [1,59]}
       foot_l_ik_ctrl.Scale.Z {type=MovieSceneScriptingFloatChannel, default=1, const [1,59]}
       ball_l_ik_ctrl.Location.X {type=MovieSceneScriptingFloatChannel, default=0, const [0,58]}
       ball_l_ik_ctrl.Location.Y {type=MovieSceneScriptingFloatChannel, default=0, const [0,58]}
       ball_l_ik_ctrl.Location.Z {type=MovieSceneScriptingFloatChannel, default=0, const [0,58]}
       ball_l_ik_ctrl.Rotation.X {type=MovieSceneScriptingFloatChannel, default=0, const [0,58]}
       ball_l_ik_ctrl.Rotation.Y {type=MovieSceneScriptingFloatChannel, default=0, const [0,58]}
       ball_l_ik_ctrl.Rotation.Z {type=MovieSceneScriptingFloatChannel, default=0, const [0,58]}
       ball_l_ik_ctrl.Scale.X {type=MovieSceneScriptingFloatChannel, default=1, const [1,58]}
       ball_l_ik_ctrl.Scale.Y {type=MovieSceneScriptingFloatChannel, default=1, const [1,58]}
       ball_l_ik_ctrl.Scale.Z {type=MovieSceneScriptingFloatChannel, default=1, const [1,58]}
       foot_bk1_l_ctrl.Location.X {type=MovieSceneScriptingFloatChannel, default=0, const [0,58]}
       foot_bk1_l_ctrl.Location.Y {type=MovieSceneScriptingFloatChannel, default=0, const [0,58]}
       foot_bk1_l_ctrl.Location.Z {type=MovieSceneScriptingFloatChannel, default=0, const [0,58]}
       foot_bk1_l_ctrl.Rotation.X {type=MovieSceneScriptingFloatChannel, default=0, const [0,58]}
       foot_bk1_l_ctrl.Rotation.Y {type=MovieSceneScriptingFloatChannel, default=0, const [0,58]}
       foot_bk1_l_ctrl.Rotation.Z {type=MovieSceneScriptingFloatChannel, default=0, const [0,58]}
       foot_bk1_l_ctrl.Scale.X {type=MovieSceneScriptingFloatChannel, default=1, const [1,58]}
       foot_bk1_l_ctrl.Scale.Y {type=MovieSceneScriptingFloatChannel, default=1, const [1,58]}
       foot_bk1_l_ctrl.Scale.Z {type=MovieSceneScriptingFloatChannel, default=1, const [1,58]}
       foot_roll_l_ctrl.Location.X {type=MovieSceneScriptingFloatChannel, default=0, const [0,58]}
       foot_roll_l_ctrl.Location.Y {type=MovieSceneScriptingFloatChannel, default=0, const [0,58]}
       foot_roll_l_ctrl.Location.Z {type=MovieSceneScriptingFloatChannel, default=0, const [0,58]}
       foot_roll_l_ctrl.Rotation.X {type=MovieSceneScriptingFloatChannel, default=0, const [0,58]}
       foot_roll_l_ctrl.Rotation.Y {type=MovieSceneScriptingFloatChannel, default=0, const [0,58]}
       foot_roll_l_ctrl.Rotation.Z {type=MovieSceneScriptingFloatChannel, default=0, const [0,58]}
       foot_roll_l_ctrl.Scale.X {type=MovieSceneScriptingFloatChannel, default=1, const [1,58]}
       foot_roll_l_ctrl.Scale.Y {type=MovieSceneScriptingFloatChannel, default=1, const [1,58]}
       foot_roll_l_ctrl.Scale.Z {type=MovieSceneScriptingFloatChannel, default=1, const [1,58]}
       roll_blend_l_ctrl {type=MovieSceneScriptingFloatChannel, default=1, const [1,58]}
       tip_l_ctrl.Location.X {type=MovieSceneScriptingFloatChannel, default=0, const [0,58]}
       tip_l_ctrl.Location.Y {type=MovieSceneScriptingFloatChannel, default=0, const [0,58]}
       tip_l_ctrl.Location.Z {type=MovieSceneScriptingFloatChannel, default=0, const [0,58]}
       tip_l_ctrl.Rotation.X {type=MovieSceneScriptingFloatChannel, default=0, const [0,58]}
       tip_l_ctrl.Rotation.Y {type=MovieSceneScriptingFloatChannel, default=0, const [0,58]}
       tip_l_ctrl.Rotation.Z {type=MovieSceneScriptingFloatChannel, default=0, const [0,58]}
       tip_l_ctrl.Scale.X {type=MovieSceneScriptingFloatChannel, default=1, const [1,58]}
       tip_l_ctrl.Scale.Y {type=MovieSceneScriptingFloatChannel, default=1, const [1,58]}
       tip_l_ctrl.Scale.Z {type=MovieSceneScriptingFloatChannel, default=1, const [1,58]}
       heel_l_ctrl.Location.X {type=MovieSceneScriptingFloatChannel, default=0, const [0,58]}
       heel_l_ctrl.Location.Y {type=MovieSceneScriptingFloatChannel, default=0, const [0,58]}
       heel_l_ctrl.Location.Z {type=MovieSceneScriptingFloatChannel, default=0, const [0,58]}
       heel_l_ctrl.Rotation.X {type=MovieSceneScriptingFloatChannel, default=0, const [0,58]}
       heel_l_ctrl.Rotation.Y {type=MovieSceneScriptingFloatChannel, default=0, const [0,58]}
       heel_l_ctrl.Rotation.Z {type=MovieSceneScriptingFloatChannel, default=0, const [0,58]}
       heel_l_ctrl.Scale.X {type=MovieSceneScriptingFloatChannel, default=1, const [1,58]}
       heel_l_ctrl.Scale.Y {type=MovieSceneScriptingFloatChannel, default=1, const [1,58]}
       heel_l_ctrl.Scale.Z {type=MovieSceneScriptingFloatChannel, default=1, const [1,58]}
       leg_l_softness {type=MovieSceneScriptingFloatChannel, default=0, const [0,59]}
       leg_l_pv_ik_ctrl.Location.X {type=MovieSceneScriptingFloatChannel, default=0, const [0,58]}
       leg_l_pv_ik_ctrl.Location.Y {type=MovieSceneScriptingFloatChannel, default=0, const [0,58]}
       leg_l_pv_ik_ctrl.Location.Z {type=MovieSceneScriptingFloatChannel, default=0, const [0,58]}
       leg_l_pv_ik_ctrl.Rotation.X {type=MovieSceneScriptingFloatChannel, default=0, const [0,58]}
       leg_l_pv_ik_ctrl.Rotation.Y {type=MovieSceneScriptingFloatChannel, default=0, const [0,58]}
       leg_l_pv_ik_ctrl.Rotation.Z {type=MovieSceneScriptingFloatChannel, default=0, const [0,58]}
       leg_l_pv_ik_ctrl.Scale.X {type=MovieSceneScriptingFloatChannel, default=1, const [1,58]}
       leg_l_pv_ik_ctrl.Scale.Y {type=MovieSceneScriptingFloatChannel, default=1, const [1,58]}
       leg_l_pv_ik_ctrl.Scale.Z {type=MovieSceneScriptingFloatChannel, default=1, const [1,58]}
       thigh_r_fk_ctrl.Location.X {type=MovieSceneScriptingFloatChannel, default=0, thinned_from=58, keys [[0,0],[351,0]]}
       thigh_r_fk_ctrl.Location.Y {type=MovieSceneScriptingFloatChannel, default=0, thinned_from=58, keys [[0,0],[351,0]]}
       thigh_r_fk_ctrl.Location.Z {type=MovieSceneScriptingFloatChannel, default=0, thinned_from=58, keys [[0,0],[351,0]]}
       thigh_r_fk_ctrl.Rotation.X {
        type=MovieSceneScriptingFloatChannel
        default=1.2139
        thinned_from=58
        keys [[0,31.6138],[19,31.7533],[31,31.9296],[41,31.9693],[48,31.8693],[54,31.7027],[64,31.6138],[76,31.7471],[82,31.886],[90,30.9842],[101,30.6836],[109,31.4857],[115,30.8813],[119,30.7301],[145,30.68],[152,31.1007],[159,31.0484],[166,30.8916],[178,30.9621],[193,31.2147],[205,31.2408],[212,31.78],[216,32.6786],[220,33.2177],[225,33.4199],[232,33.5412],[244,33.635],[249,33.7327],[263,33.7913],[271,33.383],[277,32.1582],[283,31.8179],[292,31.6648],[351,31.6138]]
       }
       thigh_r_fk_ctrl.Rotation.Y {
        type=MovieSceneScriptingFloatChannel
        default=-1.213
        thinned_from=58
        keys [[0,-0.7446],[9,-0.7375],[19,-0.7339],[25,-0.7293],[31,-0.7266],[41,-0.725],[48,-0.7305],[54,-0.7397],[64,-0.7446],[76,-0.7372],[82,-0.7296],[90,-1.0764],[101,-1.1921],[109,-0.9367],[115,-1.2014],[119,-1.2675],[126,-1.2896],[145,-1.2868],[152,-1.1777],[159,-1.1907],[166,-1.2296],[172,-1.2178],[178,-1.2107],[183,-1.1815],[188,-1.164],[193,-1.1429],[198,-1.1377],[205,-1.1359],[212,-1.0896],[216,-1.0125],[220,-0.9662],[225,-0.9488],[239,-0.9295],[244,-0.9148],[249,-0.8902],[255,-0.881],[263,-0.8754],[271,-0.8509],[277,-0.7773],[283,-0.7568],[292,-0.7476],[301,-0.7446],[351,-0.7446]]
       }
       thigh_r_fk_ctrl.Rotation.Z {
        type=MovieSceneScriptingFloatChannel
        default=-0.1592
        thinned_from=58
        keys [[0,4.5926],[13,4.6097],[19,4.4438],[25,4.1675],[31,4.0016],[41,3.9022],[48,4.0964],[54,4.42],[59,4.5279],[64,4.5926],[76,4.3337],[82,4.064],[90,1.954],[101,1.2507],[109,3.1858],[115,2.2905],[119,2.0666],[126,1.992],[145,1.9906],[152,2.2339],[159,2.1899],[166,2.0579],[172,2.072],[178,2.0804],[183,2.1152],[193,2.1611],[198,2.1674],[205,2.1695],[212,2.1032],[216,1.9928],[220,1.9266],[225,1.9017],[232,1.8868],[239,2.1587],[244,2.6117],[249,3.3668],[255,3.65],[263,3.8199],[271,3.9648],[277,4.3995],[283,4.5202],[292,4.5745],[301,4.5926],[351,4.5926]]
       }
       thigh_r_fk_ctrl.Scale.X {type=MovieSceneScriptingFloatChannel, default=1, const [1,58]}
       thigh_r_fk_ctrl.Scale.Y {type=MovieSceneScriptingFloatChannel, default=1, const [1,58]}
       thigh_r_fk_ctrl.Scale.Z {type=MovieSceneScriptingFloatChannel, default=1, const [1,58]}
       calf_r_fk_ctrl.Location.X {
        type=MovieSceneScriptingFloatChannel
        default=0
        thinned_from=58
        keys [[0,-0.5804],[3,-0.5748],[6,-0.5654],[9,-0.5592],[13,-0.5564],[19,-0.5817],[25,-0.6239],[31,-0.6492],[41,-0.6644],[48,-0.6407],[54,-0.6014],[59,-0.5882],[64,-0.5804],[76,-0.6119],[82,-0.6447],[90,-0.5862],[101,-0.5668],[109,-0.6361],[115,-0.6007],[119,-0.5919],[126,-0.589],[133,-0.5848],[139,-0.5779],[145,-0.5737],[152,-0.5168],[159,-0.5233],[166,-0.5427],[172,-0.5367],[178,-0.5332],[183,-0.5185],[188,-0.5097],[193,-0.4991],[198,-0.4964],[205,-0.4955],[212,-0.4987],[216,-0.5041],[220,-0.5073],[232,-0.5092],[239,-0.5068],[244,-0.5028],[249,-0.4961],[255,-0.4936],[263,-0.4921],[271,-0.5087],[277,-0.5583],[283,-0.5721],[292,-0.5783],[301,-0.5804],[351,-0.5804]]
       }
       calf_r_fk_ctrl.Location.Y {
        type=MovieSceneScriptingFloatChannel
        default=0
        thinned_from=58
        keys [[0,0.0444],[3,0.044],[6,0.0433],[9,0.0428],[13,0.0426],[19,0.0445],[25,0.0477],[31,0.0497],[41,0.0508],[48,0.049],[54,0.046],[59,0.045],[64,0.0444],[76,0.0468],[82,0.0493],[90,0.0448],[101,0.0434],[109,0.0487],[115,0.046],[119,0.0453],[133,0.0447],[139,0.0442],[145,0.0439],[152,0.0395],[159,0.04],[166,0.0415],[178,0.0408],[183,0.0397],[193,0.0382],[205,0.0379],[212,0.0382],[220,0.0388],[232,0.039],[239,0.0388],[244,0.0385],[249,0.038],[263,0.0376],[271,0.0389],[277,0.0427],[283,0.0438],[292,0.0442],[351,0.0444]]
       }
       calf_r_fk_ctrl.Location.Z {
        type=MovieSceneScriptingFloatChannel
        default=0
        thinned_from=58
        keys [[0,-0.0122],[13,-0.0117],[19,-0.0122],[25,-0.0131],[31,-0.0136],[41,-0.0139],[48,-0.0134],[54,-0.0126],[64,-0.0122],[76,-0.0128],[82,-0.0135],[90,-0.0123],[101,-0.0119],[109,-0.0133],[115,-0.0126],[145,-0.012],[152,-0.0108],[159,-0.011],[166,-0.0114],[178,-0.0112],[193,-0.0105],[244,-0.0106],[263,-0.0103],[271,-0.0107],[277,-0.0117],[283,-0.012],[351,-0.0122]]
       }
       calf_r_fk_ctrl.Rotation.X {type=MovieSceneScriptingFloatChannel, default=0.0561, thinned_from=58, keys [[0,-0.0409],[351,-0.0409]]}
       calf_r_fk_ctrl.Rotation.Y {type=MovieSceneScriptingFloatChannel, default=1.1982, thinned_from=58, keys [[0,1.1988],[351,1.1988]]}
       calf_r_fk_ctrl.Rotation.Z {type=MovieSceneScriptingFloatChannel, default=0.2616, thinned_from=58, keys [[0,-4.3747],[351,-4.3747]]}
       calf_r_fk_ctrl.Scale.X {type=MovieSceneScriptingFloatChannel, default=1, const [1,58]}
       calf_r_fk_ctrl.Scale.Y {type=MovieSceneScriptingFloatChannel, default=1, const [1,58]}
       calf_r_fk_ctrl.Scale.Z {type=MovieSceneScriptingFloatChannel, default=1, const [1,58]}
       foot_r_fk_ctrl.Location.X {
        type=MovieSceneScriptingFloatChannel
        default=0
        thinned_from=58
        keys [[0,-0.6344],[3,-0.6283],[6,-0.6181],[9,-0.6113],[13,-0.6083],[19,-0.6359],[25,-0.682],[31,-0.7097],[41,-0.7263],[48,-0.7005],[54,-0.6574],[59,-0.6431],[64,-0.6344],[76,-0.6689],[82,-0.7048],[90,-0.6409],[101,-0.6196],[109,-0.6954],[115,-0.6567],[119,-0.6471],[126,-0.6438],[133,-0.6393],[139,-0.6317],[145,-0.6272],[152,-0.565],[159,-0.5721],[166,-0.5933],[172,-0.5868],[178,-0.5829],[183,-0.5668],[188,-0.5571],[193,-0.5456],[198,-0.5427],[205,-0.5417],[212,-0.5452],[216,-0.551],[220,-0.5545],[232,-0.5566],[239,-0.554],[244,-0.5496],[249,-0.5423],[255,-0.5396],[263,-0.538],[271,-0.5561],[277,-0.6103],[283,-0.6254],[292,-0.6322],[301,-0.6344],[351,-0.6344]]
       }
       foot_r_fk_ctrl.Location.Y {type=MovieSceneScriptingFloatChannel, default=-0.0329, thinned_from=58, keys [[0,-0.0329],[351,-0.0329]]}
       foot_r_fk_ctrl.Location.Z {type=MovieSceneScriptingFloatChannel, default=-0.9075, thinned_from=58, keys [[0,-0.9075],[351,-0.9075]]}
       foot_r_fk_ctrl.Rotation.X {
        type=MovieSceneScriptingFloatChannel
        default=-1.1648
        thinned_from=58
        keys [[0,-11.2449],[82,-11.2408],[90,-11.1597],[101,-11.1326],[109,-11.1944],[115,-11.1378],[145,-11.1195],[205,-11.0948],[225,-11.1498],[271,-11.1873],[277,-11.2272],[351,-11.2449]]
       }
       foot_r_fk_ctrl.Rotation.Y {
        type=MovieSceneScriptingFloatChannel
        default=0.1067
        thinned_from=58
        keys [[0,0.941],[9,0.9347],[13,0.9339],[19,0.9401],[25,0.9504],[31,0.9565],[41,0.9602],[48,0.9548],[54,0.9458],[64,0.941],[76,0.9482],[82,0.9557],[90,1.615],[101,1.8348],[109,1.2887],[115,1.6994],[119,1.802],[126,1.8363],[145,1.8347],[152,1.8748],[159,1.8707],[166,1.8586],[178,1.8666],[183,1.8791],[193,1.8955],[205,1.8985],[212,1.7997],[216,1.6351],[220,1.5363],[225,1.4993],[232,1.4771],[239,1.4422],[244,1.3841],[249,1.2873],[255,1.251],[263,1.2292],[271,1.1751],[277,1.013],[283,0.968],[292,0.9478],[301,0.941],[351,0.941]]
       }
       foot_r_fk_ctrl.Rotation.Z {
        type=MovieSceneScriptingFloatChannel
        default=-0.1036
        thinned_from=58
        keys [[0,10.1103],[48,10.0954],[82,10.0945],[90,10.3356],[101,10.416],[109,10.1603],[115,10.2486],[119,10.2707],[145,10.2807],[152,10.155],[159,10.1724],[166,10.2245],[178,10.2056],[193,10.1381],[205,10.1311],[212,10.1095],[220,10.0518],[232,10.0389],[239,10.0152],[244,9.9758],[249,9.91],[263,9.8706],[271,9.9155],[277,10.0503],[283,10.0878],[351,10.1103]]
       }
       foot_r_fk_ctrl.Scale.X {type=MovieSceneScriptingFloatChannel, default=1, const [1,58]}
       foot_r_fk_ctrl.Scale.Y {type=MovieSceneScriptingFloatChannel, default=1, const [1,58]}
       foot_r_fk_ctrl.Scale.Z {type=MovieSceneScriptingFloatChannel, default=1, const [1,58]}
       ball_r_fk_ctrl.Location.X {type=MovieSceneScriptingFloatChannel, default=0, const [0,58]}
       ball_r_fk_ctrl.Location.Y {type=MovieSceneScriptingFloatChannel, default=0, const [0,58]}
       ball_r_fk_ctrl.Location.Z {type=MovieSceneScriptingFloatChannel, default=0, const [0,58]}
       ball_r_fk_ctrl.Rotation.X {type=MovieSceneScriptingFloatChannel, default=0, const [0,58]}
       ball_r_fk_ctrl.Rotation.Y {type=MovieSceneScriptingFloatChannel, default=0, const [0,58]}
       ball_r_fk_ctrl.Rotation.Z {type=MovieSceneScriptingFloatChannel, default=0, const [0,58]}
       ball_r_fk_ctrl.Scale.X {type=MovieSceneScriptingFloatChannel, default=1, const [1,58]}
       ball_r_fk_ctrl.Scale.Y {type=MovieSceneScriptingFloatChannel, default=1, const [1,58]}
       ball_r_fk_ctrl.Scale.Z {type=MovieSceneScriptingFloatChannel, default=1, const [1,58]}
       foot_r_ik_ctrl.Location.X {type=MovieSceneScriptingFloatChannel, default=0, const [-1.859,59]}
       foot_r_ik_ctrl.Location.Y {type=MovieSceneScriptingFloatChannel, default=0, const [-0.7384,60]}
       foot_r_ik_ctrl.Location.Z {type=MovieSceneScriptingFloatChannel, default=0, const [0,58]}
       foot_r_ik_ctrl.Rotation.X {type=MovieSceneScriptingFloatChannel, default=0, const [0,58]}
       foot_r_ik_ctrl.Rotation.Y {type=MovieSceneScriptingFloatChannel, default=0, const [0,58]}
       foot_r_ik_ctrl.Rotation.Z {type=MovieSceneScriptingFloatChannel, default=0, const [-30,58]}
       foot_r_ik_ctrl.Scale.X {type=MovieSceneScriptingFloatChannel, default=1, const [1,58]}
       foot_r_ik_ctrl.Scale.Y {type=MovieSceneScriptingFloatChannel, default=1, const [1,58]}
       foot_r_ik_ctrl.Scale.Z {type=MovieSceneScriptingFloatChannel, default=1, const [1,58]}
       ball_r_ik_ctrl.Location.X {type=MovieSceneScriptingFloatChannel, default=0, const [0,58]}
       ball_r_ik_ctrl.Location.Y {type=MovieSceneScriptingFloatChannel, default=0, const [0,58]}
       ball_r_ik_ctrl.Location.Z {type=MovieSceneScriptingFloatChannel, default=0, const [0,58]}
       ball_r_ik_ctrl.Rotation.X {type=MovieSceneScriptingFloatChannel, default=0, const [0,58]}
       ball_r_ik_ctrl.Rotation.Y {type=MovieSceneScriptingFloatChannel, default=0, const [0,58]}
       ball_r_ik_ctrl.Rotation.Z {type=MovieSceneScriptingFloatChannel, default=0, const [0,58]}
       ball_r_ik_ctrl.Scale.X {type=MovieSceneScriptingFloatChannel, default=1, const [1,58]}
       ball_r_ik_ctrl.Scale.Y {type=MovieSceneScriptingFloatChannel, default=1, const [1,58]}
       ball_r_ik_ctrl.Scale.Z {type=MovieSceneScriptingFloatChannel, default=1, const [1,58]}
       foot_bk1_r_ctrl.Location.X {type=MovieSceneScriptingFloatChannel, default=0, const [0,58]}
       foot_bk1_r_ctrl.Location.Y {type=MovieSceneScriptingFloatChannel, default=0, const [0,58]}
       foot_bk1_r_ctrl.Location.Z {type=MovieSceneScriptingFloatChannel, default=0, const [0,58]}
       foot_bk1_r_ctrl.Rotation.X {type=MovieSceneScriptingFloatChannel, default=0, const [0,58]}
       foot_bk1_r_ctrl.Rotation.Y {type=MovieSceneScriptingFloatChannel, default=0, const [0,58]}
       foot_bk1_r_ctrl.Rotation.Z {type=MovieSceneScriptingFloatChannel, default=0, const [0,58]}
       foot_bk1_r_ctrl.Scale.X {type=MovieSceneScriptingFloatChannel, default=1, const [1,58]}
       foot_bk1_r_ctrl.Scale.Y {type=MovieSceneScriptingFloatChannel, default=1, const [1,58]}
       foot_bk1_r_ctrl.Scale.Z {type=MovieSceneScriptingFloatChannel, default=1, const [1,58]}
       foot_roll_r_ctrl.Location.X {type=MovieSceneScriptingFloatChannel, default=0, const [0,58]}
       foot_roll_r_ctrl.Location.Y {type=MovieSceneScriptingFloatChannel, default=0, const [0,58]}
       foot_roll_r_ctrl.Location.Z {type=MovieSceneScriptingFloatChannel, default=0, const [0,58]}
       foot_roll_r_ctrl.Rotation.X {type=MovieSceneScriptingFloatChannel, default=0, const [0,58]}
       foot_roll_r_ctrl.Rotation.Y {type=MovieSceneScriptingFloatChannel, default=0, const [0,58]}
       foot_roll_r_ctrl.Rotation.Z {type=MovieSceneScriptingFloatChannel, default=0, const [0,58]}
       foot_roll_r_ctrl.Scale.X {type=MovieSceneScriptingFloatChannel, default=1, const [1,58]}
       foot_roll_r_ctrl.Scale.Y {type=MovieSceneScriptingFloatChannel, default=1, const [1,58]}
       foot_roll_r_ctrl.Scale.Z {type=MovieSceneScriptingFloatChannel, default=1, const [1,58]}
       roll_blend_r_ctrl {type=MovieSceneScriptingFloatChannel, default=1, const [1,58]}
       tip_r_ctrl.Location.X {type=MovieSceneScriptingFloatChannel, default=0, const [0,58]}
       tip_r_ctrl.Location.Y {type=MovieSceneScriptingFloatChannel, default=0, const [0,58]}
       tip_r_ctrl.Location.Z {type=MovieSceneScriptingFloatChannel, default=0, const [0,58]}
       tip_r_ctrl.Rotation.X {type=MovieSceneScriptingFloatChannel, default=0, const [0,58]}
       tip_r_ctrl.Rotation.Y {type=MovieSceneScriptingFloatChannel, default=0, const [0,58]}
       tip_r_ctrl.Rotation.Z {type=MovieSceneScriptingFloatChannel, default=0, const [0,58]}
       tip_r_ctrl.Scale.X {type=MovieSceneScriptingFloatChannel, default=1, const [1,58]}
       tip_r_ctrl.Scale.Y {type=MovieSceneScriptingFloatChannel, default=1, const [1,58]}
       tip_r_ctrl.Scale.Z {type=MovieSceneScriptingFloatChannel, default=1, const [1,58]}
       heel_r_ctrl.Location.X {type=MovieSceneScriptingFloatChannel, default=0, thinned_from=58, keys [[0,0],[351,0]]}
       heel_r_ctrl.Location.Y {type=MovieSceneScriptingFloatChannel, default=0, thinned_from=58, keys [[0,0],[351,0]]}
       heel_r_ctrl.Location.Z {type=MovieSceneScriptingFloatChannel, default=0, thinned_from=58, keys [[0,0],[351,0]]}
       heel_r_ctrl.Rotation.X {type=MovieSceneScriptingFloatChannel, default=0, const [0,58]}
       heel_r_ctrl.Rotation.Y {type=MovieSceneScriptingFloatChannel, default=0, const [0,58]}
       heel_r_ctrl.Rotation.Z {type=MovieSceneScriptingFloatChannel, default=0, const [0,58]}
       heel_r_ctrl.Scale.X {type=MovieSceneScriptingFloatChannel, default=1, const [1,58]}
       heel_r_ctrl.Scale.Y {type=MovieSceneScriptingFloatChannel, default=1, const [1,58]}
       heel_r_ctrl.Scale.Z {type=MovieSceneScriptingFloatChannel, default=1, const [1,58]}
       leg_r_softness {type=MovieSceneScriptingFloatChannel, default=0, const [0,58]}
       leg_r_pv_ik_ctrl.Location.X {type=MovieSceneScriptingFloatChannel, default=0, const [7,58]}
       leg_r_pv_ik_ctrl.Location.Y {type=MovieSceneScriptingFloatChannel, default=0, const [0,58]}
       leg_r_pv_ik_ctrl.Location.Z {type=MovieSceneScriptingFloatChannel, default=0, const [0,58]}
       leg_r_pv_ik_ctrl.Rotation.X {type=MovieSceneScriptingFloatChannel, default=0, const [0,58]}
       leg_r_pv_ik_ctrl.Rotation.Y {type=MovieSceneScriptingFloatChannel, default=0, const [0,58]}
       leg_r_pv_ik_ctrl.Rotation.Z {type=MovieSceneScriptingFloatChannel, default=0, const [0,58]}
       leg_r_pv_ik_ctrl.Scale.X {type=MovieSceneScriptingFloatChannel, default=1, const [1,58]}
       leg_r_pv_ik_ctrl.Scale.Y {type=MovieSceneScriptingFloatChannel, default=1, const [1,58]}
       leg_r_pv_ik_ctrl.Scale.Z {type=MovieSceneScriptingFloatChannel, default=1, const [1,58]}
       upperarm_l_fk_ctrl.Location.X {type=MovieSceneScriptingFloatChannel, default=0, const [0,55]}
       upperarm_l_fk_ctrl.Location.Y {type=MovieSceneScriptingFloatChannel, default=0, const [0,55]}
       upperarm_l_fk_ctrl.Location.Z {type=MovieSceneScriptingFloatChannel, default=0, const [0,55]}
       upperarm_l_fk_ctrl.Rotation.X {
        type=MovieSceneScriptingFloatChannel
        default=0
        thinned_from=55
        keys [[0,0],[263,0],[283,-0.1537],[291,-0.2193],[298,-0.2401],[313,-0.2401],[319,-0.2316],[335,-0.0539],[351,0]]
       }
       upperarm_l_fk_ctrl.Rotation.Y {
        type=MovieSceneScriptingFloatChannel
        default=0
        thinned_from=37
        keys [[0,-10.1555],[19,-10.1555],[27,-10.2723],[64,-11.1667],[70,-11.202],[109,-9.125],[115,-8.9227],[121,-9.0793],[147,-11.3737],[166,-12.2543],[174,-12.3261],[176,-12.301],[178,-12.2321],[180,-12.1135],[182,-11.9431],[210,-10.1555],[351,-10.1555]]
       }
       upperarm_l_fk_ctrl.Rotation.Z {
        type=MovieSceneScriptingFloatChannel
        default=0
        keys [[0,-2.9004],[3,-2.7831],[6,-2.6308],[9,-2.4997],[13,-2.3859],[19,-2.5319],[31,-3.3216],[41,-3.8501],[48,-4.1151],[54,-4.3152],[59,-4.5075],[64,-4.7699],[70,-5.3098],[76,-5.9522],[81,-7.1324],[85,-7.3834],[92,-6.8413],[116,-3.2788],[122,-2.7186],[127,-3.3972],[147,-6.7588],[151,-7.3204],[159,-7.9088],[166,-8.0473],[171,-7.8571],[179,-5.2467],[182,-3.4809],[185,-2.5276],[188,-2.116],[194,-3.1451],[203,-5.4135],[210,-5.8739],[216,-5.4875],[236,-3.0575],[245,-1.9717],[252,-1.7071],[260,-2.8993],[271,-7.2347],[277,-10.2878],[282,-12.6867],[288,-14.0806],[295,-14.3682],[305,-14.4576],[313,-13.748],[319,-12.1929],[329,-6.5013],[343,-2.9922],[351,-2.9004]]
       }
       upperarm_l_fk_ctrl.Scale.X {type=MovieSceneScriptingFloatChannel, default=1, const [1,55]}
       upperarm_l_fk_ctrl.Scale.Y {type=MovieSceneScriptingFloatChannel, default=1, const [1,55]}
       upperarm_l_fk_ctrl.Scale.Z {type=MovieSceneScriptingFloatChannel, default=1, const [1,55]}
       lowerarm_l_fk_ctrl.Location.X {type=MovieSceneScriptingFloatChannel, default=0, const [0,55]}
       lowerarm_l_fk_ctrl.Location.Y {type=MovieSceneScriptingFloatChannel, default=0, const [0,55]}
       lowerarm_l_fk_ctrl.Location.Z {type=MovieSceneScriptingFloatChannel, default=0, const [0,55]}
       lowerarm_l_fk_ctrl.Rotation.X {type=MovieSceneScriptingFloatChannel, default=0, const [0,55]}
       lowerarm_l_fk_ctrl.Rotation.Y {type=MovieSceneScriptingFloatChannel, default=0, thinned_from=55, keys [[0,0],[351,0]]}
       lowerarm_l_fk_ctrl.Rotation.Z {
        type=MovieSceneScriptingFloatChannel
        default=0
        keys [[0,-8.0911],[3,-8.4562],[6,-8.8035],[9,-9.0838],[13,-9.314],[19,-9.4396],[22,-9.4156],[28,-9.0633],[41,-7.4351],[45,-7.0009],[51,-7.6495],[70,-11.5705],[76,-12.2937],[82,-12.5193],[90,-11.8297],[101,-10.3992],[109,-9.3974],[115,-8.7132],[119,-8.3942],[126,-9.0165],[133,-9.6847],[145,-10.7023],[157,-11.4847],[166,-11.7487],[172,-11.4925],[178,-10.6548],[183,-9.4782],[187,-8.3523],[190,-7.9288],[195,-8.7273],[204,-11.0656],[217,-12.2875],[233,-12.8361],[249,-12.9786],[268,-12.8281],[279,-12.4766],[285,-12.0302],[296,-10.3704],[305,-8.677],[317,-7.7205],[333,-7.3601],[351,-8.0911]]
       }
       lowerarm_l_fk_ctrl.Scale.X {type=MovieSceneScriptingFloatChannel, default=1, const [1,55]}
       lowerarm_l_fk_ctrl.Scale.Y {type=MovieSceneScriptingFloatChannel, default=1, const [1,55]}
       lowerarm_l_fk_ctrl.Scale.Z {type=MovieSceneScriptingFloatChannel, default=1, const [1,55]}
       hand_l_fk_ctrl.Location.X {type=MovieSceneScriptingFloatChannel, default=0, const [0,54]}
       hand_l_fk_ctrl.Location.Y {type=MovieSceneScriptingFloatChannel, default=0, const [0,54]}
       hand_l_fk_ctrl.Location.Z {type=MovieSceneScriptingFloatChannel, default=0, const [0,54]}
       hand_l_fk_ctrl.Rotation.X {type=MovieSceneScriptingFloatChannel, default=0, const [-18.5,54]}
       hand_l_fk_ctrl.Rotation.Y {type=MovieSceneScriptingFloatChannel, default=0, const [5.5,54]}
       hand_l_fk_ctrl.Rotation.Z {
        type=MovieSceneScriptingFloatChannel
        default=0
        thinned_from=54
        keys [[0,-7.1916],[255,-7.1916],[263,-7.6644],[283,-11.3728],[291,-12.3664],[313,-12.3664],[319,-11.7195],[335,-8.8087],[351,-7.1916]]
       }
       hand_l_fk_ctrl.Scale.X {type=MovieSceneScriptingFloatChannel, default=1, const [1,54]}
       hand_l_fk_ctrl.Scale.Y {type=MovieSceneScriptingFloatChannel, default=1, const [1,54]}
       hand_l_fk_ctrl.Scale.Z {type=MovieSceneScriptingFloatChannel, default=1, const [1,54]}
       thumb_01_l_ctrl.Location.X {type=MovieSceneScriptingFloatChannel, default=0, const [0,56]}
       thumb_01_l_ctrl.Location.Y {type=MovieSceneScriptingFloatChannel, default=0, const [0,56]}
       thumb_01_l_ctrl.Location.Z {type=MovieSceneScriptingFloatChannel, default=0, const [0,56]}
       thumb_01_l_ctrl.Rotation.X {type=MovieSceneScriptingFloatChannel, default=0, const [34,56]}
       thumb_01_l_ctrl.Rotation.Y {type=MovieSceneScriptingFloatChannel, default=0, const [-5.5,56]}
       thumb_01_l_ctrl.Rotation.Z {
        type=MovieSceneScriptingFloatChannel
        default=0
        thinned_from=56
        keys [[0,-10.9071],[3,-11.4905],[6,-11.7542],[9,-11.8426],[19,-11.8426],[25,-11.7688],[31,-11.3542],[41,-10.1102],[48,-9.9071],[54,-10.2277],[59,-10.8134],[64,-11.5448],[70,-11.8569],[76,-11.9609],[82,-11.291],[90,-10.0252],[101,-9.2657],[109,-9.7424],[115,-10.9031],[119,-11.5401],[126,-11.779],[133,-11.9223],[139,-11.4713],[145,-10.4665],[152,-9.9942],[166,-9.4182],[172,-9.0835],[183,-8.6422],[188,-8.5484],[193,-8.5484],[198,-8.9014],[205,-9.5631],[212,-9.8113],[220,-10.0006],[225,-10.0321],[232,-10.0321],[239,-9.8962],[247,-9.5261],[252,-9.1626],[257,-9.0263],[263,-8.9446],[283,-8.9446],[291,-8.0417],[298,-7.7132],[305,-7.5677],[311,-7.5],[318,-8.1238],[328,-9.3829],[336,-8.258],[339,-7.7692],[351,-10.9071]]
       }
       thumb_01_l_ctrl.Scale.X {type=MovieSceneScriptingFloatChannel, default=1, const [1,56]}
       thumb_01_l_ctrl.Scale.Y {type=MovieSceneScriptingFloatChannel, default=1, const [1,56]}
       thumb_01_l_ctrl.Scale.Z {type=MovieSceneScriptingFloatChannel, default=1, const [1,56]}
       thumb_02_l_ctrl.Location.X {type=MovieSceneScriptingFloatChannel, default=0, const [0,56]}
       thumb_02_l_ctrl.Location.Y {type=MovieSceneScriptingFloatChannel, default=0, const [0,56]}
       thumb_02_l_ctrl.Location.Z {type=MovieSceneScriptingFloatChannel, default=0, const [0,56]}
       thumb_02_l_ctrl.Rotation.X {type=MovieSceneScriptingFloatChannel, default=0, const [0,56]}
       thumb_02_l_ctrl.Rotation.Y {type=MovieSceneScriptingFloatChannel, default=0, const [0,56]}
       thumb_02_l_ctrl.Rotation.Z {
        type=MovieSceneScriptingFloatChannel
        default=0
        thinned_from=56
        keys [[0,36.8698],[3,36.2864],[6,36.2996],[9,36.0445],[25,36.0081],[31,36.4227],[41,37.6666],[48,37.8698],[54,37.5491],[59,36.9635],[64,36.2321],[70,35.92],[76,35.816],[82,36.4859],[90,37.7517],[101,38.5112],[109,38.0344],[115,36.8737],[119,36.2367],[126,35.9979],[133,35.8545],[139,36.3056],[145,37.3104],[152,37.7826],[166,38.3587],[172,38.6934],[183,39.1347],[188,39.2284],[193,39.2284],[198,38.8755],[205,38.2138],[220,37.7763],[232,37.7448],[239,40.5267],[247,48.0996],[252,55.5374],[257,58.3265],[263,60],[283,60],[291,35.9375],[298,27.1832],[305,23.3047],[311,21.5],[318,20.8762],[328,29.9938],[336,37.4188],[339,38.324],[351,36.8698]]
       }
       thumb_02_l_ctrl.Scale.X {type=MovieSceneScriptingFloatChannel, default=1, const [1,56]}
       thumb_02_l_ctrl.Scale.Y {type=MovieSceneScriptingFloatChannel, default=1, const [1,56]}
       thumb_02_l_ctrl.Scale.Z {type=MovieSceneScriptingFloatChannel, default=1, const [1,56]}
       thumb_03_l_ctrl.Location.X {type=MovieSceneScriptingFloatChannel, default=0, const [0,56]}
       thumb_03_l_ctrl.Location.Y {type=MovieSceneScriptingFloatChannel, default=0, const [0,56]}
       thumb_03_l_ctrl.Location.Z {type=MovieSceneScriptingFloatChannel, default=0, const [0,56]}
       thumb_03_l_ctrl.Rotation.X {type=MovieSceneScriptingFloatChannel, default=0, const [0,56]}
       thumb_03_l_ctrl.Rotation.Y {type=MovieSceneScriptingFloatChannel, default=0, const [0,56]}
       thumb_03_l_ctrl.Rotation.Z {
        type=MovieSceneScriptingFloatChannel
        default=0
        thinned_from=56
        keys [[0,-35.5],[3,-36.0834],[6,-36.3471],[9,-36.4355],[19,-36.4355],[25,-36.3617],[31,-35.947],[41,-34.7031],[48,-34.5],[54,-34.8206],[59,-35.4062],[64,-36.1377],[70,-36.4497],[76,-36.5538],[82,-35.8839],[90,-34.6181],[101,-33.8586],[109,-34.3353],[115,-35.496],[119,-36.133],[126,-36.3719],[133,-36.5152],[139,-36.0642],[145,-35.0594],[152,-34.5871],[166,-34.0111],[172,-33.6764],[183,-33.2351],[188,-33.1413],[193,-33.1413],[198,-33.4943],[205,-34.156],[220,-34.5935],[232,-34.625],[239,-34.422],[247,-33.8694],[252,-33.3266],[257,-33.123],[263,-33.0009],[283,-33.0009],[291,-26.1253],[298,-23.6239],[305,-22.5157],[311,-22],[318,-22.6238],[328,-29.4605],[336,-31.7221],[339,-31.9699],[351,-35.5]]
       }
       thumb_03_l_ctrl.Scale.X {type=MovieSceneScriptingFloatChannel, default=1, const [1,56]}
       thumb_03_l_ctrl.Scale.Y {type=MovieSceneScriptingFloatChannel, default=1, const [1,56]}
       thumb_03_l_ctrl.Scale.Z {type=MovieSceneScriptingFloatChannel, default=1, const [1,56]}
       index_metacarpal_l_ctrl.Location.X {type=MovieSceneScriptingFloatChannel, default=0, const [0,56]}
       index_metacarpal_l_ctrl.Location.Y {type=MovieSceneScriptingFloatChannel, default=0, const [0,56]}
       index_metacarpal_l_ctrl.Location.Z {type=MovieSceneScriptingFloatChannel, default=0, const [0,56]}
       index_metacarpal_l_ctrl.Rotation.X {type=MovieSceneScriptingFloatChannel, default=0, const [0,56]}
       index_metacarpal_l_ctrl.Rotation.Y {
        type=MovieSceneScriptingFloatChannel
        default=0
        thinned_from=56
        keys [[0,1],[283,1],[291,2.25],[298,2.7048],[305,2.9062],[313,3],[319,2.9298],[328,1.8947],[339,1.2237],[341,1.1683],[351,1]]
       }
       index_metacarpal_l_ctrl.Rotation.Z {
        type=MovieSceneScriptingFloatChannel
        default=0
        thinned_from=56
        keys [[0,3.5],[3,2.9166],[6,2.6529],[9,2.5645],[19,2.5645],[25,2.6383],[31,3.053],[41,4.2969],[48,4.5],[54,4.1794],[59,3.5938],[64,2.8623],[70,2.5503],[76,2.4462],[82,3.1161],[90,4.3819],[101,5.1414],[109,4.6647],[115,3.504],[119,2.867],[126,2.6281],[133,2.4848],[139,2.9358],[145,3.9406],[152,4.4129],[166,4.9889],[172,5.3236],[183,5.7649],[188,5.8587],[193,5.8587],[198,5.5057],[205,4.844],[212,4.5958],[220,4.4065],[225,4.375],[232,4.375],[239,4.5049],[247,4.8585],[252,5.2059],[257,5.3361],[263,5.4143],[283,5.4143],[291,1.9723],[298,0.72],[305,0.1653],[311,-0.0929],[318,-0.7167],[328,1.8927],[336,5.3662],[339,6.0103],[351,3.5]]
       }
       index_metacarpal_l_ctrl.Scale.X {type=MovieSceneScriptingFloatChannel, default=1, const [1,56]}
       index_metacarpal_l_ctrl.Scale.Y {type=MovieSceneScriptingFloatChannel, default=1, const [1,56]}
       index_metacarpal_l_ctrl.Scale.Z {type=MovieSceneScriptingFloatChannel, default=1, const [1,56]}
       index_01_l_ctrl.Location.X {type=MovieSceneScriptingFloatChannel, default=0, const [0,56]}
       index_01_l_ctrl.Location.Y {type=MovieSceneScriptingFloatChannel, default=0, const [0,56]}
       index_01_l_ctrl.Location.Z {type=MovieSceneScriptingFloatChannel, default=0, const [0,56]}
       index_01_l_ctrl.Rotation.X {type=MovieSceneScriptingFloatChannel, default=0, const [0,56]}
       index_01_l_ctrl.Rotation.Y {type=MovieSceneScriptingFloatChannel, default=0, const [0,56]}
       index_01_l_ctrl.Rotation.Z {
        type=MovieSceneScriptingFloatChannel
        default=0
        thinned_from=56
        keys [[0,34],[3,33.4166],[6,33.4299],[9,33.1748],[25,33.1383],[31,33.553],[41,34.7969],[48,35],[54,34.6794],[59,34.0938],[64,33.3623],[70,33.0503],[76,32.9462],[82,33.6161],[90,34.8819],[101,35.6414],[109,35.1647],[115,34.004],[119,33.367],[126,33.1281],[133,32.9848],[139,33.4358],[145,34.4406],[152,34.9129],[166,35.4889],[172,35.8236],[183,36.2649],[188,36.3587],[193,36.3587],[198,36.0057],[205,35.344],[220,34.9065],[232,34.875],[239,35.0049],[247,35.3585],[252,35.7059],[263,35.9143],[283,35.9143],[291,22.8428],[298,18.0873],[305,15.9804],[311,15],[318,14.3762],[328,25.5],[336,34.143],[339,35.1287],[351,34]]
       }
       index_01_l_ctrl.Scale.X {type=MovieSceneScriptingFloatChannel, default=1, const [1,56]}
       index_01_l_ctrl.Scale.Y {type=MovieSceneScriptingFloatChannel, default=1, const [1,56]}
       index_01_l_ctrl.Scale.Z {type=MovieSceneScriptingFloatChannel, default=1, const [1,56]}
       index_02_l_ctrl.Location.X {type=MovieSceneScriptingFloatChannel, default=0, const [0,56]}
       index_02_l_ctrl.Location.Y {type=MovieSceneScriptingFloatChannel, default=0, const [0,56]}
       index_02_l_ctrl.Location.Z {type=MovieSceneScriptingFloatChannel, default=0, const [0,56]}
       index_02_l_ctrl.Rotation.X {type=MovieSceneScriptingFloatChannel, default=0, const [0,56]}
       index_02_l_ctrl.Rotation.Y {type=MovieSceneScriptingFloatChannel, default=0, const [0,56]}
       index_02_l_ctrl.Rotation.Z {
        type=MovieSceneScriptingFloatChannel
        default=0
        thinned_from=56
        keys [[0,0],[3,-0.5834],[6,-0.8471],[9,-0.9355],[19,-0.9355],[25,-0.8617],[31,-0.447],[41,0.7969],[48,1],[54,0.6794],[59,0.0938],[64,-0.6377],[70,-0.9497],[76,-1.0538],[82,-0.3839],[90,0.8819],[101,1.6414],[109,1.1647],[115,0.004],[119,-0.633],[126,-0.8719],[133,-1.0152],[139,-0.5642],[145,0.4406],[152,0.9129],[159,1.1962],[166,1.4889],[172,1.8236],[178,2.0591],[183,2.2649],[188,2.3587],[193,2.3587],[198,2.0057],[205,1.344],[212,1.0958],[220,0.9065],[225,0.875],[232,0.875],[239,1.0049],[247,1.3585],[252,1.7059],[257,1.8361],[263,1.9143],[283,1.9143],[291,-3.0902],[298,-4.9109],[305,-5.7176],[311,-6.0929],[318,-6.7167],[328,-2.7258],[336,1.5866],[339,2.2861],[351,0]]
       }
       index_02_l_ctrl.Scale.X {type=MovieSceneScriptingFloatChannel, default=1, const [1,56]}
       index_02_l_ctrl.Scale.Y {type=MovieSceneScriptingFloatChannel, default=1, const [1,56]}
       index_02_l_ctrl.Scale.Z {type=MovieSceneScriptingFloatChannel, default=1, const [1,56]}
       index_03_l_ctrl.Location.X {type=MovieSceneScriptingFloatChannel, default=0, const [0,56]}
       index_03_l_ctrl.Location.Y {type=MovieSceneScriptingFloatChannel, default=0, const [0,56]}
       index_03_l_ctrl.Location.Z {type=MovieSceneScriptingFloatChannel, default=0, const [0,56]}
       index_03_l_ctrl.Rotation.X {type=MovieSceneScriptingFloatChannel, default=0, const [0,56]}
       index_03_l_ctrl.Rotation.Y {type=MovieSceneScriptingFloatChannel, default=0, const [0,56]}
       index_03_l_ctrl.Rotation.Z {
        type=MovieSceneScriptingFloatChannel
        default=0
        thinned_from=56
        keys [[0,0],[3,-0.5834],[6,-0.8471],[9,-0.9355],[19,-0.9355],[25,-0.8617],[31,-0.447],[41,0.7969],[48,1],[54,0.6794],[59,0.0938],[64,-0.6377],[70,-0.9497],[76,-1.0538],[82,-0.3839],[90,0.8819],[101,1.6414],[109,1.1647],[115,0.004],[119,-0.633],[126,-0.8719],[133,-1.0152],[139,-0.5642],[145,0.4406],[152,0.9129],[159,1.1962],[166,1.4889],[172,1.8236],[178,2.0591],[183,2.2649],[188,2.3587],[193,2.3587],[198,2.0057],[205,1.344],[212,1.0958],[220,0.9065],[225,0.875],[232,0.875],[239,1.0049],[247,1.3585],[252,1.7059],[257,1.8361],[263,1.9143],[283,1.9143],[291,-3.0902],[298,-4.9109],[305,-5.7176],[311,-6.0929],[318,-6.7167],[328,-2.7258],[336,1.5866],[339,2.2861],[351,0]]
       }
       index_03_l_ctrl.Scale.X {type=MovieSceneScriptingFloatChannel, default=1, const [1,56]}
       index_03_l_ctrl.Scale.Y {type=MovieSceneScriptingFloatChannel, default=1, const [1,56]}
       index_03_l_ctrl.Scale.Z {type=MovieSceneScriptingFloatChannel, default=1, const [1,56]}
       middle_metacarpal_l_ctrl.Location.X {type=MovieSceneScriptingFloatChannel, default=0, const [0,56]}
       middle_metacarpal_l_ctrl.Location.Y {type=MovieSceneScriptingFloatChannel, default=0, const [0,56]}
       middle_metacarpal_l_ctrl.Location.Z {type=MovieSceneScriptingFloatChannel, default=0, const [0,56]}
       middle_metacarpal_l_ctrl.Rotation.X {type=MovieSceneScriptingFloatChannel, default=0, const [0,56]}
       middle_metacarpal_l_ctrl.Rotation.Y {type=MovieSceneScriptingFloatChannel, default=0, const [1.5,56]}
       middle_metacarpal_l_ctrl.Rotation.Z {
        type=MovieSceneScriptingFloatChannel
        default=0
        thinned_from=56
        keys [[0,0],[3,-0.5834],[6,-0.8471],[9,-0.9355],[19,-0.9355],[25,-0.8617],[31,-0.447],[41,0.7969],[48,1],[54,0.6794],[59,0.0938],[64,-0.6377],[70,-0.9497],[76,-1.0538],[82,-0.3839],[90,0.8819],[101,1.6414],[109,1.1647],[115,0.004],[119,-0.633],[126,-0.8719],[133,-1.0152],[139,-0.5642],[145,0.4406],[152,0.9129],[159,1.1962],[166,1.4889],[172,1.8236],[178,2.0591],[183,2.2649],[188,2.3587],[193,2.3587],[198,2.0057],[205,1.344],[212,1.0958],[220,0.9065],[225,0.875],[232,0.875],[239,1.0049],[247,1.3585],[252,1.7059],[257,1.8361],[263,1.9143],[283,1.9143],[291,-1.5277],[298,-2.78],[305,-3.3347],[311,-3.5929],[318,-4.2167],[328,-1.6073],[336,1.8662],[339,2.5103],[351,0]]
       }
       middle_metacarpal_l_ctrl.Scale.X {type=MovieSceneScriptingFloatChannel, default=1, const [1,56]}
       middle_metacarpal_l_ctrl.Scale.Y {type=MovieSceneScriptingFloatChannel, default=1, const [1,56]}
       middle_metacarpal_l_ctrl.Scale.Z {type=MovieSceneScriptingFloatChannel, default=1, const [1,56]}
       middle_01_l_ctrl.Location.X {type=MovieSceneScriptingFloatChannel, default=0, const [0,56]}
       middle_01_l_ctrl.Location.Y {type=MovieSceneScriptingFloatChannel, default=0, const [0,56]}
       middle_01_l_ctrl.Location.Z {type=MovieSceneScriptingFloatChannel, default=0, const [0,56]}
       middle_01_l_ctrl.Rotation.X {type=MovieSceneScriptingFloatChannel, default=0, const [0,56]}
       middle_01_l_ctrl.Rotation.Y {type=MovieSceneScriptingFloatChannel, default=0, const [0,56]}
       middle_01_l_ctrl.Rotation.Z {
        type=MovieSceneScriptingFloatChannel
        default=0
        thinned_from=56
        keys [[0,0],[3,-0.5834],[6,-0.8471],[9,-0.9355],[19,-0.9355],[25,-0.8617],[31,-0.447],[41,0.7969],[48,1],[54,0.6794],[59,0.0938],[64,-0.6377],[70,-0.9497],[76,-1.0538],[82,-0.3839],[90,0.8819],[101,1.6414],[109,1.1647],[115,0.004],[119,-0.633],[126,-0.8719],[133,-1.0152],[139,-0.5642],[145,0.4406],[152,0.9129],[159,1.1962],[166,1.4889],[172,1.8236],[178,2.0591],[183,2.2649],[188,2.3587],[193,2.3587],[198,2.0057],[205,1.344],[212,1.0958],[220,0.9065],[225,0.875],[232,0.875],[239,1.1077],[247,1.7412],[252,2.3634],[257,2.5967],[263,2.7367],[283,2.7367],[291,-1.2193],[298,-2.6586],[305,-3.2962],[311,-3.5929],[318,-4.2167],[328,-1.6073],[336,1.8662],[339,2.5103],[351,0]]
       }
       middle_01_l_ctrl.Scale.X {type=MovieSceneScriptingFloatChannel, default=1, const [1,56]}
       middle_01_l_ctrl.Scale.Y {type=MovieSceneScriptingFloatChannel, default=1, const [1,56]}
       middle_01_l_ctrl.Scale.Z {type=MovieSceneScriptingFloatChannel, default=1, const [1,56]}
       middle_02_l_ctrl.Location.X {type=MovieSceneScriptingFloatChannel, default=0, const [0,56]}
       middle_02_l_ctrl.Location.Y {type=MovieSceneScriptingFloatChannel, default=0, const [0,56]}
       middle_02_l_ctrl.Location.Z {type=MovieSceneScriptingFloatChannel, default=0, const [0,56]}
       middle_02_l_ctrl.Rotation.X {type=MovieSceneScriptingFloatChannel, default=0, const [0,56]}
       middle_02_l_ctrl.Rotation.Y {type=MovieSceneScriptingFloatChannel, default=0, const [0,56]}
       middle_02_l_ctrl.Rotation.Z {
        type=MovieSceneScriptingFloatChannel
        default=0
        thinned_from=56
        keys [[0,3],[3,2.4166],[6,2.1529],[9,2.0645],[19,2.0645],[25,2.1383],[31,2.553],[41,3.7969],[48,4],[54,3.6794],[59,3.0938],[64,2.3623],[70,2.0503],[76,1.9462],[82,2.6161],[90,3.8819],[101,4.6414],[109,4.1647],[115,3.004],[119,2.367],[126,2.1281],[133,1.9848],[139,2.4358],[145,3.4406],[152,3.9129],[159,4.1962],[166,4.4889],[172,4.8236],[178,5.0591],[183,5.2649],[188,5.3587],[193,5.3587],[198,5.0057],[205,4.344],[212,4.0958],[220,3.9065],[225,3.875],[232,3.875],[239,4.0049],[247,4.3585],[252,4.7059],[257,4.8361],[263,4.9143],[283,4.9143],[291,-11.5947],[298,-17.6009],[305,-20.2618],[311,-21.5],[318,-22.1238],[328,-7.9605],[336,2.5279],[339,3.7388],[351,3]]
       }
       middle_02_l_ctrl.Scale.X {type=MovieSceneScriptingFloatChannel, default=1, const [1,56]}
       middle_02_l_ctrl.Scale.Y {type=MovieSceneScriptingFloatChannel, default=1, const [1,56]}
       middle_02_l_ctrl.Scale.Z {type=MovieSceneScriptingFloatChannel, default=1, const [1,56]}
       middle_03_l_ctrl.Location.X {type=MovieSceneScriptingFloatChannel, default=0, const [0,56]}
       middle_03_l_ctrl.Location.Y {type=MovieSceneScriptingFloatChannel, default=0, const [0,56]}
       middle_03_l_ctrl.Location.Z {type=MovieSceneScriptingFloatChannel, default=0, const [0,56]}
       middle_03_l_ctrl.Rotation.X {type=MovieSceneScriptingFloatChannel, default=0, const [0,56]}
       middle_03_l_ctrl.Rotation.Y {type=MovieSceneScriptingFloatChannel, default=0, const [0,56]}
       middle_03_l_ctrl.Rotation.Z {
        type=MovieSceneScriptingFloatChannel
        default=0
        thinned_from=56
        keys [[0,-6.5],[3,-7.0834],[6,-7.3471],[9,-7.4355],[19,-7.4355],[25,-7.3617],[31,-6.947],[41,-5.7031],[48,-5.5],[54,-5.8206],[59,-6.4062],[64,-7.1377],[70,-7.4497],[76,-7.5538],[82,-6.8839],[90,-5.6181],[101,-4.8586],[109,-5.3353],[115,-6.496],[119,-7.133],[126,-7.3719],[133,-7.5152],[139,-7.0642],[145,-6.0594],[152,-5.5871],[166,-5.0111],[172,-4.6764],[178,-4.4409],[183,-4.2351],[188,-4.1413],[193,-4.1413],[198,-4.4943],[205,-5.156],[212,-5.4042],[220,-5.5935],[225,-5.625],[232,-5.625],[239,-5.4951],[247,-5.1415],[252,-4.7941],[257,-4.6639],[263,-4.5857],[283,-4.5857],[291,-10.5277],[298,-12.6895],[305,-13.6472],[311,-14.0929],[318,-14.7167],[328,-9.8968],[336,-5.0812],[339,-4.3484],[351,-6.5]]
       }
       middle_03_l_ctrl.Scale.X {type=MovieSceneScriptingFloatChannel, default=1, const [1,56]}
       middle_03_l_ctrl.Scale.Y {type=MovieSceneScriptingFloatChannel, default=1, const [1,56]}
       middle_03_l_ctrl.Scale.Z {type=MovieSceneScriptingFloatChannel, default=1, const [1,56]}
       ring_metacarpal_l_ctrl.Location.X {type=MovieSceneScriptingFloatChannel, default=0, const [0,56]}
       ring_metacarpal_l_ctrl.Location.Y {type=MovieSceneScriptingFloatChannel, default=0, const [0,56]}
       ring_metacarpal_l_ctrl.Location.Z {type=MovieSceneScriptingFloatChannel, default=0, const [0,56]}
       ring_metacarpal_l_ctrl.Rotation.X {type=MovieSceneScriptingFloatChannel, default=0, const [0,56]}
       ring_metacarpal_l_ctrl.Rotation.Y {
        type=MovieSceneScriptingFloatChannel
        default=0
        thinned_from=56
        keys [[0,2.5],[283,2.5],[291,0.9375],[298,0.369],[305,0.1172],[313,0],[319,0.0877],[328,1.3816],[339,2.2204],[341,2.2896],[351,2.5]]
       }
       ring_metacarpal_l_ctrl.Rotation.Z {
        type=MovieSceneScriptingFloatChannel
        default=0
        thinned_from=56
        keys [[0,0],[3,-0.5834],[6,-0.8471],[9,-0.9355],[19,-0.9355],[25,-0.8617],[31,-0.447],[41,0.7969],[48,1],[54,0.6794],[59,0.0938],[64,-0.6377],[70,-0.9497],[76,-1.0538],[82,-0.3839],[90,0.8819],[101,1.6414],[109,1.1647],[115,0.004],[119,-0.633],[126,-0.8719],[133,-1.0152],[139,-0.5642],[145,0.4406],[152,0.9129],[159,1.1962],[166,1.4889],[172,1.8236],[178,2.0591],[183,2.2649],[188,2.3587],[193,2.3587],[198,2.0057],[205,1.344],[212,1.0958],[220,0.9065],[225,0.875],[232,0.875],[239,1.0049],[247,1.3585],[252,1.7059],[257,1.8361],[263,1.9143],[283,1.9143],[291,-1.5277],[298,-2.78],[305,-3.3347],[311,-3.5929],[318,-4.2167],[328,-1.6073],[336,1.8662],[339,2.5103],[351,0]]
       }
       ring_metacarpal_l_ctrl.Scale.X {type=MovieSceneScriptingFloatChannel, default=1, const [1,56]}
       ring_metacarpal_l_ctrl.Scale.Y {type=MovieSceneScriptingFloatChannel, default=1, const [1,56]}
       ring_metacarpal_l_ctrl.Scale.Z {type=MovieSceneScriptingFloatChannel, default=1, const [1,56]}
       ring_01_l_ctrl.Location.X {type=MovieSceneScriptingFloatChannel, default=0, const [0,56]}
       ring_01_l_ctrl.Location.Y {type=MovieSceneScriptingFloatChannel, default=0, const [0,56]}
       ring_01_l_ctrl.Location.Z {type=MovieSceneScriptingFloatChannel, default=0, const [0,56]}
       ring_01_l_ctrl.Rotation.X {type=MovieSceneScriptingFloatChannel, default=0, const [-8.5,56]}
       ring_01_l_ctrl.Rotation.Y {type=MovieSceneScriptingFloatChannel, default=0, const [11.5,56]}
       ring_01_l_ctrl.Rotation.Z {
        type=MovieSceneScriptingFloatChannel
        default=0
        thinned_from=56
        keys [[0,-19.5],[3,-20.0834],[6,-20.3471],[9,-20.4355],[19,-20.4355],[25,-20.3617],[31,-19.947],[41,-18.7031],[48,-18.5],[54,-18.8206],[59,-19.4062],[64,-20.1377],[70,-20.4497],[76,-20.5538],[82,-19.8839],[90,-18.6181],[101,-17.8586],[109,-18.3353],[115,-19.496],[119,-20.133],[126,-20.3719],[133,-20.5152],[139,-20.0642],[145,-19.0594],[152,-18.5871],[166,-18.0111],[172,-17.6764],[183,-17.2351],[188,-17.1413],[193,-17.1413],[198,-17.4943],[205,-18.156],[212,-18.4042],[220,-18.5935],[232,-18.625],[239,-18.2637],[247,-17.2801],[252,-16.3141],[257,-15.9519],[263,-15.7345],[283,-15.7345],[291,-20.3335],[298,-22.0067],[305,-22.748],[311,-23.0929],[318,-23.7167],[328,-21.1073],[336,-17.6338],[339,-16.9897],[351,-19.5]]
       }
       ring_01_l_ctrl.Scale.X {type=MovieSceneScriptingFloatChannel, default=1, const [1,56]}
       ring_01_l_ctrl.Scale.Y {type=MovieSceneScriptingFloatChannel, default=1, const [1,56]}
       ring_01_l_ctrl.Scale.Z {type=MovieSceneScriptingFloatChannel, default=1, const [1,56]}
       ring_02_l_ctrl.Location.X {type=MovieSceneScriptingFloatChannel, default=0, const [0,56]}
       ring_02_l_ctrl.Location.Y {type=MovieSceneScriptingFloatChannel, default=0, const [0,56]}
       ring_02_l_ctrl.Location.Z {type=MovieSceneScriptingFloatChannel, default=0, const [0,56]}
       ring_02_l_ctrl.Rotation.X {type=MovieSceneScriptingFloatChannel, default=0, const [0,56]}
       ring_02_l_ctrl.Rotation.Y {type=MovieSceneScriptingFloatChannel, default=0, const [0,56]}
       ring_02_l_ctrl.Rotation.Z {
        type=MovieSceneScriptingFloatChannel
        default=0
        thinned_from=56
        keys [[0,-15.5],[3,-16.0834],[6,-16.3471],[9,-16.4355],[19,-16.4355],[25,-16.3617],[31,-15.947],[41,-14.7031],[48,-14.5],[54,-14.8206],[59,-15.4062],[64,-16.1377],[70,-16.4497],[76,-16.5538],[82,-15.8839],[90,-14.6181],[101,-13.8586],[109,-14.3353],[115,-15.496],[119,-16.133],[126,-16.3719],[133,-16.5152],[139,-16.0642],[145,-15.0594],[152,-14.5871],[166,-14.0111],[172,-13.6764],[183,-13.2351],[188,-13.1413],[193,-13.1413],[198,-13.4943],[205,-14.156],[212,-14.4042],[220,-14.5935],[225,-14.625],[232,-14.625],[239,-14.4951],[247,-14.1415],[252,-13.7941],[257,-13.6639],[263,-13.5857],[283,-13.5857],[291,-20.0947],[298,-22.4627],[305,-23.5118],[311,-24],[318,-24.6238],[328,-19.3026],[336,-14.1826],[339,-13.4298],[351,-15.5]]
       }
       ring_02_l_ctrl.Scale.X {type=MovieSceneScriptingFloatChannel, default=1, const [1,56]}
       ring_02_l_ctrl.Scale.Y {type=MovieSceneScriptingFloatChannel, default=1, const [1,56]}
       ring_02_l_ctrl.Scale.Z {type=MovieSceneScriptingFloatChannel, default=1, const [1,56]}
       ring_03_l_ctrl.Location.X {type=MovieSceneScriptingFloatChannel, default=0, const [0,56]}
       ring_03_l_ctrl.Location.Y {type=MovieSceneScriptingFloatChannel, default=0, const [0,56]}
       ring_03_l_ctrl.Location.Z {type=MovieSceneScriptingFloatChannel, default=0, const [0,56]}
       ring_03_l_ctrl.Rotation.X {type=MovieSceneScriptingFloatChannel, default=0, const [0,56]}
       ring_03_l_ctrl.Rotation.Y {type=MovieSceneScriptingFloatChannel, default=0, const [0,56]}
       ring_03_l_ctrl.Rotation.Z {
        type=MovieSceneScriptingFloatChannel
        default=0
        thinned_from=56
        keys [[0,-29],[3,-29.5834],[6,-29.8471],[9,-29.9355],[19,-29.9355],[25,-29.8617],[31,-29.447],[41,-28.2031],[48,-28],[54,-28.3206],[59,-28.9062],[64,-29.6377],[70,-29.9497],[76,-30.0538],[82,-29.3839],[90,-28.1181],[101,-27.3586],[109,-27.8353],[115,-28.996],[119,-29.633],[126,-29.8719],[133,-30.0152],[139,-29.5642],[145,-28.5594],[152,-28.0871],[166,-27.5111],[172,-27.1764],[183,-26.7351],[188,-26.6413],[193,-26.6413],[198,-26.9943],[205,-27.656],[212,-27.9042],[220,-28.0935],[232,-28.125],[239,-27.9951],[247,-27.6415],[252,-27.2941],[257,-27.1639],[263,-27.0857],[283,-27.0857],[291,-30.5277],[298,-31.78],[305,-32.3347],[311,-32.5929],[318,-33.2167],[328,-30.6073],[336,-27.1338],[339,-26.4897],[351,-29]]
       }
       ring_03_l_ctrl.Scale.X {type=MovieSceneScriptingFloatChannel, default=1, const [1,56]}
       ring_03_l_ctrl.Scale.Y {type=MovieSceneScriptingFloatChannel, default=1, const [1,56]}
       ring_03_l_ctrl.Scale.Z {type=MovieSceneScriptingFloatChannel, default=1, const [1,56]}
       pinky_metacarpal_l_ctrl.Location.X {type=MovieSceneScriptingFloatChannel, default=0, const [0,56]}
       pinky_metacarpal_l_ctrl.Location.Y {type=MovieSceneScriptingFloatChannel, default=0, const [0,56]}
       pinky_metacarpal_l_ctrl.Location.Z {type=MovieSceneScriptingFloatChannel, default=0, const [0,56]}
       pinky_metacarpal_l_ctrl.Rotation.X {type=MovieSceneScriptingFloatChannel, default=0, const [0,56]}
       pinky_metacarpal_l_ctrl.Rotation.Y {
        type=MovieSceneScriptingFloatChannel
        default=0
        thinned_from=56
        keys [[0,9.5],[283,9.5],[291,7],[298,6.0905],[305,5.6875],[313,5.5],[319,5.6404],[328,7.7105],[339,9.0526],[341,9.1634],[351,9.5]]
       }
       pinky_metacarpal_l_ctrl.Rotation.Z {
        type=MovieSceneScriptingFloatChannel
        default=0
        thinned_from=56
        keys [[0,-5.5],[3,-6.0834],[6,-6.3471],[9,-6.4355],[19,-6.4355],[25,-6.3617],[31,-5.947],[41,-4.7031],[48,-4.5],[54,-4.8206],[59,-5.4062],[64,-6.1377],[70,-6.4497],[76,-6.5538],[82,-5.8839],[90,-4.6181],[101,-3.8586],[109,-4.3353],[115,-5.496],[119,-6.133],[126,-6.3719],[133,-6.5152],[139,-6.0642],[145,-5.0594],[152,-4.5871],[159,-4.3038],[166,-4.0111],[172,-3.6764],[178,-3.4409],[183,-3.2351],[188,-3.1413],[193,-3.1413],[198,-3.4943],[205,-4.156],[212,-4.4042],[220,-4.5935],[225,-4.625],[232,-4.625],[239,-4.4951],[247,-4.1415],[252,-3.7941],[257,-3.6639],[263,-3.5857],[283,-3.5857],[291,-7.0277],[298,-8.28],[305,-8.8347],[311,-9.0929],[318,-9.7167],[328,-7.1073],[336,-3.6338],[339,-2.9897],[351,-5.5]]
       }
       pinky_metacarpal_l_ctrl.Scale.X {type=MovieSceneScriptingFloatChannel, default=1, const [1,56]}
       pinky_metacarpal_l_ctrl.Scale.Y {type=MovieSceneScriptingFloatChannel, default=1, const [1,56]}
       pinky_metacarpal_l_ctrl.Scale.Z {type=MovieSceneScriptingFloatChannel, default=1, const [1,56]}
       pinky_01_l_ctrl.Location.X {type=MovieSceneScriptingFloatChannel, default=0, const [0,56]}
       pinky_01_l_ctrl.Location.Y {type=MovieSceneScriptingFloatChannel, default=0, const [0,56]}
       pinky_01_l_ctrl.Location.Z {type=MovieSceneScriptingFloatChannel, default=0, const [0,56]}
       pinky_01_l_ctrl.Rotation.X {type=MovieSceneScriptingFloatChannel, default=0, const [-16.5,56]}
       pinky_01_l_ctrl.Rotation.Y {type=MovieSceneScriptingFloatChannel, default=0, thinned_from=56, keys [[0,0],[351,0]]}
       pinky_01_l_ctrl.Rotation.Z {
        type=MovieSceneScriptingFloatChannel
        default=0
        thinned_from=56
        keys [[0,-34.3713],[3,-34.9547],[6,-35.2184],[9,-35.3068],[19,-35.3068],[25,-35.233],[31,-34.8184],[41,-33.5744],[48,-33.3713],[54,-33.6919],[59,-34.2776],[64,-35.009],[70,-35.3211],[76,-35.4251],[82,-34.7552],[90,-33.4894],[101,-32.7299],[109,-33.2066],[115,-34.3674],[119,-35.0043],[126,-35.2432],[133,-35.3865],[139,-34.9355],[145,-33.9307],[152,-33.4585],[166,-32.8824],[172,-32.5477],[183,-32.1064],[188,-32.0126],[193,-32.0126],[198,-32.3656],[205,-33.0273],[220,-33.4648],[232,-33.4963],[239,-33.0399],[247,-31.7976],[252,-30.5774],[257,-30.1198],[263,-29.8453],[283,-29.8453],[291,-31.817],[298,-32.5343],[305,-32.8521],[311,-33],[318,-33.6238],[328,-33.7419],[336,-31.9459],[339,-31.4127],[351,-34.3713]]
       }
       pinky_01_l_ctrl.Scale.X {type=MovieSceneScriptingFloatChannel, default=1, const [1,56]}
       pinky_01_l_ctrl.Scale.Y {type=MovieSceneScriptingFloatChannel, default=1, const [1,56]}
       pinky_01_l_ctrl.Scale.Z {type=MovieSceneScriptingFloatChannel, default=1, const [1,56]}
       pinky_02_l_ctrl.Location.X {type=MovieSceneScriptingFloatChannel, default=0, const [0,56]}
       pinky_02_l_ctrl.Location.Y {type=MovieSceneScriptingFloatChannel, default=0, const [0,56]}
       pinky_02_l_ctrl.Location.Z {type=MovieSceneScriptingFloatChannel, default=0, const [0,56]}
       pinky_02_l_ctrl.Rotation.X {type=MovieSceneScriptingFloatChannel, default=0, const [0,56]}
       pinky_02_l_ctrl.Rotation.Y {type=MovieSceneScriptingFloatChannel, default=0, const [0,56]}
       pinky_02_l_ctrl.Rotation.Z {
        type=MovieSceneScriptingFloatChannel
        default=0
        thinned_from=56
        keys [[0,-33.8686],[3,-34.452],[6,-34.7157],[9,-34.8041],[19,-34.8041],[25,-34.7303],[31,-34.3157],[41,-33.0717],[48,-32.8686],[54,-33.1892],[59,-33.7749],[64,-34.5063],[70,-34.8184],[76,-34.9224],[82,-34.2525],[90,-32.9867],[101,-32.2272],[109,-32.7039],[115,-33.8647],[119,-34.5016],[126,-34.7405],[133,-34.8838],[139,-34.4328],[145,-33.428],[152,-32.9558],[166,-32.3797],[172,-32.045],[183,-31.6037],[188,-31.51],[193,-31.51],[198,-31.8629],[205,-32.5246],[220,-32.9621],[232,-32.9936],[239,-32.8637],[247,-32.5101],[252,-32.1628],[257,-32.0325],[263,-31.9544],[283,-31.9544],[291,-36.3579],[298,-37.96],[305,-38.6697],[311,-39],[318,-39.6238],[328,-36.1642],[336,-32.1745],[339,-31.4963],[351,-33.8686]]
       }
       pinky_02_l_ctrl.Scale.X {type=MovieSceneScriptingFloatChannel, default=1, const [1,56]}
       pinky_02_l_ctrl.Scale.Y {type=MovieSceneScriptingFloatChannel, default=1, const [1,56]}
       pinky_02_l_ctrl.Scale.Z {type=MovieSceneScriptingFloatChannel, default=1, const [1,56]}
       pinky_03_l_ctrl.Location.X {type=MovieSceneScriptingFloatChannel, default=0, const [0,56]}
       pinky_03_l_ctrl.Location.Y {type=MovieSceneScriptingFloatChannel, default=0, const [0,56]}
       pinky_03_l_ctrl.Location.Z {type=MovieSceneScriptingFloatChannel, default=0, const [0,56]}
       pinky_03_l_ctrl.Rotation.X {type=MovieSceneScriptingFloatChannel, default=0, const [0,56]}
       pinky_03_l_ctrl.Rotation.Y {type=MovieSceneScriptingFloatChannel, default=0, const [0,56]}
       pinky_03_l_ctrl.Rotation.Z {
        type=MovieSceneScriptingFloatChannel
        default=0
        thinned_from=56
        keys [[0,-29.644],[3,-30.2274],[6,-30.4911],[9,-30.5796],[19,-30.5796],[25,-30.5057],[31,-30.0911],[41,-28.8472],[48,-28.644],[54,-28.9647],[59,-29.5503],[64,-30.2817],[70,-30.5938],[76,-30.6978],[82,-30.0279],[90,-28.7621],[101,-28.0026],[109,-28.4794],[115,-29.6401],[119,-30.2771],[126,-30.5159],[133,-30.6593],[139,-30.2082],[145,-29.2034],[152,-28.7312],[166,-28.1551],[172,-27.8204],[183,-27.3791],[188,-27.2854],[193,-27.2854],[198,-27.6383],[205,-28.3],[212,-28.5482],[220,-28.7376],[232,-28.769],[239,-28.5307],[247,-27.8819],[252,-27.2447],[257,-27.0058],[263,-26.8624],[283,-26.8624],[291,-31.808],[298,-33.6073],[305,-34.4045],[311,-34.7754],[318,-35.3992],[328,-31.9397],[336,-27.9499],[339,-27.2717],[351,-29.644]]
       }
       pinky_03_l_ctrl.Scale.X {type=MovieSceneScriptingFloatChannel, default=1, const [1,56]}
       pinky_03_l_ctrl.Scale.Y {type=MovieSceneScriptingFloatChannel, default=1, const [1,56]}
       pinky_03_l_ctrl.Scale.Z {type=MovieSceneScriptingFloatChannel, default=1, const [1,56]}
       hand_l_ik_ctrl.Location.X {
        type=MovieSceneScriptingFloatChannel
        default=0
        thinned_from=58
        keys [[0,9.7962],[9,9.874],[13,9.884],[19,9.6557],[25,9.2751],[31,9.0468],[41,8.9098],[48,9.0848],[54,9.3765],[59,9.4737],[64,9.532],[76,9.2987],[82,9.0556],[90,8.8231],[101,8.7455],[109,8.8894],[119,8.8575],[145,8.8826],[152,8.4901],[159,8.5324],[166,8.659],[172,8.6115],[178,8.583],[183,8.4657],[193,8.3107],[198,8.2896],[205,8.2826],[212,8.0962],[216,7.7854],[220,7.599],[225,7.5291],[232,7.4872],[239,7.5336],[244,7.6109],[249,7.7397],[255,7.7881],[263,7.8171],[271,8.1881],[277,9.3014],[283,9.6107],[292,9.7499],[301,9.7962],[351,9.7962]]
       }
       hand_l_ik_ctrl.Location.Y {
        type=MovieSceneScriptingFloatChannel
        default=0
        thinned_from=58
        keys [[0,4.0039],[9,4.038],[13,4.0424],[19,4.1104],[25,4.2237],[31,4.2917],[41,4.3325],[48,4.4606],[54,4.6741],[59,4.7453],[64,4.788],[76,4.6172],[82,4.4392],[90,4.9428],[101,5.1107],[109,4.7865],[115,5.3541],[119,5.4959],[126,5.5432],[145,5.5543],[152,5.9763],[159,5.9048],[166,5.6903],[178,5.735],[183,5.8041],[193,5.8953],[205,5.9118],[212,5.8182],[216,5.6622],[220,5.5685],[225,5.5334],[232,5.5123],[239,5.5324],[244,5.5658],[249,5.6215],[255,5.6424],[263,5.6549],[271,5.3453],[277,4.4166],[283,4.1587],[292,4.0426],[301,4.0039],[351,4.0039]]
       }
       hand_l_ik_ctrl.Location.Z {
        type=MovieSceneScriptingFloatChannel
        default=0
        thinned_from=58
        keys [[0,6.5359],[3,6.5147],[9,6.4558],[13,6.4455],[19,6.3838],[25,6.2809],[31,6.2192],[41,6.1822],[48,6.2009],[54,6.232],[64,6.2487],[76,6.2237],[82,6.1978],[90,6.6035],[101,6.7387],[109,6.6401],[115,7.2124],[119,7.3555],[126,7.4032],[145,7.39],[152,8.2173],[159,8.0901],[166,7.7086],[172,7.7754],[178,7.8155],[183,7.9806],[188,8.0797],[193,8.1985],[198,8.2283],[205,8.2382],[212,8.3916],[216,8.6474],[220,8.8009],[225,8.8584],[232,8.8929],[239,8.592],[244,8.0904],[249,7.2545],[255,6.941],[263,6.7529],[271,6.7122],[277,6.5901],[283,6.5562],[351,6.5359]]
       }
       hand_l_ik_ctrl.Rotation.X {
        type=MovieSceneScriptingFloatChannel
        default=0
        thinned_from=58
        keys [[0,1.9549],[9,1.9349],[13,1.9323],[19,1.691],[25,1.2888],[31,1.0474],[41,0.9026],[48,1.0603],[54,1.323],[59,1.4106],[64,1.4631],[70,1.3549],[76,1.2529],[82,1.034],[90,3.2578],[101,3.9991],[109,1.9821],[115,3.3313],[119,3.6686],[126,3.7811],[145,3.7811],[152,2.5037],[159,2.6825],[166,3.2189],[172,3.1018],[178,3.0316],[183,2.7422],[188,2.5686],[193,2.3602],[198,2.3081],[205,2.2908],[212,1.2807],[216,-0.4027],[220,-1.4128],[225,-1.7915],[232,-2.0188],[239,-2.12],[244,-2.2887],[249,-2.5699],[255,-2.6754],[263,-2.7387],[271,-1.8586],[277,0.7815],[283,1.5149],[292,1.8449],[301,1.9549],[351,1.9549]]
       }
       hand_l_ik_ctrl.Rotation.Y {
        type=MovieSceneScriptingFloatChannel
        default=0
        thinned_from=58
        keys [[0,-3.4919],[3,-3.4583],[6,-3.4023],[9,-3.3649],[13,-3.3486],[19,-3.7633],[25,-4.4544],[31,-4.8691],[41,-5.118],[48,-4.7591],[54,-4.1611],[59,-3.9618],[64,-3.8422],[70,-4.0884],[76,-4.3206],[82,-4.8189],[90,-5.8008],[101,-6.1281],[109,-5.0665],[115,-5.0022],[119,-4.9861],[145,-4.9807],[152,-2.4714],[159,-2.83],[166,-3.906],[172,-3.6821],[178,-3.5478],[183,-2.9945],[188,-2.6626],[193,-2.2643],[198,-2.1647],[205,-2.1315],[212,-0.714],[216,1.6485],[220,3.066],[225,3.5976],[232,3.9165],[239,3.589],[244,3.0432],[249,2.1335],[255,1.7924],[263,1.5877],[271,0.6353],[277,-2.222],[283,-3.0157],[292,-3.3728],[301,-3.4919],[351,-3.4919]]
       }
       hand_l_ik_ctrl.Rotation.Z {
        type=MovieSceneScriptingFloatChannel
        default=0
        thinned_from=58
        keys [[0,12.3155],[41,12.3393],[48,12.5832],[54,12.9898],[59,13.1253],[64,13.2066],[76,12.8814],[82,12.5426],[90,12.8436],[101,12.9439],[109,12.6898],[115,13.1279],[119,13.2374],[126,13.2739],[145,13.2739],[152,13.8349],[159,13.7465],[166,13.4813],[178,13.5505],[183,13.6572],[193,13.7982],[205,13.8238],[212,13.2921],[216,12.4058],[220,11.8741],[225,11.6747],[239,11.4578],[244,11.2958],[249,11.0258],[255,10.9246],[263,10.8638],[271,11.136],[277,11.9526],[283,12.1794],[292,12.2815],[301,12.3155],[351,12.3155]]
       }
       hand_l_ik_ctrl.Scale.X {type=MovieSceneScriptingFloatChannel, default=1, const [1,58]}
       hand_l_ik_ctrl.Scale.Y {type=MovieSceneScriptingFloatChannel, default=1, const [1,58]}
       hand_l_ik_ctrl.Scale.Z {type=MovieSceneScriptingFloatChannel, default=1, const [1,58]}
       Softness {type=MovieSceneScriptingFloatChannel, default=0, const [0,58]}
       hand_l_ikorient_ctrl.Rotation.Roll {type=MovieSceneScriptingFloatChannel, default=0, const [0,58]}
       hand_l_ikorient_ctrl.Rotation.Pitch {type=MovieSceneScriptingFloatChannel, default=0, const [0,58]}
       hand_l_ikorient_ctrl.Rotation.Yaw {type=MovieSceneScriptingFloatChannel, default=0, const [0,58]}
       arm_l_pv_ik_ctrl.Location.X {
        type=MovieSceneScriptingFloatChannel
        default=0
        thinned_from=58
        keys [[0,10.6302],[13,10.6499],[19,10.4304],[25,10.0645],[31,9.8449],[41,9.7132],[48,9.7999],[54,9.9444],[64,10.0214],[76,9.9058],[82,9.7855],[90,9.8962],[101,9.9331],[109,9.6752],[115,9.6277],[145,9.6119],[152,9.0242],[159,9.1121],[166,9.3758],[172,9.3266],[178,9.2971],[183,9.1755],[193,9.0151],[198,8.9932],[205,8.9859],[212,8.8072],[216,8.5093],[220,8.3306],[225,8.2635],[232,8.2233],[239,8.1648],[244,8.0673],[249,7.9047],[255,7.8438],[263,7.8072],[271,8.3365],[277,9.9244],[283,10.3655],[292,10.564],[301,10.6302],[351,10.6302]]
       }
       arm_l_pv_ik_ctrl.Location.Y {
        type=MovieSceneScriptingFloatChannel
        default=0
        thinned_from=58
        keys [[0,16.6439],[9,16.5243],[19,16.4719],[31,16.3734],[82,16.3478],[90,17.3176],[101,17.6409],[109,16.9812],[115,17.7687],[119,17.9656],[126,18.0312],[145,18.0312],[152,17.4893],[159,17.5434],[166,17.7057],[178,17.5972],[183,17.4297],[193,17.2085],[205,17.1683],[212,16.5158],[216,15.4282],[220,14.7757],[225,14.531],[239,14.2245],[244,13.9585],[249,13.5151],[255,13.3488],[263,13.249],[271,13.8856],[277,15.7952],[283,16.3256],[292,16.5644],[301,16.6439],[351,16.6439]]
       }
       arm_l_pv_ik_ctrl.Location.Z {
        type=MovieSceneScriptingFloatChannel
        default=0
        thinned_from=58
        keys [[0,-2.877],[3,-2.8885],[6,-2.9077],[9,-2.9205],[13,-2.9261],[19,-2.7911],[25,-2.5662],[31,-2.4313],[41,-2.3503],[48,-2.538],[54,-2.8508],[59,-2.9551],[64,-3.0176],[70,-2.8888],[76,-2.7674],[82,-2.5067],[90,-1.7936],[101,-1.5559],[109,-2.2443],[115,-2.1117],[119,-2.0786],[126,-2.0676],[133,-2.0764],[139,-2.091],[145,-2.0998],[152,-1.9019],[159,-1.8992],[166,-1.8912],[172,-1.8466],[178,-1.8199],[183,-1.7097],[188,-1.6437],[193,-1.5644],[198,-1.5446],[205,-1.538],[212,-1.3065],[216,-0.9208],[220,-0.6893],[225,-0.6025],[232,-0.5505],[239,-0.7015],[244,-0.9532],[249,-1.3727],[255,-1.53],[263,-1.6244],[271,-1.8593],[277,-2.5639],[283,-2.7596],[292,-2.8477],[301,-2.877],[351,-2.877]]
       }
       arm_l_pv_ik_ctrl.Rotation.X {
        type=MovieSceneScriptingFloatChannel
        default=0
        thinned_from=58
        keys [[0,5.4301],[3,5.4827],[6,5.5704],[9,5.629],[13,5.6547],[19,5.559],[25,5.3995],[31,5.3039],[41,5.2465],[48,5.3032],[54,5.3978],[59,5.4293],[64,5.4482],[76,5.3725],[82,5.2937],[90,6.4335],[101,6.8134],[109,5.3759],[115,5.5378],[119,5.5782],[126,5.5917],[145,5.5917],[152,5.2872],[159,5.3789],[166,5.6542],[178,5.675],[183,5.7072],[193,5.7497],[205,5.7574],[212,6.0889],[216,6.6414],[220,6.9729],[225,7.0972],[232,7.1718],[239,7.1265],[244,7.0509],[249,6.925],[255,6.8778],[263,6.8495],[271,6.5834],[277,5.7849],[283,5.5631],[292,5.4633],[301,5.4301],[351,5.4301]]
       }
       arm_l_pv_ik_ctrl.Rotation.Y {
        type=MovieSceneScriptingFloatChannel
        default=0
        thinned_from=58
        keys [[0,-8.7365],[13,-8.7022],[19,-8.9725],[25,-9.4229],[31,-9.6932],[41,-9.8553],[64,-9.7766],[76,-9.8061],[90,-9.8796],[101,-9.8938],[109,-9.8566],[145,-9.8588],[152,-9.8116],[178,-9.781],[193,-9.7114],[205,-9.7042],[212,-9.7399],[216,-9.7995],[220,-9.8352],[232,-9.8567],[239,-9.9085],[244,-9.995],[249,-10.1391],[255,-10.1932],[263,-10.2256],[271,-9.9464],[277,-9.1088],[283,-8.8761],[292,-8.7714],[301,-8.7365],[351,-8.7365]]
       }
       arm_l_pv_ik_ctrl.Rotation.Z {
        type=MovieSceneScriptingFloatChannel
        default=0
        thinned_from=58
        keys [[0,24.2337],[13,24.1376],[19,23.9567],[25,23.6552],[31,23.4743],[41,23.3658],[48,23.569],[54,23.9076],[64,24.0882],[76,23.8173],[82,23.5351],[90,25.4511],[101,26.0898],[109,24.5016],[115,25.9214],[119,26.2763],[126,26.3946],[145,26.3946],[152,25.3875],[159,25.5081],[166,25.8701],[178,25.6953],[183,25.4253],[193,25.0689],[205,25.0041],[212,23.8965],[216,22.0505],[220,20.9429],[225,20.5275],[232,20.2783],[239,20.1612],[244,19.966],[249,19.6406],[255,19.5185],[263,19.4453],[271,20.3431],[277,23.0366],[283,23.7848],[292,24.1215],[301,24.2337],[351,24.2337]]
       }
       arm_l_pv_ik_ctrl.Scale.X {type=MovieSceneScriptingFloatChannel, default=1, const [1,58]}
       arm_l_pv_ik_ctrl.Scale.Y {type=MovieSceneScriptingFloatChannel, default=1, const [1,58]}
       arm_l_pv_ik_ctrl.Scale.Z {type=MovieSceneScriptingFloatChannel, default=1, const [1,58]}
       upperarm_r_fk_ctrl.Location.X {type=MovieSceneScriptingFloatChannel, default=0, const [0,53]}
       upperarm_r_fk_ctrl.Location.Y {type=MovieSceneScriptingFloatChannel, default=0, const [0,53]}
       upperarm_r_fk_ctrl.Location.Z {type=MovieSceneScriptingFloatChannel, default=0, const [0,53]}
       upperarm_r_fk_ctrl.Rotation.X {type=MovieSceneScriptingFloatChannel, default=0, const [-23.5,52]}
       upperarm_r_fk_ctrl.Rotation.Y {
        type=MovieSceneScriptingFloatChannel
        default=0
        thinned_from=56
        keys [[0,-10.5],[6,-10.5],[14,-10.6328],[25,-10.9056],[31,-10.9966],[41,-11.0437],[48,-10.9994],[54,-10.895],[70,-10.4668],[76,-10.3462],[82,-10.5962],[90,-11.2772],[101,-11.8093],[109,-11.9962],[115,-12.019],[119,-11.8813],[126,-11.2986],[133,-10.7866],[139,-10.9691],[145,-11.4004],[152,-11.9439],[159,-12.2496],[166,-11.9378],[178,-10.782],[183,-11.0207],[188,-11.5866],[193,-12.5509],[198,-13.736],[205,-14.461],[212,-14.6776],[216,-14.4464],[232,-12.611],[239,-12.2092],[263,-12.6437],[267,-12.6627],[269,-12.6466],[271,-12.5847],[289,-8.237],[313,-6.857],[321,-7.322],[336,-9.9833],[351,-10.5]]
       }
       upperarm_r_fk_ctrl.Rotation.Z {
        type=MovieSceneScriptingFloatChannel
        default=0
        keys [[0,-0.101],[3,-0.1597],[6,-0.244],[9,-0.3574],[13,-0.5365],[19,-0.8342],[25,-1.1069],[31,-1.2822],[41,-1.205],[48,-0.7918],[54,-0.186],[57,0.1185],[64,0.886],[70,1.233],[76,1.3419],[82,1.4942],[88,2.3813],[94,1.7103],[124,-2.6736],[145,-3.0869],[152,-2.7974],[159,-2.1707],[166,-1.4035],[172,-0.9445],[178,-1.443],[183,-2.2942],[188,-3.8219],[193,-4.2083],[198,-3.9302],[205,-3.0917],[212,-1.9178],[216,-1.3308],[220,-0.937],[225,-0.8211],[232,-1.2922],[244,-3.1494],[249,-4.091],[255,-4.9268],[263,-5.4898],[281,-5.8432],[298,-5.8887],[313,-5.2876],[336,-1.4486],[351,-0.101]]
       }
       upperarm_r_fk_ctrl.Scale.X {type=MovieSceneScriptingFloatChannel, default=1, const [1,53]}
       upperarm_r_fk_ctrl.Scale.Y {type=MovieSceneScriptingFloatChannel, default=1, const [1,53]}
       upperarm_r_fk_ctrl.Scale.Z {type=MovieSceneScriptingFloatChannel, default=1, const [1,53]}
       lowerarm_r_fk_ctrl.Location.X {type=MovieSceneScriptingFloatChannel, default=0, thinned_from=53, keys [[0,0],[351,0]]}
       lowerarm_r_fk_ctrl.Location.Y {type=MovieSceneScriptingFloatChannel, default=0, thinned_from=53, keys [[0,0],[351,0]]}
       lowerarm_r_fk_ctrl.Location.Z {type=MovieSceneScriptingFloatChannel, default=0, thinned_from=53, keys [[0,0],[351,0]]}
       lowerarm_r_fk_ctrl.Rotation.X {type=MovieSceneScriptingFloatChannel, default=0, thinned_from=53, keys [[0,0],[351,0]]}
       lowerarm_r_fk_ctrl.Rotation.Y {type=MovieSceneScriptingFloatChannel, default=0, thinned_from=53, keys [[0,0],[351,0]]}
       lowerarm_r_fk_ctrl.Rotation.Z {
        type=MovieSceneScriptingFloatChannel
        default=0
        thinned_from=35
        keys [[0,4.5],[6,4.1968],[9,4.0801],[13,3.9473],[19,3.818],[25,3.7748],[29,3.8474],[36,4.1762],[53,5.4848],[59,5.7751],[64,5.2804],[80,2.8663],[85,2.5175],[97,4.1602],[109,6.4861],[114,6.6995],[121,6.0043],[135,3.8578],[143,3.1861],[156,4.5753],[172,6.4955],[177,6.8561],[204,3.0361],[225,7.4313],[232,7.861],[239,7.155],[255,2.6614],[277,-7.2352],[299,-11.1384],[313,-11.1384],[319,-9.3936],[338,3.2486],[351,4.5]]
       }
       lowerarm_r_fk_ctrl.Scale.X {type=MovieSceneScriptingFloatChannel, default=1, const [1,53]}
       lowerarm_r_fk_ctrl.Scale.Y {type=MovieSceneScriptingFloatChannel, default=1, const [1,53]}
       lowerarm_r_fk_ctrl.Scale.Z {type=MovieSceneScriptingFloatChannel, default=1, const [1,53]}
       hand_r_fk_ctrl.Location.X {type=MovieSceneScriptingFloatChannel, default=0, const [0,53]}
       hand_r_fk_ctrl.Location.Y {type=MovieSceneScriptingFloatChannel, default=0, const [0,53]}
       hand_r_fk_ctrl.Location.Z {type=MovieSceneScriptingFloatChannel, default=0, const [0,53]}
       hand_r_fk_ctrl.Rotation.X {type=MovieSceneScriptingFloatChannel, default=0, const [5,53]}
       hand_r_fk_ctrl.Rotation.Y {
        type=MovieSceneScriptingFloatChannel
        default=0
        thinned_from=53
        keys [[0,5],[263,5],[281,9.4854],[298,13],[313,13],[319,12.7193],[336,8.2361],[351,5]]
       }
       hand_r_fk_ctrl.Rotation.Z {type=MovieSceneScriptingFloatChannel, default=0, const [-6.5,53]}
       hand_r_fk_ctrl.Scale.X {type=MovieSceneScriptingFloatChannel, default=1, const [1,53]}
       hand_r_fk_ctrl.Scale.Y {type=MovieSceneScriptingFloatChannel, default=1, const [1,53]}
       hand_r_fk_ctrl.Scale.Z {type=MovieSceneScriptingFloatChannel, default=1, const [1,53]}
       thumb_01_r_ctrl.Location.X {type=MovieSceneScriptingFloatChannel, default=0, const [0,55]}
       thumb_01_r_ctrl.Location.Y {type=MovieSceneScriptingFloatChannel, default=0, const [0,55]}
       thumb_01_r_ctrl.Location.Z {type=MovieSceneScriptingFloatChannel, default=0, const [0,55]}
       thumb_01_r_ctrl.Rotation.X {type=MovieSceneScriptingFloatChannel, default=0, const [34,55]}
       thumb_01_r_ctrl.Rotation.Y {type=MovieSceneScriptingFloatChannel, default=0, const [6,55]}
       thumb_01_r_ctrl.Rotation.Z {
        type=MovieSceneScriptingFloatChannel
        default=0
        thinned_from=55
        keys [[0,-4.625],[3,-4.5525],[6,-4.4619],[9,-4.4317],[19,-5.1318],[25,-5.6407],[31,-5.8623],[41,-5.9953],[48,-5.2007],[54,-4.0396],[59,-3.5171],[64,-3.343],[70,-4.0186],[76,-4.7495],[82,-5.1447],[90,-5.1447],[101,-4.7272],[109,-4.0315],[115,-3.4369],[119,-3.2387],[126,-3.6218],[133,-4.5206],[139,-4.9382],[145,-5.1972],[152,-5.461],[159,-5.6923],[166,-5.831],[172,-4.9791],[178,-3.9057],[183,-2.7445],[188,-2.1168],[193,-2.3648],[198,-2.8474],[205,-3.7486],[211,-4.3389],[215,-4.4139],[219,-4.0208],[223,-3.5428],[229,-2.9845],[236,-1.7066],[242,-0.9398],[248,-0.5355],[267,-0.5355],[276,-2.1425],[283,-5.2129],[289,-8.1783],[295,-10.3429],[303,-11.6417],[312,-12],[318,-12.9363],[327,-15.7451],[339,-6.7399],[351,-4.625]]
       }
       thumb_01_r_ctrl.Scale.X {type=MovieSceneScriptingFloatChannel, default=1, const [1,55]}
       thumb_01_r_ctrl.Scale.Y {type=MovieSceneScriptingFloatChannel, default=1, const [1,55]}
       thumb_01_r_ctrl.Scale.Z {type=MovieSceneScriptingFloatChannel, default=1, const [1,55]}
       thumb_02_r_ctrl.Location.X {type=MovieSceneScriptingFloatChannel, default=0, const [0,55]}
       thumb_02_r_ctrl.Location.Y {type=MovieSceneScriptingFloatChannel, default=0, const [0,55]}
       thumb_02_r_ctrl.Location.Z {type=MovieSceneScriptingFloatChannel, default=0, const [0,55]}
       thumb_02_r_ctrl.Rotation.X {type=MovieSceneScriptingFloatChannel, default=0, const [0,55]}
       thumb_02_r_ctrl.Rotation.Y {type=MovieSceneScriptingFloatChannel, default=0, const [0,55]}
       thumb_02_r_ctrl.Rotation.Z {
        type=MovieSceneScriptingFloatChannel
        default=0
        thinned_from=55
        keys [[0,26.5114],[6,26.9168],[9,26.9919],[19,26.2917],[25,25.7829],[31,25.5612],[41,25.4283],[48,26.2229],[54,27.3839],[59,27.9064],[64,28.0806],[70,27.4049],[76,26.674],[82,26.2789],[90,26.2789],[101,26.6963],[109,27.392],[115,27.9866],[119,28.1848],[126,27.8017],[133,26.903],[139,26.4854],[152,25.9626],[159,25.7313],[166,25.5925],[172,26.1572],[178,27.2306],[183,28.3919],[188,29.0196],[193,28.7716],[198,28.2889],[205,27.3878],[211,27.0847],[215,27.0097],[219,27.4028],[223,27.8808],[229,28.4478],[236,29.7473],[242,30.527],[248,30.9385],[267,30.9385],[276,29.2842],[283,26.1233],[289,23.0706],[295,20.8422],[303,19.5052],[312,19.1364],[318,18.2001],[327,15.3913],[339,24.3964],[351,26.5114]]
       }
       thumb_02_r_ctrl.Scale.X {type=MovieSceneScriptingFloatChannel, default=1, const [1,55]}
       thumb_02_r_ctrl.Scale.Y {type=MovieSceneScriptingFloatChannel, default=1, const [1,55]}
       thumb_02_r_ctrl.Scale.Z {type=MovieSceneScriptingFloatChannel, default=1, const [1,55]}
       thumb_03_r_ctrl.Location.X {type=MovieSceneScriptingFloatChannel, default=0, const [0,55]}
       thumb_03_r_ctrl.Location.Y {type=MovieSceneScriptingFloatChannel, default=0, const [0,55]}
       thumb_03_r_ctrl.Location.Z {type=MovieSceneScriptingFloatChannel, default=0, const [0,55]}
       thumb_03_r_ctrl.Rotation.X {type=MovieSceneScriptingFloatChannel, default=0, const [0,55]}
       thumb_03_r_ctrl.Rotation.Y {type=MovieSceneScriptingFloatChannel, default=0, const [0,55]}
       thumb_03_r_ctrl.Rotation.Z {
        type=MovieSceneScriptingFloatChannel
        default=0
        thinned_from=55
        keys [[0,-35.5],[6,-35.0946],[9,-35.0195],[19,-35.7196],[25,-36.2285],[31,-36.4501],[41,-36.5831],[48,-35.7885],[54,-34.6274],[59,-34.1049],[64,-33.9308],[76,-35.3374],[82,-35.7325],[90,-35.7325],[101,-35.315],[109,-34.6193],[115,-34.0247],[119,-33.8265],[126,-34.2096],[133,-35.1084],[139,-35.526],[159,-36.2801],[166,-36.4188],[172,-35.8541],[178,-34.7807],[183,-33.6195],[188,-32.9917],[193,-33.2397],[198,-33.7224],[205,-34.6235],[211,-34.9267],[215,-35.0017],[219,-34.6086],[223,-34.1306],[229,-33.5636],[236,-32.2641],[242,-31.4843],[248,-31.0729],[267,-31.0729],[276,-32.7272],[283,-35.888],[289,-38.9407],[295,-41.1691],[303,-42.5062],[312,-42.875],[318,-43.8113],[327,-46.6201],[339,-37.6149],[351,-35.5]]
       }
       thumb_03_r_ctrl.Scale.X {type=MovieSceneScriptingFloatChannel, default=1, const [1,55]}
       thumb_03_r_ctrl.Scale.Y {type=MovieSceneScriptingFloatChannel, default=1, const [1,55]}
       thumb_03_r_ctrl.Scale.Z {type=MovieSceneScriptingFloatChannel, default=1, const [1,55]}
       index_metacarpal_r_ctrl.Location.X {type=MovieSceneScriptingFloatChannel, default=0, const [0,55]}
       index_metacarpal_r_ctrl.Location.Y {type=MovieSceneScriptingFloatChannel, default=0, const [0,55]}
       index_metacarpal_r_ctrl.Location.Z {type=MovieSceneScriptingFloatChannel, default=0, const [0,55]}
       index_metacarpal_r_ctrl.Rotation.X {type=MovieSceneScriptingFloatChannel, default=0, const [0,55]}
       index_metacarpal_r_ctrl.Rotation.Y {type=MovieSceneScriptingFloatChannel, default=0, const [4,55]}
       index_metacarpal_r_ctrl.Rotation.Z {
        type=MovieSceneScriptingFloatChannel
        default=0
        thinned_from=55
        keys [[0,3.5],[3,3.6802],[6,3.9054],[9,3.9805],[19,3.2804],[25,2.7715],[31,2.5499],[41,2.4169],[48,3.2115],[54,4.3726],[59,4.8951],[64,5.0692],[70,4.3936],[76,3.6627],[82,3.2675],[90,3.2675],[101,3.685],[109,4.3807],[115,4.9753],[119,5.1735],[126,4.7904],[133,3.8916],[139,3.474],[145,3.215],[152,2.9512],[159,2.7199],[166,2.5812],[172,3.1459],[178,4.2193],[183,5.3805],[188,6.0083],[193,5.7603],[198,5.2776],[205,4.3765],[211,4.0733],[215,3.9983],[219,4.3914],[223,4.8694],[229,5.4364],[236,6.7359],[242,7.5157],[248,7.9271],[267,7.9271],[276,6.2728],[283,3.112],[289,0.0593],[295,-2.1691],[303,-3.5062],[312,-3.875],[318,-4.8113],[327,-7.6201],[339,1.3851],[351,3.5]]
       }
       index_metacarpal_r_ctrl.Scale.X {type=MovieSceneScriptingFloatChannel, default=1, const [1,55]}
       index_metacarpal_r_ctrl.Scale.Y {type=MovieSceneScriptingFloatChannel, default=1, const [1,55]}
       index_metacarpal_r_ctrl.Scale.Z {type=MovieSceneScriptingFloatChannel, default=1, const [1,55]}
       index_01_r_ctrl.Location.X {type=MovieSceneScriptingFloatChannel, default=0, const [0,55]}
       index_01_r_ctrl.Location.Y {type=MovieSceneScriptingFloatChannel, default=0, const [0,55]}
       index_01_r_ctrl.Location.Z {type=MovieSceneScriptingFloatChannel, default=0, const [0,55]}
       index_01_r_ctrl.Rotation.X {type=MovieSceneScriptingFloatChannel, default=0, const [0,55]}
       index_01_r_ctrl.Rotation.Y {type=MovieSceneScriptingFloatChannel, default=0, const [0,55]}
       index_01_r_ctrl.Rotation.Z {
        type=MovieSceneScriptingFloatChannel
        default=0
        thinned_from=55
        keys [[0,24.5],[9,24.5],[19,23.7999],[25,23.291],[31,23.0694],[41,22.9364],[48,23.731],[54,24.8921],[59,25.4145],[64,25.5887],[70,24.9131],[76,24.1821],[82,23.787],[90,23.787],[101,24.2044],[109,24.9001],[115,25.4947],[119,25.693],[126,25.3099],[133,24.4111],[139,23.9935],[152,23.4707],[159,23.2394],[166,23.1006],[172,23.6654],[178,24.7388],[183,25.9],[188,26.5277],[193,26.2797],[198,25.7971],[205,24.8959],[211,24.5928],[215,24.5178],[219,24.9109],[223,25.3889],[229,26.2672],[236,28.316],[242,29.5453],[248,30.1851],[267,30.1851],[276,27.1455],[283,21.3378],[289,15.7288],[295,11.6344],[303,9.1777],[318,8.0363],[327,6.6451],[339,19.9295],[351,24.0195]]
       }
       index_01_r_ctrl.Scale.X {type=MovieSceneScriptingFloatChannel, default=1, const [1,55]}
       index_01_r_ctrl.Scale.Y {type=MovieSceneScriptingFloatChannel, default=1, const [1,55]}
       index_01_r_ctrl.Scale.Z {type=MovieSceneScriptingFloatChannel, default=1, const [1,55]}
       index_02_r_ctrl.Location.X {type=MovieSceneScriptingFloatChannel, default=0, const [0,55]}
       index_02_r_ctrl.Location.Y {type=MovieSceneScriptingFloatChannel, default=0, const [0,55]}
       index_02_r_ctrl.Location.Z {type=MovieSceneScriptingFloatChannel, default=0, const [0,55]}
       index_02_r_ctrl.Rotation.X {type=MovieSceneScriptingFloatChannel, default=0, const [0,55]}
       index_02_r_ctrl.Rotation.Y {type=MovieSceneScriptingFloatChannel, default=0, const [0,55]}
       index_02_r_ctrl.Rotation.Z {
        type=MovieSceneScriptingFloatChannel
        default=0
        thinned_from=55
        keys [[0,0],[3,0.1802],[6,0.4054],[9,0.4805],[13,0.2037],[19,-0.2196],[25,-0.7285],[31,-0.9501],[41,-1.0831],[48,-0.2885],[54,0.8726],[59,1.3951],[64,1.5692],[70,0.8936],[76,0.1627],[82,-0.2325],[90,-0.2325],[101,0.185],[109,0.8807],[115,1.4753],[119,1.6735],[126,1.2904],[133,0.3916],[139,-0.026],[145,-0.285],[152,-0.5488],[159,-0.7801],[166,-0.9188],[172,-0.3541],[178,0.7193],[183,1.8805],[188,2.5083],[193,2.2603],[198,1.7776],[205,0.8765],[211,0.5733],[215,0.4983],[219,0.8914],[223,1.3694],[229,1.9364],[236,3.2359],[242,4.0157],[248,4.4271],[267,4.4271],[276,2.7728],[283,-0.388],[289,-3.4407],[295,-5.6691],[303,-7.0062],[312,-7.375],[318,-8.3113],[327,-11.1201],[339,-2.1149],[351,0]]
       }
       index_02_r_ctrl.Scale.X {type=MovieSceneScriptingFloatChannel, default=1, const [1,55]}
       index_02_r_ctrl.Scale.Y {type=MovieSceneScriptingFloatChannel, default=1, const [1,55]}
       index_02_r_ctrl.Scale.Z {type=MovieSceneScriptingFloatChannel, default=1, const [1,55]}
       index_03_r_ctrl.Location.X {type=MovieSceneScriptingFloatChannel, default=0, const [0,55]}
       index_03_r_ctrl.Location.Y {type=MovieSceneScriptingFloatChannel, default=0, const [0,55]}
       index_03_r_ctrl.Location.Z {type=MovieSceneScriptingFloatChannel, default=0, const [0,55]}
       index_03_r_ctrl.Rotation.X {type=MovieSceneScriptingFloatChannel, default=0, const [0,55]}
       index_03_r_ctrl.Rotation.Y {type=MovieSceneScriptingFloatChannel, default=0, const [0,55]}
       index_03_r_ctrl.Rotation.Z {
        type=MovieSceneScriptingFloatChannel
        default=0
        thinned_from=55
        keys [[0,0],[3,0.1802],[6,0.4054],[9,0.4805],[13,0.2037],[19,-0.2196],[25,-0.7285],[31,-0.9501],[41,-1.0831],[48,-0.2885],[54,0.8726],[59,1.3951],[64,1.5692],[70,0.8936],[76,0.1627],[82,-0.2325],[90,-0.2325],[101,0.185],[109,0.8807],[115,1.4753],[119,1.6735],[126,1.2904],[133,0.3916],[139,-0.026],[145,-0.285],[152,-0.5488],[159,-0.7801],[166,-0.9188],[172,-0.3541],[178,0.7193],[183,1.8805],[188,2.5083],[193,2.2603],[198,1.7776],[205,0.8765],[211,0.5733],[215,0.4983],[219,0.8914],[223,1.3694],[229,1.9364],[236,3.2359],[242,4.0157],[248,4.4271],[267,4.4271],[276,2.7728],[283,-0.388],[289,-3.4407],[295,-5.6691],[303,-7.0062],[312,-7.375],[318,-8.3113],[327,-11.1201],[339,-2.1149],[351,0]]
       }
       index_03_r_ctrl.Scale.X {type=MovieSceneScriptingFloatChannel, default=1, const [1,55]}
       index_03_r_ctrl.Scale.Y {type=MovieSceneScriptingFloatChannel, default=1, const [1,55]}
       index_03_r_ctrl.Scale.Z {type=MovieSceneScriptingFloatChannel, default=1, const [1,55]}
       middle_metacarpal_r_ctrl.Location.X {type=MovieSceneScriptingFloatChannel, default=0, const [0,55]}
       middle_metacarpal_r_ctrl.Location.Y {type=MovieSceneScriptingFloatChannel, default=0, const [0,55]}
       middle_metacarpal_r_ctrl.Location.Z {type=MovieSceneScriptingFloatChannel, default=0, const [0,55]}
       middle_metacarpal_r_ctrl.Rotation.X {type=MovieSceneScriptingFloatChannel, default=0, const [0,55]}
       middle_metacarpal_r_ctrl.Rotation.Y {type=MovieSceneScriptingFloatChannel, default=0, const [0.5,55]}
       middle_metacarpal_r_ctrl.Rotation.Z {
        type=MovieSceneScriptingFloatChannel
        default=0
        thinned_from=55
        keys [[0,0],[3,0.1802],[6,0.4054],[9,0.4805],[13,0.2037],[19,-0.2196],[25,-0.7285],[31,-0.9501],[41,-1.0831],[48,-0.2885],[54,0.8726],[59,1.3951],[64,1.5692],[70,0.8936],[76,0.1627],[82,-0.2325],[90,-0.2325],[101,0.185],[109,0.8807],[115,1.4753],[119,1.6735],[126,1.2904],[133,0.3916],[139,-0.026],[145,-0.285],[152,-0.5488],[159,-0.7801],[166,-0.9188],[172,-0.3541],[178,0.7193],[183,1.8805],[188,2.5083],[193,2.2603],[198,1.7776],[205,0.8765],[211,0.5733],[215,0.4983],[219,0.8914],[223,1.3694],[229,1.9364],[236,3.2359],[242,4.0157],[248,4.4271],[267,4.4271],[276,2.7728],[283,-0.388],[289,-3.4407],[295,-5.6691],[303,-7.0062],[312,-7.375],[318,-8.3113],[327,-11.1201],[339,-2.1149],[351,0]]
       }
       middle_metacarpal_r_ctrl.Scale.X {type=MovieSceneScriptingFloatChannel, default=1, const [1,55]}
       middle_metacarpal_r_ctrl.Scale.Y {type=MovieSceneScriptingFloatChannel, default=1, const [1,55]}
       middle_metacarpal_r_ctrl.Scale.Z {type=MovieSceneScriptingFloatChannel, default=1, const [1,55]}
       middle_01_r_ctrl.Location.X {type=MovieSceneScriptingFloatChannel, default=0, const [0,55]}
       middle_01_r_ctrl.Location.Y {type=MovieSceneScriptingFloatChannel, default=0, const [0,55]}
       middle_01_r_ctrl.Location.Z {type=MovieSceneScriptingFloatChannel, default=0, const [0,55]}
       middle_01_r_ctrl.Rotation.X {type=MovieSceneScriptingFloatChannel, default=0, const [0,55]}
       middle_01_r_ctrl.Rotation.Y {type=MovieSceneScriptingFloatChannel, default=0, const [0,55]}
       middle_01_r_ctrl.Rotation.Z {
        type=MovieSceneScriptingFloatChannel
        default=0
        thinned_from=55
        keys [[0,-4.5],[3,-4.3198],[6,-4.0946],[9,-4.0195],[19,-4.7196],[25,-5.2285],[31,-5.4501],[41,-5.5831],[48,-4.7885],[54,-3.6274],[59,-3.1049],[64,-2.9308],[70,-3.6064],[76,-4.3373],[82,-4.7325],[90,-4.7325],[101,-4.315],[109,-3.6193],[115,-3.0247],[119,-2.8265],[126,-3.2096],[133,-4.1084],[139,-4.526],[145,-4.785],[152,-5.0488],[159,-5.2801],[166,-5.4188],[172,-4.8541],[178,-3.7807],[183,-2.6195],[188,-1.9917],[193,-2.2397],[198,-2.7224],[205,-3.6235],[211,-3.9267],[215,-4.0017],[219,-3.6086],[223,-3.1306],[229,-2.5636],[236,-1.2641],[242,-0.4843],[248,-0.0729],[267,-0.0729],[276,-2.0951],[283,-5.959],[289,-9.6907],[295,-12.4147],[303,-14.0491],[312,-14.5],[318,-15.284],[327,-17.6359],[339,-7.2515],[351,-4.5]]
       }
       middle_01_r_ctrl.Scale.X {type=MovieSceneScriptingFloatChannel, default=1, const [1,55]}
       middle_01_r_ctrl.Scale.Y {type=MovieSceneScriptingFloatChannel, default=1, const [1,55]}
       middle_01_r_ctrl.Scale.Z {type=MovieSceneScriptingFloatChannel, default=1, const [1,55]}
       middle_02_r_ctrl.Location.X {type=MovieSceneScriptingFloatChannel, default=0, const [0,55]}
       middle_02_r_ctrl.Location.Y {type=MovieSceneScriptingFloatChannel, default=0, const [0,55]}
       middle_02_r_ctrl.Location.Z {type=MovieSceneScriptingFloatChannel, default=0, const [0,55]}
       middle_02_r_ctrl.Rotation.X {type=MovieSceneScriptingFloatChannel, default=0, const [0,55]}
       middle_02_r_ctrl.Rotation.Y {type=MovieSceneScriptingFloatChannel, default=0, const [0,55]}
       middle_02_r_ctrl.Rotation.Z {
        type=MovieSceneScriptingFloatChannel
        default=0
        thinned_from=55
        keys [[0,2.9932],[3,3.1734],[6,3.3986],[9,3.4737],[13,3.1969],[19,2.7736],[25,2.2647],[31,2.043],[41,1.9101],[48,2.7047],[54,3.8657],[59,4.3882],[64,4.5624],[70,3.8868],[76,3.1558],[82,2.7607],[90,2.7607],[101,3.1781],[109,3.8738],[115,4.4684],[119,4.6666],[126,4.2836],[133,3.3848],[139,2.9672],[145,2.7082],[152,2.4444],[159,2.2131],[166,2.0743],[172,2.6391],[178,3.7124],[183,4.8737],[188,5.5014],[193,5.2534],[198,4.7708],[205,3.8696],[211,3.5665],[215,3.4915],[219,3.8846],[223,4.3626],[229,4.9296],[236,6.2291],[242,7.0088],[248,7.4203],[267,7.4203],[276,5.398],[283,1.5342],[289,-2.1975],[295,-4.9216],[303,-6.556],[312,-7.0068],[318,-7.7908],[327,-10.1427],[339,0.2417],[351,2.9932]]
       }
       middle_02_r_ctrl.Scale.X {type=MovieSceneScriptingFloatChannel, default=1, const [1,55]}
       middle_02_r_ctrl.Scale.Y {type=MovieSceneScriptingFloatChannel, default=1, const [1,55]}
       middle_02_r_ctrl.Scale.Z {type=MovieSceneScriptingFloatChannel, default=1, const [1,55]}
       middle_03_r_ctrl.Location.X {type=MovieSceneScriptingFloatChannel, default=0, const [0,55]}
       middle_03_r_ctrl.Location.Y {type=MovieSceneScriptingFloatChannel, default=0, const [0,55]}
       middle_03_r_ctrl.Location.Z {type=MovieSceneScriptingFloatChannel, default=0, const [0,55]}
       middle_03_r_ctrl.Rotation.X {type=MovieSceneScriptingFloatChannel, default=0, const [0,55]}
       middle_03_r_ctrl.Rotation.Y {type=MovieSceneScriptingFloatChannel, default=0, const [0,55]}
       middle_03_r_ctrl.Rotation.Z {
        type=MovieSceneScriptingFloatChannel
        default=0
        thinned_from=55
        keys [[0,-6.481],[3,-6.3009],[6,-6.0756],[9,-6.0005],[19,-6.7007],[25,-7.2095],[31,-7.4312],[41,-7.5641],[48,-6.7695],[54,-5.6085],[59,-5.086],[64,-4.9118],[70,-5.5875],[76,-6.3184],[82,-6.7135],[90,-6.7135],[101,-6.2961],[109,-5.6004],[115,-5.0058],[119,-4.8076],[126,-5.1907],[133,-6.0894],[139,-6.507],[145,-6.766],[152,-7.0298],[159,-7.2611],[166,-7.3999],[172,-6.8352],[178,-5.7618],[183,-4.6005],[188,-3.9728],[193,-4.2208],[198,-4.7035],[205,-5.6046],[211,-5.9077],[215,-5.9827],[219,-5.5896],[223,-5.1116],[229,-4.5447],[236,-3.2451],[242,-2.4654],[248,-2.0539],[267,-2.0539],[276,-3.7082],[283,-6.8691],[289,-9.9218],[295,-12.1502],[303,-13.4872],[312,-13.856],[318,-14.7923],[327,-17.6011],[339,-8.596],[351,-6.4811]]
       }
       middle_03_r_ctrl.Scale.X {type=MovieSceneScriptingFloatChannel, default=1, const [1,55]}
       middle_03_r_ctrl.Scale.Y {type=MovieSceneScriptingFloatChannel, default=1, const [1,55]}
       middle_03_r_ctrl.Scale.Z {type=MovieSceneScriptingFloatChannel, default=1, const [1,55]}
       ring_metacarpal_r_ctrl.Location.X {type=MovieSceneScriptingFloatChannel, default=0, const [0,55]}
       ring_metacarpal_r_ctrl.Location.Y {type=MovieSceneScriptingFloatChannel, default=0, const [0,55]}
       ring_metacarpal_r_ctrl.Location.Z {type=MovieSceneScriptingFloatChannel, default=0, const [0,55]}
       ring_metacarpal_r_ctrl.Rotation.X {type=MovieSceneScriptingFloatChannel, default=0, const [0,55]}
       ring_metacarpal_r_ctrl.Rotation.Y {
        type=MovieSceneScriptingFloatChannel
        default=0
        thinned_from=55
        keys [[0,0],[267,0],[276,-0.1402],[283,-0.408],[289,-0.6666],[295,-0.8555],[303,-0.9687],[312,-1],[319,-0.9649],[328,-0.6694],[337,-0.1769],[351,0]]
       }
       ring_metacarpal_r_ctrl.Rotation.Z {
        type=MovieSceneScriptingFloatChannel
        default=0
        thinned_from=55
        keys [[0,0],[3,0.1802],[6,0.4054],[9,0.4805],[13,0.2037],[19,-0.2196],[25,-0.7285],[31,-0.9501],[41,-1.0831],[48,-0.2885],[54,0.8726],[59,1.3951],[64,1.5692],[70,0.8936],[76,0.1627],[82,-0.2325],[90,-0.2325],[101,0.185],[109,0.8807],[115,1.4753],[119,1.6735],[126,1.2904],[133,0.3916],[139,-0.026],[145,-0.285],[152,-0.5488],[159,-0.7801],[166,-0.9188],[172,-0.3541],[178,0.7193],[183,1.8805],[188,2.5083],[193,2.2603],[198,1.7776],[205,0.8765],[211,0.5733],[215,0.4983],[219,0.8914],[223,1.3694],[229,1.9364],[236,3.2359],[242,4.0157],[248,4.4271],[267,4.4271],[276,2.7728],[283,-0.388],[289,-3.4407],[295,-5.6691],[303,-7.0062],[312,-7.375],[318,-8.3113],[327,-11.1201],[339,-2.1149],[351,0]]
       }
       ring_metacarpal_r_ctrl.Scale.X {type=MovieSceneScriptingFloatChannel, default=1, const [1,55]}
       ring_metacarpal_r_ctrl.Scale.Y {type=MovieSceneScriptingFloatChannel, default=1, const [1,55]}
       ring_metacarpal_r_ctrl.Scale.Z {type=MovieSceneScriptingFloatChannel, default=1, const [1,55]}
       ring_01_r_ctrl.Location.X {type=MovieSceneScriptingFloatChannel, default=0, const [0,55]}
       ring_01_r_ctrl.Location.Y {type=MovieSceneScriptingFloatChannel, default=0, const [0,55]}
       ring_01_r_ctrl.Location.Z {type=MovieSceneScriptingFloatChannel, default=0, const [0,55]}
       ring_01_r_ctrl.Rotation.X {type=MovieSceneScriptingFloatChannel, default=0, const [-8.5,55]}
       ring_01_r_ctrl.Rotation.Y {type=MovieSceneScriptingFloatChannel, default=0, const [6.5,55]}
       ring_01_r_ctrl.Rotation.Z {
        type=MovieSceneScriptingFloatChannel
        default=0
        thinned_from=55
        keys [[0,-19.5],[3,-19.3198],[6,-19.0946],[9,-19.0195],[19,-19.7196],[25,-20.2285],[31,-20.4501],[41,-20.5831],[48,-19.7885],[54,-18.6274],[59,-18.1049],[64,-17.9308],[70,-18.6064],[76,-19.3374],[82,-19.7325],[90,-19.7325],[101,-19.315],[109,-18.6193],[115,-18.0247],[119,-17.8265],[126,-18.2096],[133,-19.1084],[139,-19.526],[152,-20.0488],[159,-20.2801],[166,-20.4188],[172,-19.8541],[178,-18.7807],[183,-17.6195],[188,-16.9917],[193,-17.2397],[198,-17.7224],[205,-18.6235],[211,-18.9267],[215,-19.0017],[219,-18.6086],[223,-18.1306],[229,-17.5636],[236,-16.2641],[242,-15.4843],[248,-15.0729],[267,-15.0729],[276,-16.7272],[283,-19.888],[289,-22.9407],[295,-25.1691],[303,-26.5062],[312,-26.875],[318,-27.8113],[327,-30.6201],[339,-21.6149],[351,-19.5]]
       }
       ring_01_r_ctrl.Scale.X {type=MovieSceneScriptingFloatChannel, default=1, const [1,55]}
       ring_01_r_ctrl.Scale.Y {type=MovieSceneScriptingFloatChannel, default=1, const [1,55]}
       ring_01_r_ctrl.Scale.Z {type=MovieSceneScriptingFloatChannel, default=1, const [1,55]}
       ring_02_r_ctrl.Location.X {type=MovieSceneScriptingFloatChannel, default=0, const [0,55]}
       ring_02_r_ctrl.Location.Y {type=MovieSceneScriptingFloatChannel, default=0, const [0,55]}
       ring_02_r_ctrl.Location.Z {type=MovieSceneScriptingFloatChannel, default=0, const [0,55]}
       ring_02_r_ctrl.Rotation.X {type=MovieSceneScriptingFloatChannel, default=0, const [0,55]}
       ring_02_r_ctrl.Rotation.Y {type=MovieSceneScriptingFloatChannel, default=0, const [0,55]}
       ring_02_r_ctrl.Rotation.Z {
        type=MovieSceneScriptingFloatChannel
        default=0
        thinned_from=55
        keys [[0,-11.5],[3,-11.3198],[6,-11.0946],[9,-11.0195],[19,-11.7196],[25,-12.2285],[31,-12.4501],[41,-12.5831],[48,-11.7885],[54,-10.6274],[59,-10.1049],[64,-9.9308],[70,-10.6064],[76,-11.3373],[82,-11.7325],[90,-11.7325],[101,-11.315],[109,-10.6193],[115,-10.0247],[119,-9.8265],[126,-10.2096],[133,-11.1084],[139,-11.526],[145,-11.785],[152,-12.0488],[159,-12.2801],[166,-12.4188],[172,-11.8541],[178,-10.7807],[183,-9.6195],[188,-8.9917],[193,-9.2397],[198,-9.7224],[205,-10.6235],[211,-10.9267],[215,-11.0017],[219,-10.6086],[223,-10.1306],[229,-9.5636],[236,-8.2641],[242,-7.4843],[248,-7.0729],[267,-7.0729],[276,-8.7272],[283,-11.888],[289,-14.9407],[295,-17.1691],[303,-18.5062],[312,-18.875],[318,-19.8113],[327,-22.6201],[339,-13.6149],[351,-11.5]]
       }
       ring_02_r_ctrl.Scale.X {type=MovieSceneScriptingFloatChannel, default=1, const [1,55]}
       ring_02_r_ctrl.Scale.Y {type=MovieSceneScriptingFloatChannel, default=1, const [1,55]}
       ring_02_r_ctrl.Scale.Z {type=MovieSceneScriptingFloatChannel, default=1, const [1,55]}
       ring_03_r_ctrl.Location.X {type=MovieSceneScriptingFloatChannel, default=0, const [0,55]}
       ring_03_r_ctrl.Location.Y {type=MovieSceneScriptingFloatChannel, default=0, const [0,55]}
       ring_03_r_ctrl.Location.Z {type=MovieSceneScriptingFloatChannel, default=0, const [0,55]}
       ring_03_r_ctrl.Rotation.X {type=MovieSceneScriptingFloatChannel, default=0, const [0,55]}
       ring_03_r_ctrl.Rotation.Y {type=MovieSceneScriptingFloatChannel, default=0, const [0,55]}
       ring_03_r_ctrl.Rotation.Z {
        type=MovieSceneScriptingFloatChannel
        default=0
        thinned_from=55
        keys [[0,-5],[3,-4.8198],[6,-4.5946],[9,-4.5195],[19,-5.2196],[25,-5.7285],[31,-5.9501],[41,-6.0831],[48,-5.2885],[54,-4.1274],[59,-3.6049],[64,-3.4308],[70,-4.1064],[76,-4.8373],[82,-5.2325],[90,-5.2325],[101,-4.815],[109,-4.1193],[115,-3.5247],[119,-3.3265],[126,-3.7096],[133,-4.6084],[139,-5.026],[145,-5.285],[152,-5.5488],[159,-5.7801],[166,-5.9188],[172,-5.3541],[178,-4.2807],[183,-3.1195],[188,-2.4917],[193,-2.7397],[198,-3.2224],[205,-4.1235],[211,-4.4267],[215,-4.5017],[219,-4.1086],[223,-3.6306],[229,-3.0636],[236,-1.7641],[242,-0.9843],[248,-0.5729],[267,-0.5729],[276,-2.2272],[283,-5.388],[289,-8.4407],[295,-10.6691],[303,-12.0062],[312,-12.375],[318,-13.3113],[327,-16.1201],[339,-7.1149],[351,-5]]
       }
       ring_03_r_ctrl.Scale.X {type=MovieSceneScriptingFloatChannel, default=1, const [1,55]}
       ring_03_r_ctrl.Scale.Y {type=MovieSceneScriptingFloatChannel, default=1, const [1,55]}
       ring_03_r_ctrl.Scale.Z {type=MovieSceneScriptingFloatChannel, default=1, const [1,55]}
       pinky_metacarpal_r_ctrl.Location.X {type=MovieSceneScriptingFloatChannel, default=0, const [0,55]}
       pinky_metacarpal_r_ctrl.Location.Y {type=MovieSceneScriptingFloatChannel, default=0, const [0,55]}
       pinky_metacarpal_r_ctrl.Location.Z {type=MovieSceneScriptingFloatChannel, default=0, const [0,55]}
       pinky_metacarpal_r_ctrl.Rotation.X {type=MovieSceneScriptingFloatChannel, default=0, const [0,55]}
       pinky_metacarpal_r_ctrl.Rotation.Y {
        type=MovieSceneScriptingFloatChannel
        default=0
        thinned_from=55
        keys [[0,7.5],[267,7.5],[276,6.8692],[283,5.6641],[289,4.5001],[295,3.6504],[303,3.1406],[312,3],[319,3.1579],[328,4.4876],[337,6.7038],[351,7.5]]
       }
       pinky_metacarpal_r_ctrl.Rotation.Z {
        type=MovieSceneScriptingFloatChannel
        default=0
        thinned_from=55
        keys [[0,-5.5],[3,-5.3198],[6,-5.0946],[9,-5.0195],[19,-5.7196],[25,-6.2285],[31,-6.4501],[41,-6.5831],[48,-5.7885],[54,-4.6274],[59,-4.1049],[64,-3.9308],[70,-4.6064],[76,-5.3373],[82,-5.7325],[90,-5.7325],[101,-5.315],[109,-4.6193],[115,-4.0247],[119,-3.8265],[126,-4.2096],[133,-5.1084],[139,-5.526],[145,-5.785],[152,-6.0488],[159,-6.2801],[166,-6.4188],[172,-5.8541],[178,-4.7807],[183,-3.6195],[188,-2.9917],[193,-3.2397],[198,-3.7224],[205,-4.6235],[211,-4.9267],[215,-5.0017],[219,-4.6086],[223,-4.1306],[229,-3.5636],[236,-2.2641],[242,-1.4843],[248,-1.0729],[267,-1.0729],[276,-2.7272],[283,-5.888],[289,-8.9407],[295,-11.1691],[303,-12.5062],[312,-12.875],[318,-13.8113],[327,-16.6201],[339,-7.6149],[351,-5.5]]
       }
       pinky_metacarpal_r_ctrl.Scale.X {type=MovieSceneScriptingFloatChannel, default=1, const [1,55]}
       pinky_metacarpal_r_ctrl.Scale.Y {type=MovieSceneScriptingFloatChannel, default=1, const [1,55]}
       pinky_metacarpal_r_ctrl.Scale.Z {type=MovieSceneScriptingFloatChannel, default=1, const [1,55]}
       pinky_01_r_ctrl.Location.X {type=MovieSceneScriptingFloatChannel, default=0, const [0,55]}
       pinky_01_r_ctrl.Location.Y {type=MovieSceneScriptingFloatChannel, default=0, const [0,55]}
       pinky_01_r_ctrl.Location.Z {type=MovieSceneScriptingFloatChannel, default=0, const [0,55]}
       pinky_01_r_ctrl.Rotation.X {type=MovieSceneScriptingFloatChannel, default=0, const [-16.5,55]}
       pinky_01_r_ctrl.Rotation.Y {type=MovieSceneScriptingFloatChannel, default=0, const [4,55]}
       pinky_01_r_ctrl.Rotation.Z {
        type=MovieSceneScriptingFloatChannel
        default=0
        thinned_from=55
        keys [[0,-31.5],[6,-31.0946],[9,-31.0195],[19,-31.7196],[25,-32.2285],[31,-32.4501],[41,-32.5831],[48,-31.7885],[54,-30.6274],[59,-30.1049],[64,-29.9308],[76,-31.3373],[82,-31.7325],[90,-31.7325],[101,-31.315],[109,-30.6193],[115,-30.0247],[119,-29.8265],[126,-30.2096],[133,-31.1084],[139,-31.526],[152,-32.0488],[159,-32.2801],[166,-32.4188],[172,-31.8541],[178,-30.7807],[183,-29.6195],[188,-28.9917],[193,-29.2397],[198,-29.7224],[205,-30.6235],[211,-30.9267],[215,-31.0017],[219,-30.6086],[223,-30.1306],[229,-29.5636],[236,-28.2641],[242,-27.4843],[248,-27.0729],[267,-27.0729],[276,-28.7272],[283,-31.888],[289,-34.9407],[295,-37.1691],[303,-38.5062],[312,-38.875],[318,-39.8113],[327,-42.6201],[339,-33.6149],[351,-31.5]]
       }
       pinky_01_r_ctrl.Scale.X {type=MovieSceneScriptingFloatChannel, default=1, const [1,55]}
       pinky_01_r_ctrl.Scale.Y {type=MovieSceneScriptingFloatChannel, default=1, const [1,55]}
       pinky_01_r_ctrl.Scale.Z {type=MovieSceneScriptingFloatChannel, default=1, const [1,55]}
       pinky_02_r_ctrl.Location.X {type=MovieSceneScriptingFloatChannel, default=0, const [0,55]}
       pinky_02_r_ctrl.Location.Y {type=MovieSceneScriptingFloatChannel, default=0, const [0,55]}
       pinky_02_r_ctrl.Location.Z {type=MovieSceneScriptingFloatChannel, default=0, const [0,55]}
       pinky_02_r_ctrl.Rotation.X {type=MovieSceneScriptingFloatChannel, default=0, const [0,55]}
       pinky_02_r_ctrl.Rotation.Y {type=MovieSceneScriptingFloatChannel, default=0, const [0,55]}
       pinky_02_r_ctrl.Rotation.Z {
        type=MovieSceneScriptingFloatChannel
        default=0
        thinned_from=55
        keys [[0,-14.5],[3,-14.3198],[6,-14.0946],[9,-14.0195],[19,-14.7196],[25,-15.2285],[31,-15.4501],[41,-15.5831],[48,-14.7885],[54,-13.6274],[59,-13.1049],[64,-12.9308],[70,-13.6064],[76,-14.3374],[82,-14.7325],[90,-14.7325],[101,-14.315],[109,-13.6193],[115,-13.0247],[119,-12.8265],[126,-13.2096],[133,-14.1084],[139,-14.526],[145,-14.785],[152,-15.0488],[159,-15.2801],[166,-15.4188],[172,-14.8541],[178,-13.7807],[183,-12.6195],[188,-11.9917],[193,-12.2397],[198,-12.7224],[205,-13.6235],[211,-13.9267],[215,-14.0017],[219,-13.6086],[223,-13.1306],[229,-12.5636],[236,-11.2641],[242,-10.4843],[248,-10.0729],[267,-10.0729],[276,-11.7272],[283,-14.888],[289,-17.9407],[295,-20.1691],[303,-21.5062],[312,-21.875],[318,-22.8113],[327,-25.6201],[339,-16.6149],[351,-14.5]]
       }
       pinky_02_r_ctrl.Scale.X {type=MovieSceneScriptingFloatChannel, default=1, const [1,55]}
       pinky_02_r_ctrl.Scale.Y {type=MovieSceneScriptingFloatChannel, default=1, const [1,55]}
       pinky_02_r_ctrl.Scale.Z {type=MovieSceneScriptingFloatChannel, default=1, const [1,55]}
       pinky_03_r_ctrl.Location.X {type=MovieSceneScriptingFloatChannel, default=0, const [0,55]}
       pinky_03_r_ctrl.Location.Y {type=MovieSceneScriptingFloatChannel, default=0, const [0,55]}
       pinky_03_r_ctrl.Location.Z {type=MovieSceneScriptingFloatChannel, default=0, const [0,55]}
       pinky_03_r_ctrl.Rotation.X {type=MovieSceneScriptingFloatChannel, default=0, const [0,55]}
       pinky_03_r_ctrl.Rotation.Y {type=MovieSceneScriptingFloatChannel, default=0, const [0,55]}
       pinky_03_r_ctrl.Rotation.Z {
        type=MovieSceneScriptingFloatChannel
        default=0
        thinned_from=55
        keys [[0,-7],[3,-6.8198],[6,-6.5946],[9,-6.5195],[19,-7.2196],[25,-7.7285],[31,-7.9501],[41,-8.0831],[48,-7.2885],[54,-6.1274],[59,-5.6049],[64,-5.4308],[70,-6.1064],[76,-6.8373],[82,-7.2325],[90,-7.2325],[101,-6.815],[109,-6.1193],[115,-5.5247],[119,-5.3265],[126,-5.7096],[133,-6.6084],[139,-7.026],[145,-7.285],[152,-7.5488],[159,-7.7801],[166,-7.9188],[172,-7.3541],[178,-6.2807],[183,-5.1195],[188,-4.4917],[193,-4.7397],[198,-5.2224],[205,-6.1235],[211,-6.4267],[215,-6.5017],[219,-6.1086],[223,-5.6306],[229,-5.0636],[236,-3.7641],[242,-2.9843],[248,-2.5729],[267,-2.5729],[276,-4.2272],[283,-7.388],[289,-10.4407],[295,-12.6691],[303,-14.0062],[312,-14.375],[318,-15.3113],[327,-18.1201],[339,-9.1149],[351,-7]]
       }
       pinky_03_r_ctrl.Scale.X {type=MovieSceneScriptingFloatChannel, default=1, const [1,55]}
       pinky_03_r_ctrl.Scale.Y {type=MovieSceneScriptingFloatChannel, default=1, const [1,55]}
       pinky_03_r_ctrl.Scale.Z {type=MovieSceneScriptingFloatChannel, default=1, const [1,55]}
       hand_r_ik_ctrl.Location.X {
        type=MovieSceneScriptingFloatChannel
        default=0
        thinned_from=58
        keys [[0,-6.7969],[9,-6.7484],[13,-6.7421],[19,-6.8736],[25,-7.0927],[31,-7.2242],[41,-7.303],[48,-7.1201],[54,-6.8151],[59,-6.7135],[64,-6.6525],[76,-6.8965],[82,-7.1506],[90,-8.0201],[101,-8.31],[109,-7.6396],[115,-8.0604],[119,-8.1656],[126,-8.2006],[145,-8.1734],[152,-7.6722],[159,-7.7575],[166,-8.0132],[178,-7.9612],[183,-7.881],[193,-7.7752],[205,-7.7559],[212,-7.4884],[216,-7.0425],[220,-6.775],[225,-6.6747],[232,-6.6145],[239,-6.5169],[244,-6.3543],[249,-6.0832],[255,-5.9815],[263,-5.9205],[271,-6.0849],[277,-6.5778],[283,-6.7148],[292,-6.7764],[301,-6.7969],[351,-6.7969]]
       }
       hand_r_ik_ctrl.Location.Y {
        type=MovieSceneScriptingFloatChannel
        default=0
        thinned_from=58
        keys [[0,10.4715],[9,10.58],[13,10.5939],[31,10.5477],[41,10.5402],[48,10.6941],[54,10.9507],[59,11.0362],[64,11.0875],[76,10.8822],[82,10.6684],[101,10.6486],[109,10.4249],[115,10.2808],[119,10.2447],[145,10.2438],[152,9.6567],[159,9.7046],[166,9.8482],[172,9.7653],[178,9.7157],[183,9.511],[188,9.3883],[193,9.241],[198,9.2041],[205,9.1918],[212,9.0462],[216,8.8035],[220,8.6578],[225,8.6032],[232,8.5704],[239,8.6568],[244,8.8007],[249,9.0405],[255,9.1304],[263,9.1844],[271,9.4257],[277,10.1497],[283,10.3508],[292,10.4413],[301,10.4715],[351,10.4715]]
       }
       hand_r_ik_ctrl.Location.Z {
        type=MovieSceneScriptingFloatChannel
        default=0
        thinned_from=58
        keys [[0,-12.3083],[9,-12.4464],[13,-12.4642],[19,-12.37],[25,-12.2131],[31,-12.1189],[41,-12.0624],[64,-12.1502],[82,-12.083],[90,-13.0891],[101,-13.4245],[109,-12.5664],[115,-13.1964],[119,-13.3539],[126,-13.4064],[145,-13.4196],[152,-12.5802],[159,-12.7083],[166,-13.0926],[172,-13.0241],[178,-12.9829],[183,-12.8135],[193,-12.5899],[198,-12.5594],[205,-12.5493],[212,-12.0858],[216,-11.3133],[220,-10.8498],[225,-10.676],[232,-10.5717],[239,-10.6158],[244,-10.6892],[249,-10.8117],[255,-10.8576],[263,-10.8851],[271,-11.152],[277,-11.9525],[283,-12.1749],[292,-12.2749],[301,-12.3083],[351,-12.3083]]
       }
       hand_r_ik_ctrl.Rotation.X {
        type=MovieSceneScriptingFloatChannel
        default=0
        thinned_from=58
        keys [[0,-32.9875],[13,-32.8816],[19,-32.7818],[25,-32.6156],[31,-32.5158],[41,-32.456],[48,-32.7775],[54,-33.3135],[59,-33.4921],[64,-33.5993],[76,-33.1706],[82,-32.7239],[90,-34.0255],[101,-34.4594],[109,-33.1591],[115,-34.0381],[119,-34.2579],[126,-34.3311],[145,-34.3311],[152,-34.0149],[166,-33.9504],[178,-33.8234],[183,-33.6274],[193,-33.3687],[205,-33.3217],[212,-32.6609],[216,-31.5596],[220,-30.8988],[225,-30.651],[232,-30.5023],[239,-30.4457],[244,-30.3514],[249,-30.1941],[263,-30.0998],[271,-30.6412],[277,-32.2656],[283,-32.7168],[292,-32.9198],[351,-32.9875]]
       }
       hand_r_ik_ctrl.Rotation.Y {
        type=MovieSceneScriptingFloatChannel
        default=0
        thinned_from=58
        keys [[0,-16.839],[3,-16.89],[9,-17.032],[13,-17.0568],[19,-17.1448],[25,-17.2914],[31,-17.3794],[41,-17.4321],[48,-17.2949],[54,-17.0663],[64,-16.9443],[76,-17.1272],[82,-17.3178],[90,-18.9857],[101,-19.5417],[109,-17.979],[115,-18.6626],[119,-18.8335],[126,-18.8905],[145,-18.8905],[152,-17.3944],[159,-17.5943],[166,-18.194],[172,-18.0489],[178,-17.9618],[183,-17.6033],[188,-17.3881],[193,-17.13],[198,-17.0654],[205,-17.0439],[212,-16.7404],[216,-16.2346],[220,-15.9311],[225,-15.8173],[239,-15.6512],[244,-15.4883],[249,-15.2168],[255,-15.115],[263,-15.0539],[271,-15.3886],[277,-16.3927],[283,-16.6716],[292,-16.7971],[351,-16.839]]
       }
       hand_r_ik_ctrl.Rotation.Z {
        type=MovieSceneScriptingFloatChannel
        default=0
        thinned_from=58
        keys [[0,-22.6177],[9,-22.4288],[13,-22.4045],[19,-22.2637],[25,-22.0291],[31,-21.8883],[41,-21.8039],[76,-21.7694],[82,-21.791],[90,-22.2042],[101,-22.3419],[109,-22.2279],[115,-22.7578],[119,-22.8903],[152,-22.9175],[159,-22.9619],[166,-23.0949],[178,-23.1483],[193,-23.3399],[205,-23.3597],[212,-22.9475],[216,-22.2605],[220,-21.8483],[225,-21.6938],[239,-21.5025],[244,-21.3382],[249,-21.0644],[255,-20.9617],[263,-20.9001],[271,-21.2222],[277,-22.1883],[283,-22.4567],[292,-22.5774],[351,-22.6177]]
       }
       hand_r_ik_ctrl.Scale.X {type=MovieSceneScriptingFloatChannel, default=1, const [1,58]}
       hand_r_ik_ctrl.Scale.Y {type=MovieSceneScriptingFloatChannel, default=1, const [1,58]}
       hand_r_ik_ctrl.Scale.Z {type=MovieSceneScriptingFloatChannel, default=1, const [1,58]}
       Softness_2 {type=MovieSceneScriptingFloatChannel, default=0, const [0,58]}
       hand_r_ikorient_ctrl.Rotation.Roll {type=MovieSceneScriptingFloatChannel, default=0, const [0,58]}
       hand_r_ikorient_ctrl.Rotation.Pitch {type=MovieSceneScriptingFloatChannel, default=0, const [0,58]}
       hand_r_ikorient_ctrl.Rotation.Yaw {type=MovieSceneScriptingFloatChannel, default=0, const [0,58]}
       arm_r_pv_ik_ctrl.Location.X {
        type=MovieSceneScriptingFloatChannel
        default=0
        thinned_from=58
        keys [[0,-32.6878],[41,-32.6634],[48,-32.7893],[54,-32.9992],[64,-33.1111],[76,-32.9432],[82,-32.7683],[90,-33.6835],[101,-33.9885],[109,-33.0447],[115,-33.5213],[119,-33.6405],[145,-33.6802],[152,-32.8808],[159,-32.9707],[166,-33.2404],[178,-33.0938],[183,-32.8674],[193,-32.5686],[205,-32.5143],[212,-31.9456],[216,-30.9979],[220,-30.4292],[225,-30.216],[232,-30.0881],[263,-29.9492],[271,-30.4627],[277,-32.0032],[283,-32.4311],[292,-32.6237],[351,-32.6878]]
       }
       arm_r_pv_ik_ctrl.Location.Y {
        type=MovieSceneScriptingFloatChannel
        default=0
        thinned_from=58
        keys [[0,-3.2719],[3,-3.292],[6,-3.3255],[9,-3.3479],[13,-3.3577],[19,-3.3407],[25,-3.3124],[31,-3.2954],[41,-3.2852],[48,-3.2072],[54,-3.0772],[59,-3.0339],[64,-3.0079],[76,-3.1119],[82,-3.2202],[90,-3.4867],[101,-3.5756],[109,-3.4486],[115,-3.6529],[119,-3.704],[126,-3.721],[145,-3.721],[152,-3.4442],[159,-3.5125],[166,-3.7174],[205,-3.7113],[212,-3.4895],[216,-3.1199],[220,-2.8981],[225,-2.815],[232,-2.7651],[239,-2.8266],[244,-2.9291],[249,-3.1],[255,-3.1641],[263,-3.2025],[271,-3.2155],[277,-3.2545],[283,-3.2654],[351,-3.2719]]
       }
       arm_r_pv_ik_ctrl.Location.Z {
        type=MovieSceneScriptingFloatChannel
        default=0
        thinned_from=58
        keys [[0,-4.4751],[13,-4.4835],[19,-4.2297],[25,-3.8068],[31,-3.553],[41,-3.4008],[48,-3.7221],[54,-4.2576],[59,-4.4361],[64,-4.5433],[70,-4.3228],[76,-4.1148],[90,-3.0741],[101,-2.876],[109,-3.2886],[115,-3.1484],[119,-3.1134],[126,-3.1017],[133,-3.1105],[145,-3.1339],[152,-3.2054],[159,-3.1669],[166,-3.0515],[172,-3.0332],[178,-3.0222],[183,-2.9771],[193,-2.9175],[198,-2.9094],[205,-2.9066],[212,-2.8383],[216,-2.7243],[220,-2.656],[225,-2.6303],[232,-2.615],[239,-2.7263],[244,-2.9118],[249,-3.2209],[255,-3.3369],[263,-3.4064],[271,-3.6068],[277,-4.2079],[283,-4.3749],[292,-4.4501],[301,-4.4751],[351,-4.4751]]
       }
       arm_r_pv_ik_ctrl.Rotation.X {
        type=MovieSceneScriptingFloatChannel
        default=0
        thinned_from=58
        keys [[0,-2.3009],[3,-2.25],[6,-2.1652],[9,-2.1085],[13,-2.0837],[19,-1.9136],[25,-1.6299],[31,-1.4598],[41,-1.3577],[48,-1.5667],[54,-1.9151],[59,-2.0312],[64,-2.1009],[70,-1.9575],[76,-1.8222],[82,-1.5319],[90,-0.915],[101,-0.7094],[109,-1.2419],[115,-1.0972],[119,-1.061],[126,-1.049],[145,-1.049],[152,-1.1981],[159,-1.1407],[166,-0.9686],[172,-0.9519],[178,-0.9419],[183,-0.9005],[188,-0.8757],[193,-0.8459],[198,-0.8385],[205,-0.836],[212,-0.8732],[216,-0.9352],[220,-0.9723],[225,-0.9863],[232,-0.9947],[239,-0.9645],[244,-0.9142],[249,-0.8304],[255,-0.7989],[263,-0.7801],[271,-1.0652],[277,-1.9207],[283,-2.1583],[292,-2.2653],[301,-2.3009],[351,-2.3009]]
       }
       arm_r_pv_ik_ctrl.Rotation.Y {
        type=MovieSceneScriptingFloatChannel
        default=0
        thinned_from=58
        keys [[0,-3.2897],[3,-3.276],[6,-3.2531],[9,-3.2378],[13,-3.2311],[19,-3.2014],[25,-3.1519],[31,-3.1222],[41,-3.1044],[48,-3.3589],[54,-3.7832],[59,-3.9247],[64,-4.0095],[70,-3.8348],[76,-3.6701],[82,-3.3165],[90,-3.0666],[101,-2.9833],[109,-3.0504],[115,-2.983],[119,-2.9662],[145,-2.9606],[152,-3.485],[159,-3.3608],[166,-2.9881],[178,-2.9973],[193,-3.0303],[205,-3.0337],[212,-2.9899],[216,-2.9169],[220,-2.8731],[225,-2.8566],[232,-2.8468],[239,-2.9086],[244,-3.0115],[249,-3.1831],[255,-3.2474],[263,-3.286],[351,-3.2897]]
       }
       arm_r_pv_ik_ctrl.Rotation.Z {
        type=MovieSceneScriptingFloatChannel
        default=0
        thinned_from=58
        keys [[0,-48.1166],[19,-47.9918],[31,-47.7917],[41,-47.7467],[48,-47.8733],[54,-48.0842],[64,-48.1968],[76,-48.028],[82,-47.8522],[90,-50.2079],[101,-50.9931],[109,-48.9108],[115,-50.3659],[119,-50.7297],[126,-50.851],[145,-50.851],[152,-49.825],[159,-49.9541],[166,-50.3413],[178,-50.1714],[183,-49.909],[193,-49.5626],[205,-49.4997],[212,-48.5254],[216,-46.9016],[220,-45.9273],[225,-45.5619],[239,-45.2004],[244,-44.9632],[249,-44.5679],[255,-44.4197],[263,-44.3307],[271,-45.0406],[277,-47.1702],[283,-47.7617],[292,-48.0279],[351,-48.1166]]
       }
       arm_r_pv_ik_ctrl.Scale.X {type=MovieSceneScriptingFloatChannel, default=1, const [1,58]}
       arm_r_pv_ik_ctrl.Scale.Y {type=MovieSceneScriptingFloatChannel, default=1, const [1,58]}
       arm_r_pv_ik_ctrl.Scale.Z {type=MovieSceneScriptingFloatChannel, default=1, const [1,58]}
       body_ctrl.Location.X {
        type=MovieSceneScriptingFloatChannel
        default=0
        keys [[0,2.5],[3,2.5462],[6,2.5776],[9,2.589],[14,2.5713],[19,2.5006],[25,2.3781],[31,2.25],[41,2.15],[48,2.2],[54,2.3],[59,2.4],[64,2.425],[70,2.35],[76,2.15],[82,1.7],[90,1.15],[101,0.85],[109,0.725],[119,0.625],[126,0.5875],[139,0.55],[152,0.5375],[159,0.575],[166,0.725],[172,1.0794],[178,1.732],[183,2.2004],[188,2.4946],[194,2.6552],[204,2.7859],[216,2.8917],[225,3.0417],[232,3.1917],[239,3.2696],[244,3.2196],[249,3.1065],[277,2.0653],[283,1.87],[292,1.6861],[303,1.6172],[311,1.6622],[343,2.3418],[351,2.5]]
       }
       body_ctrl.Location.Y {
        type=MovieSceneScriptingFloatChannel
        default=0
        thinned_from=55
        keys [[0,9.6121],[19,9.6122],[48,9.5856],[64,9.6487],[76,9.6584],[90,9.5991],[109,9.5867],[133,9.6376],[139,9.6291],[152,9.5637],[166,9.5383],[178,9.6114],[183,9.6268],[198,9.5505],[212,9.5319],[225,9.5986],[239,9.6195],[255,9.5492],[271,9.5516],[284,9.6328],[307,9.6445],[319,9.5918],[351,9.6121]]
       }
       body_ctrl.Location.Z {
        type=MovieSceneScriptingFloatChannel
        default=0
        keys [[0,0.6925],[3,0.7243],[6,0.7426],[9,0.7333],[13,0.7053],[19,0.68],[25,0.7742],[31,0.8742],[40,0.9117],[48,0.7867],[54,0.6675],[59,0.555],[64,0.5175],[70,0.555],[76,0.655],[82,0.8175],[90,0.955],[101,1.0175],[109,0.9425],[115,0.7675],[119,0.6412],[126,0.5925],[133,0.655],[139,0.8175],[145,0.9425],[152,0.9925],[159,0.9425],[166,0.7925],[172,0.655],[178,0.575],[183,0.6175],[188,0.7315],[193,0.855],[198,0.9175],[205,0.9425],[212,0.88],[216,0.7675],[220,0.68],[225,0.63],[232,0.605],[239,0.6425],[244,0.73],[249,0.855],[255,0.9425],[263,0.98],[271,0.8925],[277,0.705],[283,0.5925],[291,0.53],[300,0.4899],[313,0.4695],[322,0.4871],[334,0.5521],[343,0.6235],[351,0.6925]]
       }
       body_ctrl.Rotation.X {
        type=MovieSceneScriptingFloatChannel
        default=0
        thinned_from=40
        keys [[0,-17.7468],[9,-17.8951],[13,-17.9268],[19,-17.9046],[25,-17.7949],[31,-17.6396],[38,-17.544],[46,-17.5128],[54,-17.7508],[64,-18.1712],[70,-18.2593],[76,-18.2893],[98,-17.3661],[109,-17.2204],[120,-17.603],[129,-17.9965],[135,-18.1406],[143,-17.9112],[152,-17.5025],[158,-17.2923],[166,-17.2118],[175,-17.5949],[181,-17.8083],[185,-17.8473],[191,-17.6843],[198,-17.2612],[205,-17.0049],[212,-16.919],[216,-17.043],[229,-17.6382],[237,-17.7985],[258,-17.3465],[279,-16.6509],[300,-16.1273],[310,-16.0599],[321,-16.5304],[335,-17.4248],[351,-17.7468]]
       }
       body_ctrl.Rotation.Y {
        type=MovieSceneScriptingFloatChannel
        default=0
        thinned_from=55
        keys [[0,-9.743],[19,-9.7519],[59,-9.731],[76,-9.756],[82,-9.7464],[90,-9.7103],[102,-9.6313],[115,-9.6017],[139,-9.6037],[152,-9.4988],[178,-9.4259],[183,-9.452],[188,-9.5601],[198,-9.92],[205,-10.0823],[212,-10.1684],[216,-10.1734],[232,-10.0942],[239,-10.0809],[249,-10.1176],[255,-10.1646],[263,-10.1932],[271,-10.1248],[283,-9.8422],[292,-9.7284],[300,-9.673],[319,-9.641],[351,-9.743]]
       }
       body_ctrl.Rotation.Z {
        type=MovieSceneScriptingFloatChannel
        default=0
        thinned_from=52
        keys [[0,4.5],[9,4.5],[13,4.4656],[19,4.2535],[25,3.9749],[31,3.7933],[39,3.6699],[48,3.6199],[54,3.6737],[59,3.8542],[64,4.1989],[70,5.0207],[76,6.0721],[82,7.1924],[90,8.2285],[101,8.6095],[109,8.5441],[115,8.2706],[126,7.5413],[133,7.323],[139,7.2862],[145,7.4019],[152,7.5985],[159,7.7146],[166,7.7621],[172,7.6082],[178,6.7092],[188,3.1512],[193,2.2712],[198,1.8678],[205,1.721],[212,1.7765],[216,1.8584],[220,1.9968],[225,2.1809],[232,2.3011],[239,2.2259],[244,2.1343],[249,2.0737],[255,2.2217],[271,4.8926],[279,5.5126],[288,5.721],[300,5.8308],[307,5.7811],[313,5.6342],[333,4.8194],[343,4.5601],[351,4.5]]
       }
       body_ctrl.Scale.X {type=MovieSceneScriptingFloatChannel, default=1, const [1,58]}
       body_ctrl.Scale.Y {type=MovieSceneScriptingFloatChannel, default=1, const [1,58]}
       body_ctrl.Scale.Z {type=MovieSceneScriptingFloatChannel, default=1, const [1,58]}
       hips_ctrl.Location.X {type=MovieSceneScriptingFloatChannel, default=0, const [0.588,58]}
       hips_ctrl.Location.Y {type=MovieSceneScriptingFloatChannel, default=0, const [-0.6148,58]}
       hips_ctrl.Location.Z {type=MovieSceneScriptingFloatChannel, default=0, const [0.3035,58]}
       hips_ctrl.Rotation.X {type=MovieSceneScriptingFloatChannel, default=0, const [23.5,41]}
       hips_ctrl.Rotation.Y {type=MovieSceneScriptingFloatChannel, default=0, const [9.7177,59]}
       hips_ctrl.Rotation.Z {type=MovieSceneScriptingFloatChannel, default=0, const [0,58]}
       hips_ctrl.Scale.X {type=MovieSceneScriptingFloatChannel, default=1, const [1,58]}
       hips_ctrl.Scale.Y {type=MovieSceneScriptingFloatChannel, default=1, const [1,58]}
       hips_ctrl.Scale.Z {type=MovieSceneScriptingFloatChannel, default=1, const [1,58]}
       hips_tan_ctrl.Location.X {type=MovieSceneScriptingFloatChannel, default=0, thinned_from=58, keys [[0,1.2446],[351,1.2446]]}
       hips_tan_ctrl.Location.Y {type=MovieSceneScriptingFloatChannel, default=0.008, thinned_from=58, keys [[0,-1.6155],[351,-1.6155]]}
       hips_tan_ctrl.Location.Z {type=MovieSceneScriptingFloatChannel, default=-0.0073, thinned_from=58, keys [[0,-0.3011],[351,-0.3011]]}
       hips_tan_ctrl.Rotation.X {type=MovieSceneScriptingFloatChannel, default=0, const [0,58]}
       hips_tan_ctrl.Rotation.Y {type=MovieSceneScriptingFloatChannel, default=0, const [0,58]}
       hips_tan_ctrl.Rotation.Z {type=MovieSceneScriptingFloatChannel, default=0, const [0,58]}
       hips_tan_ctrl.Scale.X {type=MovieSceneScriptingFloatChannel, default=1, const [1,58]}
       hips_tan_ctrl.Scale.Y {type=MovieSceneScriptingFloatChannel, default=1, const [1,58]}
       hips_tan_ctrl.Scale.Z {type=MovieSceneScriptingFloatChannel, default=1, const [1,58]}
       spine_01_ctrl.Location.X {type=MovieSceneScriptingFloatChannel, default=0, const [0.0606,58]}
       spine_01_ctrl.Location.Y {type=MovieSceneScriptingFloatChannel, default=0, const [0.7467,58]}
       spine_01_ctrl.Location.Z {type=MovieSceneScriptingFloatChannel, default=0, const [-0.2281,58]}
       spine_01_ctrl.Rotation.X {type=MovieSceneScriptingFloatChannel, default=0, const [-3.0631,58]}
       spine_01_ctrl.Rotation.Y {type=MovieSceneScriptingFloatChannel, default=0, const [0.67,56]}
       spine_01_ctrl.Rotation.Z {type=MovieSceneScriptingFloatChannel, default=0, const [0,58]}
       spine_01_ctrl.Scale.X {type=MovieSceneScriptingFloatChannel, default=1, const [1,58]}
       spine_01_ctrl.Scale.Y {type=MovieSceneScriptingFloatChannel, default=1, const [1,58]}
       spine_01_ctrl.Scale.Z {type=MovieSceneScriptingFloatChannel, default=1, const [1,58]}
       spine_02_ctrl.Location.X {type=MovieSceneScriptingFloatChannel, default=0, const [0,58]}
       spine_02_ctrl.Location.Y {type=MovieSceneScriptingFloatChannel, default=0, const [0,58]}
       spine_02_ctrl.Location.Z {type=MovieSceneScriptingFloatChannel, default=0, const [0,58]}
       spine_02_ctrl.Rotation.X {
        type=MovieSceneScriptingFloatChannel
        default=0
        thinned_from=54
        keys [[0,15.5],[13,15.5],[19,15.4124],[31,15.0063],[41,14.8939],[48,14.9479],[54,15.1879],[59,15.53],[64,15.8301],[70,16.0221],[76,15.9321],[82,15.71],[90,15.284],[101,14.8279],[109,14.9119],[115,15.338],[119,15.74],[126,16.2201],[133,16.4422],[139,16.3282],[145,16.0881],[152,15.734],[159,15.5],[166,15.6121],[172,15.8962],[178,16.2924],[183,16.5685],[188,16.4404],[193,16.0633],[198,15.6514],[205,15.2914],[212,15.1932],[216,15.2995],[220,15.5],[225,15.8232],[232,16.175],[239,16.3182],[244,16.2036],[249,15.8845],[263,14.8732],[275,14.466],[292,14.2961],[307,14.2961],[313,14.3984],[333,15.3673],[343,15.5],[351.9987,15.5]]
       }
       spine_02_ctrl.Rotation.Y {type=MovieSceneScriptingFloatChannel, default=0, const [3,58]}
       spine_02_ctrl.Rotation.Z {type=MovieSceneScriptingFloatChannel, default=0, const [0,58]}
       spine_02_ctrl.Scale.X {type=MovieSceneScriptingFloatChannel, default=1, const [1,58]}
       spine_02_ctrl.Scale.Y {type=MovieSceneScriptingFloatChannel, default=1, const [1,58]}
       spine_02_ctrl.Scale.Z {type=MovieSceneScriptingFloatChannel, default=1, const [1,58]}
       spine_03_ctrl.Location.X {type=MovieSceneScriptingFloatChannel, default=0, const [0,58]}
       spine_03_ctrl.Location.Y {type=MovieSceneScriptingFloatChannel, default=0, const [0,58]}
       spine_03_ctrl.Location.Z {type=MovieSceneScriptingFloatChannel, default=0, const [0,58]}
       spine_03_ctrl.Rotation.X {
        type=MovieSceneScriptingFloatChannel
        default=0
        thinned_from=48
        keys [[0,5],[13,5],[19,4.7312],[25,4.25],[31,3.9],[41,3.7],[48,3.9],[54,4.5],[59,5.225],[64,5.75],[70,6],[76,6.075],[82,5.3739],[90,4.5239],[100,3.9989],[109,3.8489],[115,3.9489],[119,4.4],[126,5.375],[133,5.575],[139,4.9],[145,4.161],[152,3.7334],[159,3.5834],[166,3.7584],[172,4.475],[178,5.275],[183,5.65],[188,5.4],[198,4.0821],[205,3.5821],[211,3.4321],[216,3.6821],[225,5],[232,5.625],[239,5.875],[252,5.5],[263,4.0721],[280,0.778],[293,0.0382],[307,-0.1788],[316,0.3442],[331,3.9984],[338,4.6814],[351,5]]
       }
       spine_03_ctrl.Rotation.Y {
        type=MovieSceneScriptingFloatChannel
        default=0
        thinned_from=55
        keys [[0,3],[152,3],[159,3.0542],[166,3.1827],[183,3.5197],[193,3.563],[205,3.575],[263,3.575],[271,3.4672],[277,3.2398],[283,3.0917],[292,3.0135],[301,3],[351,3]]
       }
       spine_03_ctrl.Rotation.Z {
        type=MovieSceneScriptingFloatChannel
        default=0
        thinned_from=56
        keys [[0,13.5],[13,13.5],[19,13.2998],[25,12.9259],[31,12.7315],[39,12.6057],[59,12.4364],[70,12.2793],[76,12.5021],[81,14.1538],[88,19.8996],[94,22.6676],[107,24.0342],[115,23.9803],[133,23.5941],[145,23.4517],[172,23.3318],[178,22.5134],[183,19.7382],[188,15.3331],[193,12.1882],[198,10.7612],[205,9.9332],[212,9.6944],[216,9.7685],[225,10.2261],[232,10.4569],[239,10.5538],[244,10.577],[249,10.4737],[255,10.3066],[261,10.2297],[269,10.7711],[275,12.0585],[280,14.5495],[287,16.8999],[297,17.9633],[306,18.0612],[312,17.8908],[319,17.3199],[326,16.2214],[338,14.1224],[343,13.6762],[351,13.5]]
       }
       spine_03_ctrl.Scale.X {type=MovieSceneScriptingFloatChannel, default=1, const [1,58]}
       spine_03_ctrl.Scale.Y {type=MovieSceneScriptingFloatChannel, default=1, const [1,58]}
       spine_03_ctrl.Scale.Z {type=MovieSceneScriptingFloatChannel, default=1, const [1,58]}
       spine_04_ctrl.Location.X {type=MovieSceneScriptingFloatChannel, default=0, const [0,59]}
       spine_04_ctrl.Location.Y {type=MovieSceneScriptingFloatChannel, default=0, const [0,59]}
       spine_04_ctrl.Location.Z {
        type=MovieSceneScriptingFloatChannel
        default=0
        keys [[0,0],[3,-0.0098],[6,-0.0176],[9,-0.0228],[13,-0.0251],[19,-0.0132],[25,0.048],[31,0.084],[39,0.0973],[41,0.0977],[48,0.0638],[54,0.0222],[59,-0.005],[64,-0.0218],[70,-0.03],[76,-0.0161],[82,0.0219],[90,0.0588],[101,0.0696],[109,0.0499],[115,0.005],[119,-0.0197],[126,-0.03],[133,-0.0119],[139,0.0335],[145,0.08],[152,0.1009],[159,0.0908],[166,0.0652],[172,0.0414],[178,0.0301],[183,0.0444],[188,0.0683],[193,0.0842],[198,0.0921],[205,0.0957],[212,0.0832],[216,0.0458],[220,0.0065],[225,-0.0126],[232,-0.0211],[239,-0.0119],[244,0.0426],[249,0.0855],[255,0.1016],[263,0.1069],[271,0.0832],[277,0.0363],[283,-0.0004],[292,-0.0166],[301,-0.0035],[313,0.0512],[319,0.0643],[326,0.0696],[333,0.059],[343,0.0282],[351,0]]
       }
       spine_04_ctrl.Rotation.X {
        type=MovieSceneScriptingFloatChannel
        default=0
        thinned_from=47
        keys [[0,5],[3,5.0281],[9,5.2859],[13,5.3739],[19,5.2929],[25,5.0707],[31,4.8067],[39,4.6147],[41,4.6],[48,4.575],[54,4.725],[59,5.05],[64,5.2],[70,5.25],[76,5.2],[90,4.2865],[101,4.0865],[109,4.1865],[119,4.8],[126,5.15],[133,5.25],[152,3.7722],[159,3.6222],[166,3.7222],[172,4.0222],[178,4.75],[183,5.15],[188,5.25],[193,5],[198,4.125],[205,3.5],[212,3.7],[216,4.375],[220,4.85],[225,5.1],[232,5.225],[246,5.05],[262,4.3003],[280,2.176],[294,0.6176],[300,0.403],[309,0.3112],[321,0.6824],[338,4.2178],[343,4.8812],[351,5]]
       }
       spine_04_ctrl.Rotation.Y {type=MovieSceneScriptingFloatChannel, default=0, const [3,59]}
       spine_04_ctrl.Rotation.Z {type=MovieSceneScriptingFloatChannel, default=0, const [0,59]}
       spine_04_ctrl.Scale.X {type=MovieSceneScriptingFloatChannel, default=1, const [1,59]}
       spine_04_ctrl.Scale.Y {type=MovieSceneScriptingFloatChannel, default=1, const [1,59]}
       spine_04_ctrl.Scale.Z {type=MovieSceneScriptingFloatChannel, default=1, const [1,59]}
       head_ctrl.Location.X {type=MovieSceneScriptingFloatChannel, default=0, const [0,58]}
       head_ctrl.Location.Y {type=MovieSceneScriptingFloatChannel, default=0, const [0,58]}
       head_ctrl.Location.Z {type=MovieSceneScriptingFloatChannel, default=0, const [0,58]}
       head_ctrl.Rotation.X {
        type=MovieSceneScriptingFloatChannel
        default=0
        keys [[0,5.5],[3,4.852],[6,4.2501],[9,3.8775],[13,3.5721],[18,3.4424],[23,3.8833],[27,4.6189],[34,5.3131],[42,5.6228],[50,5.333],[60,4.8116],[68,4.5339],[76,5.2687],[80,4.7729],[86,2.0218],[93,-0.1679],[102,-1.0021],[114,-1.2051],[121,-0.848],[131,0.0958],[138,0.5146],[144,0.2164],[150,-0.3274],[156,-0.9474],[166,-1.5742],[172,-1.1444],[178,0.2554],[183,1.5055],[187,0.2082],[191,-2.0108],[196,-4.4755],[204,-6.2083],[214,-6.5927],[220,-6.1912],[224,-5.3659],[231,-3.653],[237,-2.7034],[244,-1.9712],[249,-1.6686],[255,-1.4995],[263,-1.9913],[271,-4.172],[278,-9.4211],[285,-13.337],[293,-14.8344],[307,-15.4918],[316,-14.0676],[323,-9.6658],[330,0.6624],[336,4.7815],[344,5.9814],[351,5.5]]
       }
       head_ctrl.Rotation.Y {
        type=MovieSceneScriptingFloatChannel
        default=0
        thinned_from=56
        keys [[0,8.5],[64,8.5],[70,7.9356],[82,5.3687],[90,4.1152],[101,3.5],[152,3.5],[159,3.5219],[166,3.5877],[172,4.1908],[178,5.792],[183,7.5943],[188,9.3126],[193,10.1475],[205,10.5],[255,10.5],[263,10.623],[271,11.3828],[277,12.201],[283,12.992],[292,13.354],[307,13.354],[313,12.7428],[319,11.432],[326,9.7718],[333,8.6941],[338,8.5],[351,8.5]]
       }
       head_ctrl.Rotation.Z {
        type=MovieSceneScriptingFloatChannel
        default=0
        thinned_from=54
        keys [[0,24],[54,24],[59,23.6524],[64,23.2173],[70,24.4426],[75,29.5],[80,38.5117],[86,42.3726],[96,44.3864],[109,45.1845],[115,45.4345],[119,45.2103],[126,44.0237],[139,41.0908],[145,40.5],[152,40.5],[159,40.3673],[166,39.5325],[172,36.3207],[178,25.375],[183,1.0813],[188,-20.4518],[195,-31.4141],[206,-36.5558],[220,-37.5855],[230,-36.3237],[249,-31.6844],[255,-30.0227],[262,-25.9868],[269,-14.716],[273,-0.0968],[279,14.0306],[288,20.1597],[299,22.2383],[319,24.0635],[326,24.5783],[333,24.914],[338,24.6836],[343,24.2869],[351,24]]
       }
       head_ctrl.Scale.X {type=MovieSceneScriptingFloatChannel, default=1, const [1,58]}
       head_ctrl.Scale.Y {type=MovieSceneScriptingFloatChannel, default=1, const [1,58]}
       head_ctrl.Scale.Z {type=MovieSceneScriptingFloatChannel, default=1, const [1,58]}
       clavicle_l_ctrl.Location.X {type=MovieSceneScriptingFloatChannel, default=0, const [0,58]}
       clavicle_l_ctrl.Location.Y {type=MovieSceneScriptingFloatChannel, default=0, const [0,58]}
       clavicle_l_ctrl.Location.Z {
        type=MovieSceneScriptingFloatChannel
        default=0
        keys [[0,0.5],[26,0.8629],[40,0.9927],[48,1.0185],[55,0.9224],[68,0.4208],[76,0.3241],[83,0.4988],[97,0.9698],[109,1.1285],[115,0.9809],[133,0.4563],[140,0.6539],[151,1.1547],[159,1.3396],[166,1.388],[172,1.2198],[183,0.7741],[188,0.692],[193,0.9226],[198,1.2968],[205,1.6064],[212,1.6564],[216,1.5532],[232,0.9277],[239,0.8511],[245,1.0275],[254,1.3532],[262,1.4317],[270,1.4166],[282,1.1734],[289,1.3275],[297,1.7108],[314,1.8196],[326,1.3992],[336,0.3541],[341,0.2048],[351,0.5]]
       }
       clavicle_l_ctrl.Rotation.X {type=MovieSceneScriptingFloatChannel, default=0, thinned_from=58, keys [[0,0],[351,0]]}
       clavicle_l_ctrl.Rotation.Y {type=MovieSceneScriptingFloatChannel, default=0, const [0,58]}
       clavicle_l_ctrl.Rotation.Z {type=MovieSceneScriptingFloatChannel, default=0, const [0,58]}
       clavicle_l_ctrl.Scale.X {type=MovieSceneScriptingFloatChannel, default=1, const [1,58]}
       clavicle_l_ctrl.Scale.Y {type=MovieSceneScriptingFloatChannel, default=1, const [1,58]}
       clavicle_l_ctrl.Scale.Z {type=MovieSceneScriptingFloatChannel, default=1, const [1,58]}
       clavicle_r_ctrl.Location.X {type=MovieSceneScriptingFloatChannel, default=0, const [0,57]}
       clavicle_r_ctrl.Location.Y {type=MovieSceneScriptingFloatChannel, default=0, const [0,57]}
       clavicle_r_ctrl.Location.Z {
        type=MovieSceneScriptingFloatChannel
        default=0
        keys [[0,0],[23,0.208],[40,0.2843],[48,0.293],[55,0.2446],[68,-0.0533],[76,-0.15],[83,-0.0752],[97,0.2457],[109,0.3544],[115,0.248],[133,-0.1178],[140,0.0011],[151,0.2938],[159,0.4544],[166,0.4947],[172,0.3702],[183,-0.05],[188,-0.1321],[193,0.0203],[198,0.2908],[205,0.5383],[212,0.5883],[216,0.5301],[232,0.1536],[239,0.0489],[245,0.1668],[254,0.4312],[262,0.5046],[270,0.5397],[280,0.3532],[289,0.6193],[297,0.92],[307,1.0597],[319,1.0597],[326,0.7087],[336,-0.1507],[341,-0.2984],[351,0]]
       }
       clavicle_r_ctrl.Rotation.X {type=MovieSceneScriptingFloatChannel, default=0, thinned_from=57, keys [[0,0],[351,0]]}
       clavicle_r_ctrl.Rotation.Y {type=MovieSceneScriptingFloatChannel, default=0, const [0,57]}
       clavicle_r_ctrl.Rotation.Z {type=MovieSceneScriptingFloatChannel, default=0, const [0,57]}
       clavicle_r_ctrl.Scale.X {type=MovieSceneScriptingFloatChannel, default=1, const [1,57]}
       clavicle_r_ctrl.Scale.Y {type=MovieSceneScriptingFloatChannel, default=1, const [1,57]}
       clavicle_r_ctrl.Scale.Z {type=MovieSceneScriptingFloatChannel, default=1, const [1,57]}
       neck_01_ctrl.Location.X {type=MovieSceneScriptingFloatChannel, default=0, const [0,58]}
       neck_01_ctrl.Location.Y {type=MovieSceneScriptingFloatChannel, default=0, const [0,58]}
       neck_01_ctrl.Location.Z {type=MovieSceneScriptingFloatChannel, default=0, const [0,58]}
       neck_01_ctrl.Rotation.X {type=MovieSceneScriptingFloatChannel, default=0, const [2.5,58]}
       neck_01_ctrl.Rotation.Y {type=MovieSceneScriptingFloatChannel, default=0, const [2.5,58]}
       neck_01_ctrl.Rotation.Z {type=MovieSceneScriptingFloatChannel, default=0, thinned_from=58, keys [[0,0],[351,0]]}
       neck_01_ctrl.Scale.X {type=MovieSceneScriptingFloatChannel, default=1, const [1,58]}
       neck_01_ctrl.Scale.Y {type=MovieSceneScriptingFloatChannel, default=1, const [1,58]}
       neck_01_ctrl.Scale.Z {type=MovieSceneScriptingFloatChannel, default=1, const [1,58]}
       neck_02_ctrl.Location.X {type=MovieSceneScriptingFloatChannel, default=0, const [0,58]}
       neck_02_ctrl.Location.Y {type=MovieSceneScriptingFloatChannel, default=0, const [0,58]}
       neck_02_ctrl.Location.Z {type=MovieSceneScriptingFloatChannel, default=0, const [0,58]}
       neck_02_ctrl.Rotation.X {type=MovieSceneScriptingFloatChannel, default=0, const [0,58]}
       neck_02_ctrl.Rotation.Y {type=MovieSceneScriptingFloatChannel, default=0, const [0,58]}
       neck_02_ctrl.Rotation.Z {type=MovieSceneScriptingFloatChannel, default=0, const [0,58]}
       neck_02_ctrl.Scale.X {type=MovieSceneScriptingFloatChannel, default=1, const [1,58]}
       neck_02_ctrl.Scale.Y {type=MovieSceneScriptingFloatChannel, default=1, const [1,58]}
       neck_02_ctrl.Scale.Z {type=MovieSceneScriptingFloatChannel, default=1, const [1,58]}
       head_ik_ctrl.Location.X {
        type=MovieSceneScriptingFloatChannel
        default=0
        thinned_from=58
        keys [[0,-0.5689],[82,-0.5685],[90,-0.5724],[101,-0.5737],[109,-0.5704],[115,-0.5728],[145,-0.5736],[152,-0.572],[205,-0.5708],[212,-0.5689],[216,-0.5657],[220,-0.5639],[244,-0.5612],[249,-0.5596],[263,-0.5586],[271,-0.5605],[277,-0.5663],[283,-0.5679],[351,-0.5689]]
       }
       head_ik_ctrl.Location.Y {
        type=MovieSceneScriptingFloatChannel
        default=0
        thinned_from=58
        keys [[0,0.1103],[13,0.1104],[19,0.1122],[25,0.1153],[31,0.1172],[41,0.1183],[48,0.116],[54,0.1123],[59,0.1111],[64,0.1103],[76,0.1133],[82,0.1164],[90,0.0927],[101,0.0847],[109,0.1064],[115,0.0916],[119,0.0879],[126,0.0867],[145,0.0867],[152,0.098],[159,0.0965],[166,0.0917],[172,0.0928],[178,0.0934],[183,0.096],[188,0.0976],[193,0.0994],[198,0.0999],[205,0.1001],[212,0.1097],[216,0.1257],[220,0.1353],[225,0.1389],[232,0.141],[239,0.1422],[244,0.1442],[249,0.1475],[255,0.1487],[263,0.1495],[271,0.1421],[277,0.1201],[283,0.114],[292,0.1112],[301,0.1103],[351,0.1103]]
       }
       head_ik_ctrl.Location.Z {
        type=MovieSceneScriptingFloatChannel
        default=0
        thinned_from=58
        keys [[0,-0.1581],[9,-0.1595],[13,-0.1596],[19,-0.1583],[25,-0.156],[31,-0.1546],[41,-0.1538],[48,-0.155],[54,-0.157],[64,-0.1581],[76,-0.1565],[82,-0.1548],[90,-0.1559],[101,-0.1563],[109,-0.1546],[115,-0.1555],[145,-0.1558],[152,-0.1547],[159,-0.1552],[166,-0.1566],[178,-0.1569],[193,-0.1579],[205,-0.158],[212,-0.1575],[216,-0.1565],[220,-0.156],[232,-0.1557],[239,-0.1566],[244,-0.1582],[249,-0.1608],[255,-0.1618],[263,-0.1624],[271,-0.1616],[277,-0.1591],[283,-0.1585],[351,-0.1581]]
       }
       head_ik_ctrl.Rotation.X {
        type=MovieSceneScriptingFloatChannel
        default=0
        thinned_from=58
        keys [[0,-14],[3,-13.9238],[6,-13.7969],[9,-13.7121],[13,-13.675],[19,-13.3996],[25,-12.9406],[31,-12.6652],[41,-12.5],[48,-12.9219],[54,-13.625],[59,-13.8594],[64,-14],[70,-12.6491],[76,-11.375],[82,-8.6406],[90,-7.4102],[101,-7],[145,-7],[152,-5.3125],[159,-5.4043],[166,-5.6798],[172,-5.4048],[178,-5.2398],[183,-4.5602],[188,-4.1524],[193,-3.6631],[198,-3.5408],[205,-3.5],[263,-3.5],[271,-5.4688],[277,-11.375],[283,-13.0156],[292,-13.7539],[301,-14],[351,-14]]
       }
       head_ik_ctrl.Rotation.Y {
        type=MovieSceneScriptingFloatChannel
        default=0
        thinned_from=58
        keys [[0,14],[64,14],[70,11.9737],[76,10.0625],[82,5.9609],[90,4.1152],[101,3.5],[205,3.5],[212,3.6172],[216,3.8125],[220,3.9297],[225,3.9736],[232,4],[239,4.9141],[244,6.4375],[249,8.9766],[255,9.9287],[263,10.5],[271,11.1562],[277,13.125],[283,13.6719],[292,13.918],[301,14],[351,14]]
       }
       head_ik_ctrl.Rotation.Z {
        type=MovieSceneScriptingFloatChannel
        default=0
        thinned_from=58
        keys [[0,20],[64,20],[70,23.9561],[76,27.6875],[82,35.6953],[90,40.7988],[101,42.5],[109,40.5],[205,40.5],[212,26.3203],[216,2.6875],[220,-11.4922],[225,-16.8096],[232,-20],[239,-20.3516],[244,-20.9375],[249,-21.9141],[255,-22.2803],[263,-22.5],[271,-14.5312],[277,9.375],[283,16.0156],[292,19.0039],[301,20],[351,20]]
       }
       head_ik_ctrl.Scale.X {type=MovieSceneScriptingFloatChannel, default=1, const [1,58]}
       head_ik_ctrl.Scale.Y {type=MovieSceneScriptingFloatChannel, default=1, const [1,58]}
       head_ik_ctrl.Scale.Z {type=MovieSceneScriptingFloatChannel, default=1, const [1,58]}
       chest_ctrl.Location.X {type=MovieSceneScriptingFloatChannel, default=0, thinned_from=58, keys [[0,0],[351,0]]}
       chest_ctrl.Location.Y {type=MovieSceneScriptingFloatChannel, default=0, const [2.3934,58]}
       chest_ctrl.Location.Z {type=MovieSceneScriptingFloatChannel, default=0, thinned_from=58, keys [[0,-0.093],[351,-0.093]]}
       chest_ctrl.Rotation.X {
        type=MovieSceneScriptingFloatChannel
        default=0
        thinned_from=58
        keys [[0,25.3113],[9,25.4885],[13,25.5113],[19,25.3011],[25,24.9507],[31,24.7404],[41,24.6143],[48,24.8103],[54,25.1371],[64,25.3113],[76,25.0499],[82,24.7777],[90,24.3353],[101,24.1878],[109,24.5481],[119,24.4447],[145,24.4378],[152,24.3146],[159,24.3535],[166,24.4702],[205,24.5235],[212,24.4851],[220,24.3828],[232,24.3597],[239,24.4898],[244,24.7066],[249,25.0678],[255,25.2033],[263,25.2846],[351,25.3113]]
       }
       chest_ctrl.Rotation.Y {
        type=MovieSceneScriptingFloatChannel
        default=0
        thinned_from=58
        keys [[0,-3.2653],[13,-3.2653],[19,-3.2333],[25,-3.1801],[31,-3.1481],[41,-3.129],[48,-3.1673],[54,-3.2312],[59,-3.2525],[64,-3.2653],[76,-3.2142],[82,-3.1609],[90,-3.4679],[101,-3.5702],[109,-3.2944],[115,-3.5013],[119,-3.553],[126,-3.5702],[145,-3.5702],[152,-3.4128],[159,-3.4324],[166,-3.4911],[178,-3.4647],[183,-3.4239],[193,-3.3702],[205,-3.3604],[212,-3.1124],[216,-2.6991],[220,-2.4511],[225,-2.3581],[232,-2.3023],[239,-2.283],[244,-2.2508],[249,-2.1971],[255,-2.177],[263,-2.1649],[271,-2.3712],[277,-2.9902],[283,-3.1621],[292,-3.2395],[301,-3.2653],[351,-3.2653]]
       }
       chest_ctrl.Rotation.Z {
        type=MovieSceneScriptingFloatChannel
        default=0
        thinned_from=58
        keys [[0,21.4631],[13,21.4631],[19,21.3504],[25,21.1624],[31,21.0497],[41,20.982],[48,21.1173],[54,21.3428],[64,21.4631],[76,21.2827],[82,21.0948],[90,22.1834],[101,22.5463],[109,21.5686],[115,22.3019],[119,22.4852],[126,22.5463],[145,22.5463],[152,21.9866],[159,22.0561],[166,22.2648],[178,22.171],[183,22.0261],[193,21.8348],[205,21.8],[212,21.42],[216,20.7867],[220,20.4066],[225,20.2641],[239,20.1111],[244,19.9986],[249,19.8111],[255,19.7407],[263,19.6985],[271,20.0294],[277,21.022],[283,21.2977],[292,21.4218],[351,21.4631]]
       }
       chest_ctrl.Scale.X {type=MovieSceneScriptingFloatChannel, default=1, const [1,58]}
       chest_ctrl.Scale.Y {type=MovieSceneScriptingFloatChannel, default=1, const [1,58]}
       chest_ctrl.Scale.Z {type=MovieSceneScriptingFloatChannel, default=1, const [1,58]}
       chest_tan_ctrl.Location.X {
        type=MovieSceneScriptingFloatChannel
        default=0
        thinned_from=58
        keys [[0,-0.6201],[41,-0.6182],[64,-0.6201],[82,-0.6186],[90,-0.6228],[101,-0.6242],[109,-0.6205],[115,-0.6233],[145,-0.6242],[152,-0.6221],[205,-0.6214],[212,-0.6438],[216,-0.6812],[220,-0.7036],[225,-0.712],[232,-0.717],[263,-0.7152],[271,-0.6973],[277,-0.6438],[283,-0.629],[292,-0.6223],[301,-0.6201],[351,-0.6201]]
       }
       chest_tan_ctrl.Location.Y {
        type=MovieSceneScriptingFloatChannel
        default=-0.0303
        thinned_from=58
        keys [[0,0.3257],[3,0.334],[6,0.3478],[9,0.357],[13,0.361],[19,0.3215],[25,0.2557],[31,0.2161],[41,0.1924],[48,0.2299],[54,0.2924],[59,0.3132],[64,0.3257],[70,0.3],[76,0.2757],[82,0.2237],[90,0.1688],[101,0.1505],[109,0.1933],[115,0.1944],[145,0.1948],[152,0.1607],[159,0.1691],[166,0.1944],[205,0.1936],[212,0.1783],[216,0.1527],[220,0.1374],[225,0.1316],[232,0.1281],[239,0.1498],[244,0.1859],[249,0.2461],[255,0.2687],[263,0.2822],[271,0.2904],[277,0.3148],[283,0.3216],[292,0.3247],[301,0.3257],[351,0.3257]]
       }
       chest_tan_ctrl.Location.Z {
        type=MovieSceneScriptingFloatChannel
        default=0.265
        thinned_from=58
        keys [[0,0.0421],[3,0.0434],[6,0.0454],[9,0.0467],[13,0.0473],[19,0.0417],[25,0.0324],[31,0.0268],[41,0.0235],[48,0.0287],[54,0.0375],[59,0.0404],[64,0.0421],[70,0.0385],[76,0.0351],[82,0.0278],[90,0.0207],[101,0.0184],[109,0.0237],[119,0.0241],[145,0.0241],[152,0.0196],[159,0.0207],[166,0.024],[205,0.0238],[212,0.0233],[216,0.0225],[220,0.022],[232,0.0216],[239,0.0245],[244,0.0293],[249,0.0373],[255,0.0403],[263,0.0421],[351,0.0421]]
       }
       chest_tan_ctrl.Rotation.X {type=MovieSceneScriptingFloatChannel, default=0, thinned_from=58, keys [[0,0],[351,0]]}
       chest_tan_ctrl.Rotation.Y {type=MovieSceneScriptingFloatChannel, default=0, const [0,58]}
       chest_tan_ctrl.Rotation.Z {type=MovieSceneScriptingFloatChannel, default=0, thinned_from=58, keys [[0,0],[351,0]]}
       chest_tan_ctrl.Scale.X {type=MovieSceneScriptingFloatChannel, default=1, const [1,58]}
       chest_tan_ctrl.Scale.Y {type=MovieSceneScriptingFloatChannel, default=1, const [1,58]}
       chest_tan_ctrl.Scale.Z {type=MovieSceneScriptingFloatChannel, default=1, const [1,58]}
       Weight {type=MovieSceneScriptingFloatChannel, default=1}
      ]
     }]
   }
   {
    type=MovieSceneControlRigParameterTrack
    display_name=Fortnite_Legacy_Face_CtrlRig
    sections [{
      type=MovieSceneControlRigParameterSection
      channels [
       Flipbook {type=MovieSceneScriptingBoolChannel, default=false}
       Face_Ctrl_Master.Location.X {type=MovieSceneScriptingFloatChannel, default=-18.1678}
       Face_Ctrl_Master.Location.Y {type=MovieSceneScriptingFloatChannel, default=0}
       Face_Ctrl_Master.Location.Z {type=MovieSceneScriptingFloatChannel, default=0}
       Face_Ctrl_Master.Rotation.X {type=MovieSceneScriptingFloatChannel, default=0}
       Face_Ctrl_Master.Rotation.Y {type=MovieSceneScriptingFloatChannel, default=0}
       Face_Ctrl_Master.Rotation.Z {type=MovieSceneScriptingFloatChannel, default=0}
       Face_Ctrl_Master.Scale.X {type=MovieSceneScriptingFloatChannel, default=1}
       Face_Ctrl_Master.Scale.Y {type=MovieSceneScriptingFloatChannel, default=1}
       Face_Ctrl_Master.Scale.Z {type=MovieSceneScriptingFloatChannel, default=1}
       gui_follow_head {type=MovieSceneScriptingFloatChannel, default=1}
      ]
     }]
   }
   {type=MovieSceneSkeletalAnimationTrack, sections [{type=MovieSceneSkeletalAnimationSection, end_frame=352, start_frame=0}]}
  ]
 }
 mannequin {
  id=#2
  possessed_class=FortPlayerMannequin
  tracks [
   {type=MovieSceneSpawnTrack, sections [{type=MovieSceneSpawnSection, channels [None {type=MovieSceneScriptingBoolChannel, default=true}]}]}
   {
    type=MovieScene3DTransformTrack
    sections [{
      type=MovieScene3DTransformSection
      channels [
       Location.X {type=MovieSceneScriptingDoubleChannel, default=78.4909}
       Location.Y {type=MovieSceneScriptingDoubleChannel, default=0}
       Location.Z {type=MovieSceneScriptingDoubleChannel, default=0}
       Rotation.X {type=MovieSceneScriptingDoubleChannel, default=0}
       Rotation.Y {type=MovieSceneScriptingDoubleChannel, default=0}
       Rotation.Z {type=MovieSceneScriptingDoubleChannel, default=-90}
       Scale.X {type=MovieSceneScriptingDoubleChannel, default=1}
       Scale.Y {type=MovieSceneScriptingDoubleChannel, default=1}
       Scale.Z {type=MovieSceneScriptingDoubleChannel, default=1}
      ]
     }]
   }
   {type=MovieSceneSkeletalAnimationTrack, sections [{type=MovieSceneSkeletalAnimationSection, end_frame=352, start_frame=0}]}
  ]
 }
 mannequin {
  id=#3
  possessed_class=FortPlayerMannequin
  tracks [
   {type=MovieSceneSpawnTrack, sections [{type=MovieSceneSpawnSection, channels [None {type=MovieSceneScriptingBoolChannel, default=true}]}]}
   {
    type=MovieScene3DTransformTrack
    sections [{
      type=MovieScene3DTransformSection
      channels [
       Location.X {type=MovieSceneScriptingDoubleChannel, default=172.4275}
       Location.Y {type=MovieSceneScriptingDoubleChannel, default=0}
       Location.Z {type=MovieSceneScriptingDoubleChannel, default=0}
       Rotation.X {type=MovieSceneScriptingDoubleChannel, default=0}
       Rotation.Y {type=MovieSceneScriptingDoubleChannel, default=0}
       Rotation.Z {type=MovieSceneScriptingDoubleChannel, default=180}
       Scale.X {type=MovieSceneScriptingDoubleChannel, default=1}
       Scale.Y {type=MovieSceneScriptingDoubleChannel, default=1}
       Scale.Z {type=MovieSceneScriptingDoubleChannel, default=1}
      ]
     }]
   }
   {type=MovieSceneSkeletalAnimationTrack, sections [{type=MovieSceneSkeletalAnimationSection, end_frame=351, start_frame=-1}]}
  ]
 }
 mannequin {
  id=#4
  possessed_class=FortPlayerMannequin
  tracks [
   {type=MovieSceneSpawnTrack, sections [{type=MovieSceneSpawnSection, channels [None {type=MovieSceneScriptingBoolChannel, default=true}]}]}
   {
    type=MovieScene3DTransformTrack
    sections [{
      type=MovieScene3DTransformSection
      channels [
       Location.X {type=MovieSceneScriptingDoubleChannel, default=-87.7352}
       Location.Y {type=MovieSceneScriptingDoubleChannel, default=0}
       Location.Z {type=MovieSceneScriptingDoubleChannel, default=0}
       Rotation.X {type=MovieSceneScriptingDoubleChannel, default=0}
       Rotation.Y {type=MovieSceneScriptingDoubleChannel, default=0}
       Rotation.Z {type=MovieSceneScriptingDoubleChannel, default=90}
       Scale.X {type=MovieSceneScriptingDoubleChannel, default=1}
       Scale.Y {type=MovieSceneScriptingDoubleChannel, default=1}
       Scale.Z {type=MovieSceneScriptingDoubleChannel, default=1}
      ]
     }]
   }
   {type=MovieSceneSkeletalAnimationTrack, sections [{type=MovieSceneSkeletalAnimationSection, end_frame=352, start_frame=0}]}
  ]
 }
]
display_rate {denominator=1, numerator=30}
tick_resolution {denominator=1, numerator=24000}
tracks []
